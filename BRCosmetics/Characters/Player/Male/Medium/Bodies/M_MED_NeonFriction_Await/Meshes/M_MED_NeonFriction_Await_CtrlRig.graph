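# Control Rig for the NeonFriction Await male body cosmetic that drives its skirt and sleeve bones from limb motion.
# Parent constraints blend skirt bones between thighs and pelvis, and sleeve bones along the arms.
# DeltaRotation v4 converts swing or twist against reference nulls into clamped weights for corrective ModifyTransforms.
# Multiplied weights drive combination correctives, while spherical pose readers and a passenger flag add further skirt adjustments.
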ControlRigBlueprint /BRCosmetics/Characters/Player/Male/Medium/Bodies/M_MED_NeonFriction_Await/Meshes/M_MED_NeonFriction_Await_CtrlRig
graphs [
 Multiply_18_ContainedGraph {nodes [
   n1 = RigVMFunctionEntryNode()
   RigVMFunctionReturnNode(Result=&RigVMFunction_MathDoubleMul(B=@n1.C, A=&RigVMFunction_MathDoubleMul(A=@n1.A, B=@n1.B).Result).Result)
  ]}
 Multiply_18_ContainedGraph {nodes [
   n1 = RigVMFunctionEntryNode()
   RigVMFunctionReturnNode(Result=&RigVMFunction_MathDoubleMul(B=@n1.C, A=&RigVMFunction_MathDoubleMul(A=@n1.A, B=@n1.B).Result).Result)
  ]}
 Multiply_22_ContainedGraph {nodes [
   n1 = RigVMFunctionEntryNode()
   RigVMFunctionReturnNode(Result=&RigVMFunction_MathDoubleMul(B=@n1.C, A=&RigVMFunction_MathDoubleMul(A=@n1.A, B=@n1.B).Result).Result)
  ]}
 Multiply_22_ContainedGraph {nodes [
   n1 = RigVMFunctionEntryNode()
   RigVMFunctionReturnNode(Result=&RigVMFunction_MathDoubleMul(B=@n1.C, A=&RigVMFunction_MathDoubleMul(A=@n1.A, B=@n1.B).Result).Result)
  ]}
 Multiply_32_ContainedGraph {nodes [
   n1 = RigVMFunctionEntryNode()
   RigVMFunctionReturnNode(Result=&RigVMFunction_MathDoubleMul(B=@n1.C, A=&RigVMFunction_MathDoubleMul(A=@n1.A, B=@n1.B).Result).Result)
  ]}
 Multiply_32_ContainedGraph {nodes [
   n1 = RigVMFunctionEntryNode()
   RigVMFunctionReturnNode(Result=&RigVMFunction_MathDoubleMul(B=@n1.C, A=&RigVMFunction_MathDoubleMul(A=@n1.A, B=@n1.B).Result).Result)
  ]}
 Multiply_34_ContainedGraph {nodes [
   n1 = RigVMFunctionEntryNode()
   RigVMFunctionReturnNode(Result=&RigVMFunction_MathDoubleMul(B=@n1.C, A=&RigVMFunction_MathDoubleMul(A=@n1.A, B=@n1.B).Result).Result)
  ]}
 Multiply_3_ContainedGraph {nodes [
   n1 = RigVMFunctionEntryNode()
   RigVMFunctionReturnNode(Result=&RigVMFunction_MathDoubleMul(B=@n1.C, A=&RigVMFunction_MathDoubleMul(A=@n1.A, B=@n1.B).Result).Result)
  ]}
 Multiply_8_ContainedGraph {nodes [
   n1 = RigVMFunctionEntryNode()
   RigVMFunctionReturnNode(Result=&RigVMFunction_MathDoubleMul(B=@n1.C, A=&RigVMFunction_MathDoubleMul(A=@n1.A, B=@n1.B).Result).Result)
  ]}
 Multiply_8_ContainedGraph {nodes [
   n1 = RigVMFunctionEntryNode()
   RigVMFunctionReturnNode(Result=&RigVMFunction_MathDoubleMul(B=@n1.C, A=&RigVMFunction_MathDoubleMul(A=@n1.A, B=@n1.B).Result).Result)
  ]}
 New Function_ContainedGraph {nodes [
   n1 = RigVMFunctionEntryNode()
   RigVMFunction_Sequence = &RigVMFunction_Sequence(ExecuteContext=@n1.ExecuteContext)
   &RigUnit_ParentConstraint(
    AdvancedSettings=(InterpolationType=Average,RotationOrderForFilter=XZY)
    Child=(Type=Bone,Name="MSTR_frontSkirt_l")
    Filter=(TranslationFilter=(bX=false,bY=false,bZ=false),RotationFilter=(bX=True,bY=True,bZ=True),ScaleFilter=(bX=false,bY=false,bZ=false))
    Parents=((Item=(Type=Bone,Name="thigh_l"),Weight=1),(Item=(Type=Bone,Name="pelvis"),Weight=0))
    Weight=1
    ExecutePin=@RigVMFunction_Sequence.A
   )
   > &RigUnit_ParentConstraint(
    AdvancedSettings=(InterpolationType=Average,RotationOrderForFilter=XZY)
    Child=(Type=Bone,Name="MSTR_sideSkirt_l")
    Filter=(TranslationFilter=(bX=false,bY=false,bZ=false),RotationFilter=(bX=True,bY=True,bZ=True),ScaleFilter=(bX=false,bY=false,bZ=false))
    Parents=((Item=(Type=Bone,Name="thigh_l"),Weight=0.75),(Item=(Type=Bone,Name="pelvis"),Weight=0.25))
    Weight=1
   )
   > &RigUnit_ParentConstraint(
    AdvancedSettings=(InterpolationType=Average,RotationOrderForFilter=XZY)
    Child=(Type=Bone,Name="MSTR_backSkirt_l")
    Filter=(TranslationFilter=(bX=false,bY=false,bZ=false),RotationFilter=(bX=True,bY=True,bZ=True),ScaleFilter=(bX=false,bY=false,bZ=false))
    Parents=((Item=(Type=Bone,Name="thigh_l"),Weight=0.25),(Item=(Type=Bone,Name="pelvis"),Weight=0.75))
    Weight=1
   )
   > &RigUnit_ParentConstraint(
    AdvancedSettings=(InterpolationType=Average,RotationOrderForFilter=XZY)
    Child=(Type=Bone,Name="MSTR_frontSkirt_r")
    Filter=(TranslationFilter=(bX=false,bY=false,bZ=false),RotationFilter=(bX=True,bY=True,bZ=True),ScaleFilter=(bX=false,bY=false,bZ=false))
    Parents=((Item=(Type=Bone,Name="thigh_r"),Weight=1),(Item=(Type=Bone,Name="pelvis"),Weight=0))
    Weight=1
   )
   > &RigUnit_ParentConstraint(
    AdvancedSettings=(InterpolationType=Average,RotationOrderForFilter=XZY)
    Child=(Type=Bone,Name="MSTR_sideSkirt_r")
    Filter=(TranslationFilter=(bX=false,bY=false,bZ=false),RotationFilter=(bX=True,bY=True,bZ=True),ScaleFilter=(bX=false,bY=false,bZ=false))
    Parents=((Item=(Type=Bone,Name="thigh_r"),Weight=0.75),(Item=(Type=Bone,Name="pelvis"),Weight=0.25))
    Weight=1
   )
   > &RigUnit_ParentConstraint(
    AdvancedSettings=(InterpolationType=Average,RotationOrderForFilter=XZY)
    Child=(Type=Bone,Name="MSTR_backSkirt_r")
    Filter=(TranslationFilter=(bX=false,bY=false,bZ=false),RotationFilter=(bX=True,bY=True,bZ=True),ScaleFilter=(bX=false,bY=false,bZ=false))
    Parents=((Item=(Type=Bone,Name="thigh_r"),Weight=0.25),(Item=(Type=Bone,Name="pelvis"),Weight=0.75))
    Weight=1
   )
   > &RigUnit_ParentConstraint(
    AdvancedSettings=(InterpolationType=Average,RotationOrderForFilter=XZY)
    Child=(Type=Bone,Name="MSTR_frontSkirt")
    Filter=(TranslationFilter=(bX=True,bY=True,bZ=True),RotationFilter=(bX=True,bY=True,bZ=True),ScaleFilter=(bX=false,bY=false,bZ=false))
    Parents=((Item=(Type=Bone,Name="MSTR_frontSkirt_l"),Weight=0.5),(Item=(Type=Bone,Name="MSTR_frontSkirt_r"),Weight=0.5))
    Weight=1
   )
   > &RigUnit_ParentConstraint(
    AdvancedSettings=(InterpolationType=Average,RotationOrderForFilter=XZY)
    Child=(Type=Bone,Name="MSTR_backSkirt")
    Parents=((Item=(Type=Bone,Name="MSTR_backSkirt_l"),Weight=0.5),(Item=(Type=Bone,Name="MSTR_backSkirt_r"),Weight=0.5))
    Weight=1
   )
   > RigVMFunction_Sequence_1 = RigVMAggregateNode()
   DeltaRotation v4_8 = RigVMFunctionReferenceNode(
    Rotation Order=ZYX
    child=(Type=Bone,Name="thigh_l")
    debugRotation=false
    outputMinMax=(X=0,Y=1)
    parent=(Type=Null,Name="thigh_l_null")
    premult=1
    rotationMinMax=(X=-20,Y=-120)
    triggerAxis=Z
    twistAxis=(X=1,Y=0,Z=0)
    ExecuteContext=@RigVMFunction_Sequence_1.A
   )
   > &RigUnit_ModifyTransforms(
    ItemToModify="((Item=(Type=Bone,Name=\"dyn_frontSkirt_01_l\"),Transform=(Rotation=(X=-0.2119,Y=0.1066,Z=-0.0014,W=0.9715),Translation=(X=5.0592,Y=19.7247,Z=15.8247),Scale3D=(X=1,Y=1,Z=1))),(Item=(Type=Bone,Name=\"dyn_sideSkirt_01_l\"),Transform=(Rotation=(X=-0.0017,Y=0.0274 ...(truncated)"
    WeightMinimum=0
    Weight=@"DeltaRotation v4_8".outputWeight
   )
   > DeltaRotation v4_9 = RigVMFunctionReferenceNode(
    Rotation Order=ZYX
    child=(Type=Bone,Name="thigh_r")
    outputMinMax=(X=0,Y=1)
    parent=(Type=Null,Name="thigh_r_null")
    premult=1
    rotationMinMax=(X=-20,Y=-120)
    triggerAxis=Z
    twistAxis=(X=1,Y=0,Z=0)
   )
   > &RigUnit_ModifyTransforms(
    ItemToModify="((Item=(Type=Bone,Name=\"dyn_frontSkirt_01_r\"),Transform=(Rotation=(X=0.2119,Y=0.1066,Z=0.0014,W=0.9715),Translation=(X=5.0592,Y=-19.7247,Z=15.8247),Scale3D=(X=1,Y=1,Z=1))),(Item=(Type=Bone,Name=\"dyn_sideSkirt_01_r\"),Transform=(Rotation=(X=0,Y=0,Z=-0.061,W= ...(truncated)"
    WeightMinimum=0
    Weight=@"DeltaRotation v4_9".outputWeight
   )
   Multiply = &RigVMFunction_MathDoubleMul(A=@"DeltaRotation v4_9".outputWeight, B=@"DeltaRotation v4_8".outputWeight)
   > &RigUnit_ModifyTransforms(
    ItemToModify="((Item=(Type=Bone,Name=\"dyn_frontSkirt_01_r\"),Transform=(Rotation=(X=0,Y=0,Z=0,W=1),Translation=(X=0,Y=0,Z=0),Scale3D=(X=1,Y=1,Z=1))),(Item=(Type=Bone,Name=\"dyn_frontSkirt_01_l\"),Transform=(Rotation=(X=0,Y=0,Z=0,W=1),Translation=(X=0,Y=0,Z=0),Scale3D=(X=1, ...(truncated)"
    WeightMinimum=0
    Weight=@Multiply.Result
   )
   > DeltaRotation v4_6 = RigVMFunctionReferenceNode(
    Rotation Order=ZYX
    child=(Type=Bone,Name="thigh_l")
    debugRotation=false
    outputMinMax=(X=0,Y=1)
    parent=(Type=Null,Name="thigh_l_null")
    premult=1
    rotationMinMax=(X=0,Y=-20)
    triggerAxis=Z
    twistAxis=(X=1,Y=0,Z=0)
   )
   > &RigUnit_ModifyTransforms(
    ItemToModify="((Item=(Type=Bone,Name=\"dyn_frontSkirt_01_l\"),Transform=(Rotation=(X=0,Y=0,Z=0,W=1),Translation=(X=0,Y=0,Z=0),Scale3D=(X=1,Y=1,Z=1))),(Item=(Type=Bone,Name=\"dyn_sideSkirt_01_l\"),Transform=(Rotation=(X=0,Y=0,Z=0,W=1),Translation=(X=0,Y=0,Z=0),Scale3D=(X=1,Y ...(truncated)"
    WeightMinimum=0
    Weight=@"DeltaRotation v4_6".outputWeight
   )
   > DeltaRotation v4_7 = RigVMFunctionReferenceNode(
    Rotation Order=ZYX
    child=(Type=Bone,Name="thigh_r")
    outputMinMax=(X=0,Y=1)
    parent=(Type=Null,Name="thigh_r_null")
    premult=1
    rotationMinMax=(X=0,Y=-20)
    triggerAxis=Z
    twistAxis=(X=1,Y=0,Z=0)
   )
   > &RigUnit_ModifyTransforms(
    ItemToModify="((Item=(Type=Bone,Name=\"dyn_frontSkirt_01_r\"),Transform=(Rotation=(X=0,Y=0,Z=0,W=1),Translation=(X=0,Y=0,Z=0),Scale3D=(X=1,Y=1,Z=1))),(Item=(Type=Bone,Name=\"dyn_sideSkirt_01_r\"),Transform=(Rotation=(X=0,Y=0,Z=0,W=1),Translation=(X=0,Y=0,Z=0),Scale3D=(X=1,Y ...(truncated)"
    WeightMinimum=0
    Weight=@"DeltaRotation v4_7".outputWeight
   )
   > &RigUnit_ModifyTransforms(
    ItemToModify=((Item=(Type=Bone,Name="dyn_frontSkirt_01"),Transform=(Rotation=(X=0,Y=0,Z=0,W=1),Translation=(X=0,Y=0,Z=0),Scale3D=(X=1,Y=1,Z=1))))
    WeightMinimum=0
    Weight=&RigVMFunction_MathDoubleMul(B=@"DeltaRotation v4_7".outputWeight, A=@"DeltaRotation v4_6".outputWeight).Result
   )
   DeltaRotation v4_1 = RigVMFunctionReferenceNode(
    Rotation Order=ZYX
    child=(Type=Bone,Name="thigh_l")
    outputMinMax=(X=0,Y=1)
    parent=(Type=Null,Name="thigh_l_null")
    premult=1
    rotationMinMax=(X=0,Y=75)
    triggerAxis=Z
    twistAxis=(X=1,Y=0,Z=0)
    ExecuteContext=@RigVMFunction_Sequence_1.B
   )
   > &RigUnit_ModifyTransforms(
    ItemToModify="((Item=(Type=Bone,Name=\"dyn_backSkirt_02\"),Transform=(Rotation=(X=0,Y=0.3461,Z=0,W=-0.9382),Translation=(X=-2.2,Y=0,Z=2.6),Scale3D=(X=1,Y=1,Z=1))),(Item=(Type=Bone,Name=\"dyn_backSkirt_01_l\"),Transform=(Rotation=(X=0.1088,Y=-0.4692,Z=-0.2183,W=0.8488),Trans ...(truncated)"
    WeightMinimum=0
    Weight=@"DeltaRotation v4_1".outputWeight
   )
   > DeltaRotation v4_10 = RigVMFunctionReferenceNode(
    Rotation Order=ZYX
    child=(Type=Bone,Name="thigh_r")
    debugRotation=false
    outputMinMax=(X=0,Y=1)
    parent=(Type=Null,Name="thigh_r_null")
    premult=1
    rotationMinMax=(X=0,Y=75)
    triggerAxis=Z
    twistAxis=(X=1,Y=0,Z=0)
   )
   > &RigUnit_ModifyTransforms(
    ItemToModify="((Item=(Type=Bone,Name=\"dyn_backSkirt_02\"),Transform=(Rotation=(X=0,Y=-0.3461,Z=0,W=0.9382),Translation=(X=-2.2,Y=0,Z=2.6),Scale3D=(X=1,Y=1,Z=1))),(Item=(Type=Bone,Name=\"dyn_backSkirt_01_r\"),Transform=(Rotation=(X=-0.1088,Y=-0.4692,Z=0.2183,W=0.8488),Trans ...(truncated)"
    WeightMinimum=0
    Weight=@"DeltaRotation v4_10".outputWeight
   )
   > &RigUnit_ModifyTransforms(
    ItemToModify="((Item=(Type=Bone,Name=\"dyn_backSkirt_02\"),Transform=(Rotation=(X=-0.0616,Y=-0.7044,Z=0.0616,W=-0.7044),Translation=(X=-1.4207,Y=0.2505,Z=-7.1132),Scale3D=(X=1,Y=1,Z=1))),(Item=(Type=Bone,Name=\"dyn_frontSkirt_01\"),Transform=(Rotation=(X=0,Y=-0.113,Z=0,W=0. ...(truncated)"
    WeightMinimum=0
    Weight=&RigVMFunction_MathDoubleMul(A=@"DeltaRotation v4_10".outputWeight, B=@"DeltaRotation v4_1".outputWeight).Result
   )
   > &RigUnit_ModifyTransforms(
    ItemToModify="((Item=(Type=Bone,Name=\"dyn_backSkirt_01_r\"),Transform=(Rotation=(X=0.0046,Y=-0.0434,Z=0.1044,W=0.9936),Translation=(X=0,Y=0,Z=0),Scale3D=(X=1,Y=1,Z=1))),(Item=(Type=Bone,Name=\"dyn_frontSkirt_01\"),Transform=(Rotation=(X=-0.0747,Y=-0.0974,Z=0.122,W=0.9849), ...(truncated)"
    WeightMinimum=0
    Weight=&RigVMFunction_MathDoubleMul(B=@"DeltaRotation v4_10".outputWeight, A=@"DeltaRotation v4_8".outputWeight).Result
   )
   > &RigUnit_ModifyTransforms(
    ItemToModify="((Item=(Type=Bone,Name=\"dyn_backSkirt_01_l\"),Transform=(Rotation=(X=0.0946,Y=0.0106,Z=-0.0414,W=-0.9946),Translation=(X=-0.0152,Y=0.0984,Z=0.5157),Scale3D=(X=1,Y=1,Z=1))),(Item=(Type=Bone,Name=\"dyn_frontSkirt_01\"),Transform=(Rotation=(X=0.0596,Y=-0.1013,Z= ...(truncated)"
    WeightMinimum=0
    Weight=&RigVMFunction_MathDoubleMul(B=@"DeltaRotation v4_1".outputWeight, A=@"DeltaRotation v4_9".outputWeight).Result
   )
   > DeltaRotation v4_11 = RigVMFunctionReferenceNode(
    Rotation Order=ZYX
    child=(Type=Bone,Name="thigh_l")
    outputMinMax=(X=0,Y=1)
    parent=(Type=Null,Name="thigh_l_null")
    premult=1
    rotationMinMax=(X=0,Y=12)
    triggerAxis=Z
    twistAxis=(X=1,Y=0,Z=0)
   )
   > &RigUnit_ModifyTransforms(
    ItemToModify="((Item=(Type=Bone,Name=\"dyn_backSkirt_01_l\"),Transform=(Rotation=(X=0,Y=0,Z=0.0523,W=0.9986),Translation=(X=0,Y=0,Z=0),Scale3D=(X=1,Y=1,Z=1))),(Item=(Type=Bone,Name=\"dyn_sideSkirt_01_l\"),Transform=(Rotation=(X=0,Y=0,Z=0,W=1),Translation=(X=0,Y=0,Z=0),Scale ...(truncated)"
    WeightMinimum=0
    Weight=@"DeltaRotation v4_11".outputWeight
   )
   > DeltaRotation v4_12 = RigVMFunctionReferenceNode(
    Rotation Order=ZYX
    child=(Type=Bone,Name="thigh_r")
    outputMinMax=(X=0,Y=1)
    parent=(Type=Null,Name="thigh_r_null")
    premult=1
    rotationMinMax=(X=0,Y=12)
    triggerAxis=Z
    twistAxis=(X=1,Y=0,Z=0)
   )
   > &RigUnit_ModifyTransforms(
    ItemToModify="((Item=(Type=Bone,Name=\"dyn_backSkirt_01_r\"),Transform=(Rotation=(X=0,Y=0,Z=0,W=1),Translation=(X=0,Y=0,Z=0),Scale3D=(X=1,Y=1,Z=1))),(Item=(Type=Bone,Name=\"dyn_sideSkirt_01_r\"),Transform=(Rotation=(X=0,Y=0,Z=0,W=1),Translation=(X=0,Y=0,Z=0),Scale3D=(X=1,Y= ...(truncated)"
    WeightMinimum=0
    Weight=@"DeltaRotation v4_12".outputWeight
   )
   > &RigUnit_ParentConstraint(
    AdvancedSettings=(InterpolationType=Average,RotationOrderForFilter=XZY)
    Child=(Type=Bone,Name="dyn_backSkirt_01")
    Parents=((Item=(Type=Bone,Name="dyn_backSkirt_01_l"),Weight=0.5),(Item=(Type=Bone,Name="dyn_backSkirt_01_r"),Weight=0.5))
    Weight=1
   )
   DeltaRotation v4_2 = RigVMFunctionReferenceNode(
    Rotation Order=ZYX
    child=(Type=Bone,Name="thigh_l")
    outputMinMax=(X=0,Y=1)
    parent=(Type=Null,Name="thigh_l_null")
    premult=1
    rotationMinMax=(X=0,Y=-90)
    triggerAxis=Y
    twistAxis=(X=1,Y=0,Z=0)
    ExecuteContext=@RigVMFunction_Sequence_1.C
   )
   > &RigUnit_ModifyTransforms(
    ItemToModify="((Item=(Type=Bone,Name=\"dyn_frontSkirt_01_l\"),Transform=(Rotation=(X=-0.1687,Y=-0.0632,Z=0.108,W=0.9777),Translation=(X=2.3347,Y=-1.6208,Z=0.6052),Scale3D=(X=1,Y=1,Z=1))),(Item=(Type=Bone,Name=\"dyn_sideSkirt_01_l\"),Transform=(Rotation=(X=-0.1007,Y=-0.1719, ...(truncated)"
    WeightMinimum=0
    Weight=@"DeltaRotation v4_2".outputWeight
   )
   > DeltaRotation v4_3 = RigVMFunctionReferenceNode(
    Rotation Order=ZYX
    child=(Type=Bone,Name="thigh_r")
    outputMinMax=(X=0,Y=1)
    parent=(Type=Null,Name="thigh_r_null")
    premult=1
    rotationMinMax=(X=0,Y=-90)
    triggerAxis=Y
    twistAxis=(X=1,Y=0,Z=0)
   )
   > &RigUnit_ModifyTransforms(
    ItemToModify="((Item=(Type=Bone,Name=\"dyn_frontSkirt_01_r\"),Transform=(Rotation=(X=0.1689,Y=-0.0625,Z=-0.1038,W=0.9782),Translation=(X=2.3347,Y=1.6208,Z=0.6052),Scale3D=(X=1,Y=1,Z=1))),(Item=(Type=Bone,Name=\"dyn_sideSkirt_01_r\"),Transform=(Rotation=(X=0.1007,Y=-0.1719,Z ...(truncated)"
    WeightMinimum=0
    Weight=@"DeltaRotation v4_3".outputWeight
   )
   > &RigUnit_ModifyTransforms(
    ItemToModify="((Item=(Type=Bone,Name=\"dyn_backSkirt_01_l\"),Transform=(Rotation=(X=0,Y=0,Z=0,W=1),Translation=(X=0,Y=0,Z=0),Scale3D=(X=1,Y=1,Z=1))),(Item=(Type=Bone,Name=\"dyn_backSkirt_01_r\"),Transform=(Rotation=(X=0,Y=0,Z=0,W=1),Translation=(X=0,Y=0,Z=0),Scale3D=(X=1,Y= ...(truncated)"
    WeightMinimum=0
    Weight=&RigVMFunction_MathDoubleMul(A=@"DeltaRotation v4_3".outputWeight, B=@"DeltaRotation v4_2".outputWeight).Result
   )
   > DeltaRotation v4_4 = RigVMFunctionReferenceNode(
    Rotation Order=ZYX
    child=(Type=Bone,Name="thigh_l")
    outputMinMax=(X=0,Y=1)
    parent=(Type=Null,Name="thigh_l_null")
    premult=1
    rotationMinMax=(X=0,Y=30)
    triggerAxis=Y
    twistAxis=(X=1,Y=0,Z=0)
   )
   > &RigUnit_ModifyTransforms(
    ItemToModify="((Item=(Type=Bone,Name=\"dyn_frontSkirt_01_l\"),Transform=(Rotation=(X=-0.0114,Y=-0.0864,Z=-0.13,W=0.9877),Translation=(X=0,Y=0,Z=0),Scale3D=(X=1,Y=1,Z=1))),(Item=(Type=Bone,Name=\"dyn_sideSkirt_01_l\"),Transform=(Rotation=(X=0,Y=0.171,Z=0,W=-0.9853),Translati ...(truncated)"
    WeightMinimum=0
    Weight=@"DeltaRotation v4_4".outputWeight
   )
   > DeltaRotation v4_5 = RigVMFunctionReferenceNode(
    Rotation Order=ZYX
    child=(Type=Bone,Name="thigh_r")
    outputMinMax=(X=0,Y=1)
    parent=(Type=Null,Name="thigh_r_null")
    premult=1
    rotationMinMax=(X=0,Y=30)
    triggerAxis=Y
    twistAxis=(X=1,Y=0,Z=0)
   )
   > &RigUnit_ModifyTransforms(
    ItemToModify="((Item=(Type=Bone,Name=\"dyn_frontSkirt_01_r\"),Transform=(Rotation=(X=0.0114,Y=-0.0864,Z=0.13,W=0.9877),Translation=(X=0,Y=0,Z=0),Scale3D=(X=1,Y=1,Z=1))),(Item=(Type=Bone,Name=\"dyn_sideSkirt_01_r\"),Transform=(Rotation=(X=0,Y=-0.1305,Z=0,W=0.9914),Translatio ...(truncated)"
    WeightMinimum=0
    Weight=@"DeltaRotation v4_5".outputWeight
   )
   > &RigUnit_ModifyTransforms(
    ItemToModify="((Item=(Type=Bone,Name=\"dyn_frontSkirt_01_l\"),Transform=(Rotation=(X=-0.2701,Y=-0.2121,Z=-0.0361,W=-0.9385),Translation=(X=0,Y=0,Z=0),Scale3D=(X=1,Y=1,Z=1))),(Item=(Type=Bone,Name=\"dyn_sideSkirt_01_l\"),Transform=(Rotation=(X=0,Y=0,Z=0,W=1),Translation=(X=0 ...(truncated)"
    WeightMinimum=0
    Weight=&RigVMFunction_MathDoubleMul(A=@"DeltaRotation v4_4".outputWeight, B=@"DeltaRotation v4_8".outputWeight).Result
   )
   > &RigUnit_ModifyTransforms(
    ItemToModify="((Item=(Type=Bone,Name=\"dyn_frontSkirt_01_r\"),Transform=(Rotation=(X=-0.139,Y=0.3466,Z=-0.2233,W=0.9004),Translation=(X=0,Y=0,Z=0),Scale3D=(X=1,Y=1,Z=1))),(Item=(Type=Bone,Name=\"dyn_sideSkirt_01_r\"),Transform=(Rotation=(X=0,Y=0,Z=0,W=1),Translation=(X=0,Y= ...(truncated)"
    WeightMinimum=0
    Weight=&RigVMFunction_MathDoubleMul(A=@"DeltaRotation v4_5".outputWeight, B=@"DeltaRotation v4_9".outputWeight).Result
   )
   DeltaRotation v4_13 = RigVMFunctionReferenceNode(
    Rotation Order=ZYX
    child=(Type=Bone,Name="thigh_l")
    onlyTwist=true
    outputMinMax=(X=0,Y=1)
    parent=(Type=Null,Name="thigh_l_null")
    premult=1
    rotationMinMax=(X=0,Y=-100)
    triggerAxis=X
    twistAxis=(X=1,Y=0,Z=0)
    ExecuteContext=@RigVMFunction_Sequence_1.D
   )
   > &RigUnit_ModifyTransforms(
    ItemToModify="((Item=(Type=Bone,Name=\"dyn_backSkirt_01\"),Transform=(Rotation=(X=0,Y=-0.0915,Z=0,W=0.9958),Translation=(X=-0.5,Y=0,Z=1.5),Scale3D=(X=1,Y=1,Z=1))),(Item=(Type=Bone,Name=\"dyn_frontSkirt_01_l\"),Transform=(Rotation=(X=0.7677,Y=0.0459,Z=-0.0946,W=-0.6321),Tran ...(truncated)"
    WeightMinimum=0
    Weight=@"DeltaRotation v4_13".outputWeight
   )
   > DeltaRotation v4_14 = RigVMFunctionReferenceNode(
    Rotation Order=ZYX
    child=(Type=Bone,Name="thigh_r")
    onlyTwist=true
    outputMinMax=(X=0,Y=1)
    parent=(Type=Null,Name="thigh_r_null")
    premult=1
    rotationMinMax=(X=0,Y=-100)
    triggerAxis=X
    twistAxis=(X=1,Y=0,Z=0)
   )
   > &RigUnit_ModifyTransforms(
    ItemToModify="((Item=(Type=Bone,Name=\"dyn_backSkirt_01\"),Transform=(Rotation=(X=0,Y=-0.0915,Z=0,W=0.9958),Translation=(X=-0.5,Y=0,Z=1.5),Scale3D=(X=1,Y=1,Z=1))),(Item=(Type=Bone,Name=\"dyn_frontSkirt_01_r\"),Transform=(Rotation=(X=0.8243,Y=-0.0162,Z=-0.0971,W=0.5576),Tran ...(truncated)"
    WeightMinimum=0
    Weight=@"DeltaRotation v4_14".outputWeight
   )
   > DeltaRotation v4_15 = RigVMFunctionReferenceNode(
    Rotation Order=ZYX
    child=(Type=Bone,Name="thigh_l")
    onlyTwist=true
    outputMinMax=(X=0,Y=1)
    parent=(Type=Null,Name="thigh_l_null")
    premult=1
    rotationMinMax=(X=0,Y=45)
    triggerAxis=X
    twistAxis=(X=1,Y=0,Z=0)
   )
   > &RigUnit_ModifyTransforms(
    ItemToModify="((Item=(Type=Bone,Name=\"dyn_backSkirt_01\"),Transform=(Rotation=(X=0,Y=0,Z=0,W=1),Translation=(X=0,Y=0,Z=0),Scale3D=(X=1,Y=1,Z=1))),(Item=(Type=Bone,Name=\"dyn_frontSkirt_01_l\"),Transform=(Rotation=(X=-0.2601,Y=0.0817,Z=-0.0247,W=-0.9618),Translation=(X=0,Y= ...(truncated)"
    WeightMinimum=0
    Weight=@"DeltaRotation v4_15".outputWeight
   )
   > DeltaRotation v4_16 = RigVMFunctionReferenceNode(
    Rotation Order=ZYX
    child=(Type=Bone,Name="thigh_r")
    onlyTwist=true
    outputMinMax=(X=0,Y=1)
    parent=(Type=Null,Name="thigh_r_null")
    premult=1
    rotationMinMax=(X=0,Y=45)
    triggerAxis=X
    twistAxis=(X=1,Y=0,Z=0)
   )
   > &RigUnit_ModifyTransforms(
    ItemToModify="((Item=(Type=Bone,Name=\"dyn_frontSkirt_01_r\"),Transform=(Rotation=(X=-0.1066,Y=-0.0892,Z=-0.0096,W=0.9902),Translation=(X=0,Y=0,Z=0),Scale3D=(X=1,Y=1,Z=1))),(Item=(Type=Bone,Name=\"dyn_sideSkirt_01_r\"),Transform=(Rotation=(X=0,Y=0,Z=0,W=1),Translation=(X=0, ...(truncated)"
    WeightMinimum=0
    Weight=@"DeltaRotation v4_16".outputWeight
   )
   > &RigUnit_ModifyTransforms(
    ItemToModify="((Item=(Type=Bone,Name=\"dyn_frontSkirt_01_l\"),Transform=(Rotation=(X=-0.0079,Y=-0.0133,Z=-0.0513,W=0.9986),Translation=(X=-2.8194,Y=-13.6847,Z=9.0442),Scale3D=(X=1,Y=1,Z=1))),(Item=(Type=Bone,Name=\"dyn_frontSkirt_01\"),Transform=(Rotation=(X=0,Y=0,Z=0,W=1), ...(truncated)"
    WeightMinimum=0
    Weight=&RigVMFunction_MathDoubleMul(A=@"DeltaRotation v4_15".outputWeight, B=@"DeltaRotation v4_8".outputWeight).Result
   )
   > &RigUnit_ModifyTransforms(
    ItemToModify="((Item=(Type=Bone,Name=\"dyn_frontSkirt_01_r\"),Transform=(Rotation=(X=0.0079,Y=-0.0133,Z=0.0513,W=0.9986),Translation=(X=-2.8194,Y=13.6847,Z=9.0442),Scale3D=(X=1,Y=1,Z=1))),(Item=(Type=Bone,Name=\"dyn_frontSkirt_01\"),Transform=(Rotation=(X=0,Y=0,Z=0,W=1),Tra ...(truncated)"
    WeightMinimum=0
    Weight=&RigVMFunction_MathDoubleMul(B=@"DeltaRotation v4_16".outputWeight, A=@"DeltaRotation v4_9".outputWeight).Result
   )
   > DeltaRotation v4_17 = RigVMFunctionReferenceNode(
    Rotation Order=ZYX
    child=(Type=Bone,Name="thigh_l")
    onlyTwist=true
    outputMinMax=(X=0,Y=1)
    parent=(Type=Null,Name="thigh_l_null")
    premult=1
    rotationMinMax=(X=0,Y=-60)
    triggerAxis=X
    twistAxis=(X=1,Y=0,Z=0)
   )
   n2 = &RigVMFunction_MathDoubleMul(A=@"DeltaRotation v4_17".outputWeight, B=@"DeltaRotation v4_8".outputWeight)
   > &RigUnit_ModifyTransforms(
    ItemToModify="((Item=(Type=Bone,Name=\"dyn_frontSkirt_01_l\"),Transform=(Rotation=(X=0.1371,Y=-0.1266,Z=-0.1167,W=0.9755),Translation=(X=4.8606,Y=15.0642,Z=-5.9088),Scale3D=(X=1,Y=1,Z=1))),(Item=(Type=Bone,Name=\"dyn_frontSkirt_01\"),Transform=(Rotation=(X=-0.4304,Y=-0.0261 ...(truncated)"
    WeightMinimum=0
    Weight=@n2.Result
   )
   > DeltaRotation v4_18 = RigVMFunctionReferenceNode(
    Rotation Order=ZYX
    child=(Type=Bone,Name="thigh_r")
    onlyTwist=true
    outputMinMax=(X=0,Y=1)
    parent=(Type=Null,Name="thigh_r_null")
    premult=1
    rotationMinMax=(X=0,Y=-60)
    triggerAxis=X
    twistAxis=(X=1,Y=0,Z=0)
   )
   n3 = &RigVMFunction_MathDoubleMul(A=@"DeltaRotation v4_18".outputWeight, B=@"DeltaRotation v4_9".outputWeight)
   > &RigUnit_ModifyTransforms(
    ItemToModify="((Item=(Type=Bone,Name=\"dyn_frontSkirt_01_r\"),Transform=(Rotation=(X=-0.1371,Y=-0.1266,Z=0.1167,W=0.9755),Translation=(X=4.8606,Y=-15.0642,Z=-5.9088),Scale3D=(X=1,Y=1,Z=1))),(Item=(Type=Bone,Name=\"dyn_frontSkirt_01\"),Transform=(Rotation=(X=0.4304,Y=-0.0261 ...(truncated)"
    WeightMinimum=0
    Weight=@n3.Result
   )
   > &RigUnit_ModifyTransforms(
    ItemToModify="((Item=(Type=Bone,Name=\"dyn_frontSkirt_01\"),Transform=(Rotation=(X=-0.0339,Y=-0.0059,Z=0.0214,W=-0.9992),Translation=(X=0.0922,Y=2.4347,Z=-0.8437),Scale3D=(X=1,Y=1,Z=1))),(Item=(Type=Bone,Name=\"dyn_frontSkirt_01_l\"),Transform=(Rotation=(X=-0.0752,Y=-0.0329 ...(truncated)"
    WeightMinimum=0
    Weight=&RigVMFunction_MathDoubleMul(A=@"DeltaRotation v4_17".outputWeight, B=@"DeltaRotation v4_18".outputWeight).Result
   )
   > &RigUnit_ModifyTransforms(
    ItemToModify="((Item=(Type=Bone,Name=\"dyn_frontSkirt_01\"),Transform=(Rotation=(X=-0.3272,Y=-0.1879,Z=0.218,W=-0.9001),Translation=(X=4.2243,Y=3.1128,Z=-7.4503),Scale3D=(X=1,Y=1,Z=1))),(Item=(Type=Bone,Name=\"dyn_frontSkirt_01_l\"),Transform=(Rotation=(X=0.2721,Y=0.0491,Z= ...(truncated)"
    WeightMinimum=0
    Weight=&RigVMFunction_MathDoubleMul(A=@n2.Result, B=@n3.Result).Result
   )
   > &RigUnit_ModifyTransforms(
    ItemToModify="((Item=(Type=Bone,Name=\"dyn_sideSkirt_01_l\"),Transform=(Rotation=(X=-0.2268,Y=0.0429,Z=-0.0125,W=-0.9729),Translation=(X=4.2905,Y=4.389,Z=2.547),Scale3D=(X=1,Y=1,Z=1))),(Item=(Type=Bone,Name=\"dyn_backSkirt_01_l\"),Transform=(Rotation=(X=-0.0569,Y=0.1518,Z=- ...(truncated)"
    WeightMinimum=0
    Weight=&RigVMFunction_MathDoubleMul(A=@"DeltaRotation v4_13".outputWeight, B=@"DeltaRotation v4_2".outputWeight).Result
   )
   > &RigUnit_ModifyTransforms(
    ItemToModify="((Item=(Type=Bone,Name=\"dyn_sideSkirt_01_r\"),Transform=(Rotation=(X=-0.185,Y=-0.1754,Z=0.1223,W=-0.9592),Translation=(X=0.6852,Y=-2.0385,Z=2.5367),Scale3D=(X=1,Y=1,Z=1))),(Item=(Type=Bone,Name=\"dyn_backSkirt_01_r\"),Transform=(Rotation=(X=0.0218,Y=-0.0316,Z ...(truncated)"
    WeightMinimum=0
    Weight=&RigVMFunction_MathDoubleMul(A=@"DeltaRotation v4_14".outputWeight, B=@"DeltaRotation v4_3".outputWeight).Result
   )
   > &RigUnit_ModifyTransforms(
    ItemToModify="((Item=(Type=Bone,Name=\"dyn_sideSkirt_01_l\"),Transform=(Rotation=(X=-0.0228,Y=0.0984,Z=0.0244,W=-0.9946),Translation=(X=0,Y=0,Z=0),Scale3D=(X=1,Y=1,Z=1))),(Item=(Type=Bone,Name=\"dyn_backSkirt_01_l\"),Transform=(Rotation=(X=0,Y=0.1077,Z=0,W=-0.9942),Translat ...(truncated)"
    WeightMinimum=0
    Weight=&RigVMFunction_MathDoubleMul(B=@"DeltaRotation v4_15".outputWeight, A=@"DeltaRotation v4_2".outputWeight).Result
   )
   > &RigUnit_ModifyTransforms(
    ItemToModify="((Item=(Type=Bone,Name=\"dyn_sideSkirt_01_r\"),Transform=(Rotation=(X=0.0223,Y=-0.0415,Z=-0.0733,W=-0.9962),Translation=(X=0,Y=0,Z=0),Scale3D=(X=1,Y=1,Z=1))),(Item=(Type=Bone,Name=\"dyn_backSkirt_01_r\"),Transform=(Rotation=(X=0,Y=0.1183,Z=0,W=-0.993),Translat ...(truncated)"
    WeightMinimum=0
    Weight=&RigVMFunction_MathDoubleMul(B=@"DeltaRotation v4_16".outputWeight, A=@"DeltaRotation v4_3".outputWeight).Result
   )
   > &RigUnit_ModifyTransforms(
    ItemToModify="((Item=(Type=Bone,Name=\"dyn_frontSkirt_01_l\"),Transform=(Rotation=(X=-0.0018,Y=0.1331,Z=-0.178,W=0.975),Translation=(X=0,Y=1.0001,Z=1),Scale3D=(X=1,Y=1,Z=1))),(Item=(Type=Bone,Name=\"dyn_frontSkirt_01\"),Transform=(Rotation=(X=0.0286,Y=0.409,Z=-0.0033,W=0.91 ...(truncated)"
    WeightMaximum=0.5
    WeightMinimum=0
    Weight=&RigVMFunction_MathDoubleMul(B=@"DeltaRotation v4_2".outputWeight, A=@"DeltaRotation v4_8".outputWeight).Result
   )
   > &RigUnit_ModifyTransforms(
    ItemToModify="((Item=(Type=Bone,Name=\"dyn_frontSkirt_01_r\"),Transform=(Rotation=(X=0.0018,Y=0.1331,Z=0.178,W=0.975),Translation=(X=1.5613,Y=5.7658,Z=0.2717),Scale3D=(X=1,Y=1,Z=1))),(Item=(Type=Bone,Name=\"dyn_frontSkirt_01\"),Transform=(Rotation=(X=-0.0286,Y=0.409,Z=0.003 ...(truncated)"
    WeightMaximum=0.5
    WeightMinimum=0
    Weight=&RigVMFunction_MathDoubleMul(B=@"DeltaRotation v4_3".outputWeight, A=@"DeltaRotation v4_9".outputWeight).Result
   )
   > &RigUnit_ModifyTransforms(
    ItemToModify=((Item=(Type=Bone,Name="dyn_frontSkirt_01"),Transform=(Rotation=(X=-0.1215,Y=0,Z=0,W=0.9926),Translation=(X=0,Y=10,Z=-0.7153),Scale3D=(X=1,Y=1,Z=1))))
    WeightMaximum=0.6
    WeightMinimum=0
    Weight=&RigVMFunction_MathDoubleMul(A=@"DeltaRotation v4_4".outputWeight, B=@"DeltaRotation v4_8".outputWeight).Result
   )
   > &RigUnit_ModifyTransforms(
    ItemToModify=((Item=(Type=Bone,Name="dyn_frontSkirt_01"),Transform=(Rotation=(X=0.1215,Y=0,Z=0,W=0.9926),Translation=(X=0,Y=-10,Z=-0.7153),Scale3D=(X=1,Y=1,Z=1))))
    WeightMaximum=0.6
    WeightMinimum=0
    Weight=&RigVMFunction_MathDoubleMul(B=@"DeltaRotation v4_5".outputWeight, A=@"DeltaRotation v4_9".outputWeight).Result
   )
   Multiply_20 = RigVMAggregateNode(C=@"DeltaRotation v4_17".outputWeight, B=@"DeltaRotation v4_4".outputWeight, A=@"DeltaRotation v4_8".outputWeight)
   > &RigUnit_ModifyTransforms(
    ItemToModify="((Item=(Type=Bone,Name=\"dyn_frontSkirt_01_l\"),Transform=(Rotation=(X=0,Y=0,Z=0,W=1),Translation=(X=0,Y=4.8909,Z=-4.1294),Scale3D=(X=1,Y=1,Z=1))),(Item=(Type=Bone,Name=\"dyn_frontSkirt_01\"),Transform=(Rotation=(X=-0.0652,Y=-0.1828,Z=0.4564,W=0.8683),Translat ...(truncated)"
    WeightMaximum=0.5
    WeightMinimum=0
    Weight=@Multiply_20.Result
   )
   Multiply_21 = RigVMAggregateNode(C=@"DeltaRotation v4_18".outputWeight, B=@"DeltaRotation v4_5".outputWeight, A=@"DeltaRotation v4_9".outputWeight)
   > &RigUnit_ModifyTransforms(
    ItemToModify="((Item=(Type=Bone,Name=\"dyn_frontSkirt_01_r\"),Transform=(Rotation=(X=0,Y=0,Z=0,W=1),Translation=(X=0,Y=-4.8909,Z=-4.1294),Scale3D=(X=1,Y=1,Z=1))),(Item=(Type=Bone,Name=\"dyn_frontSkirt_01\"),Transform=(Rotation=(X=0.0652,Y=-0.1828,Z=-0.4564,W=0.8683),Transla ...(truncated)"
    WeightMaximum=0.5
    WeightMinimum=0
    Weight=@Multiply_21.Result
   )
   > n4 = &RigUnit_SphericalPoseReader(
    ActiveRegionScaleFactors=(PositiveWidth=1,NegativeWidth=1,PositiveHeight=1,NegativeHeight=1)
    ActiveRegionSize=0.05
    Debug=(bDrawDebug=True,bDraw2D=false,bDrawLocalAxes=True,DebugScale=25,DebugSegments=20,DebugThickness=0.25)
    DriverAxis=(X=1,Y=0,Z=0)
    DriverItem=(Type=Bone,Name="thigh_l")
    FalloffRegionScaleFactors=(PositiveWidth=1,NegativeWidth=1,PositiveHeight=1,NegativeHeight=1)
    FalloffSize=0.1
    FlipHeightScaling=False
    FlipWidthScaling=False
    OptionalParentItem=(Type=Bone,Name="None")
    RotationOffset=(X=-15,Y=20,Z=0)
   )
   > &RigUnit_ModifyTransforms(
    ItemToModify="((Item=(Type=Bone,Name=\"dyn_sideSkirt_01_l\"),Transform=(Rotation=(X=-0.0854,Y=0.0028,Z=-0.0331,W=0.9958),Translation=(X=-0.3363,Y=2.3009,Z=8.1433),Scale3D=(X=1,Y=1,Z=1))),(Item=(Type=Bone,Name=\"dyn_frontSkirt_01_l\"),Transform=(Rotation=(X=-0.2022,Y=0,Z=0,W ...(truncated)"
    WeightMinimum=0
    Weight=@n4.OutputParam
   )
   > n5 = &RigUnit_SphericalPoseReader(
    ActiveRegionScaleFactors=(PositiveWidth=1,NegativeWidth=1,PositiveHeight=1,NegativeHeight=1)
    ActiveRegionSize=0.05
    Debug=(bDrawDebug=True,bDraw2D=false,bDrawLocalAxes=True,DebugScale=25,DebugSegments=20,DebugThickness=0.25)
    DriverAxis=(X=1,Y=0,Z=0)
    DriverItem=(Type=Bone,Name="thigh_r")
    FalloffRegionScaleFactors=(PositiveWidth=1,NegativeWidth=1,PositiveHeight=1,NegativeHeight=1)
    FalloffSize=0.1
    FlipHeightScaling=False
    FlipWidthScaling=False
    OptionalParentItem=(Type=Bone,Name="None")
    RotationOffset=(X=-15,Y=20,Z=0)
   )
   > &RigUnit_ModifyTransforms(
    ItemToModify="((Item=(Type=Bone,Name=\"dyn_sideSkirt_01_r\"),Transform=(Rotation=(X=0.0854,Y=0.0028,Z=0.0331,W=0.9958),Translation=(X=-0.3363,Y=-2.3009,Z=8.1433),Scale3D=(X=1,Y=1,Z=1))),(Item=(Type=Bone,Name=\"dyn_frontSkirt_01_r\"),Transform=(Rotation=(X=-0.2022,Y=0,Z=0,W= ...(truncated)"
    WeightMinimum=0
    Weight=@n5.OutputParam
   )
   RigVMFunction_MathBoolToFloat = &RigVMFunction_MathBoolToFloat(Value=RigVMVariableNode() {variable=passenger}.Value)
   &RigUnit_ModifyTransforms(
    ItemToModify=((Item=(Type=Bone,Name="dyn_frontSkirt_01"),Transform=(Rotation=(X=-0.1036,Y=-0.2074,Z=-0.1036,W=0.9672),Translation=(X=-10,Y=0,Z=5),Scale3D=(X=1,Y=1,Z=1))))
    WeightMinimum=0
    Weight=@RigVMFunction_MathBoolToFloat.Result
    ExecutePin=@RigVMFunction_Sequence_1.E
   )
   RigVMFunctionReturnNode(ExecuteContext=@RigVMFunction_Sequence.B)
  ]}
 New Function_ContainedGraph {nodes [
   n1 = RigVMFunctionEntryNode()
   RigVMFunction_MathBoolToInteger = &RigVMFunction_MathBoolToInteger(Value=@n1.onlyTwist)
   RigVMFunction_MathQuaternionSwingTwist = &RigVMFunction_MathQuaternionSwingTwist(
    twistAxis=@n1.twistAxis
    Input=&RigUnit_GetRelativeTransformForItem(bChildInitial=False, bParentInitial=False, child=@n1.child, parent=@n1.parent).RelativeTransform.Rotation
   )
   n2 = RigVMDispatchNode(
    Values=((),())
    Index=@RigVMFunction_MathBoolToInteger.Result
    Values.0=@RigVMFunction_MathQuaternionSwingTwist.Swing
    Values.1=@RigVMFunction_MathQuaternionSwingTwist.Twist
   ) {template="DISPATCH_RigVMDispatch_SelectInt32(in Index,in Values,out Result)"}
   RigVMFunction_MathQuaternionToEuler = &RigVMFunction_MathQuaternionToEuler(Value=@n2.Result, RotationOrder=@n1.Rotation Order)
   > DISPATCH_RigVMDispatch_Print = RigVMDispatchNode(ScreenColor=(R=1,G=1,B=1,A=1), ScreenDuration=1.5, Enabled=@n1.debugRotation, Value=@RigVMFunction_MathQuaternionToEuler.Result) {template="DISPATCH_RigVMDispatch_Print(in Prefix,in Value,in Enabled,in ScreenDuration,in ScreenColor)"}
   Cast = RigVMDispatchNode(Value=@n1.triggerAxis) {template="DISPATCH_RigVMDispatch_CastEnumToInt(in Value,out Result)"}
   DISPATCH_RigVMDispatch_SelectInt32 = RigVMDispatchNode(
    Values=(0,0,0)
    Index=@Cast.Result
    Values.0=@RigVMFunction_MathQuaternionToEuler.Result.X
    Values.1=@RigVMFunction_MathQuaternionToEuler.Result.Y
    Values.2=@RigVMFunction_MathQuaternionToEuler.Result.Z
   ) {template="DISPATCH_RigVMDispatch_SelectInt32(in Index,in Values,out Result)"}
   Remap = &RigVMFunction_MathDoubleRemap(
    bClamp=False
    Value=@DISPATCH_RigVMDispatch_SelectInt32.Result
    TargetMinimum=@n1.outputMinMax.X
    TargetMaximum=@n1.outputMinMax.Y
    SourceMinimum=@n1.rotationMinMax.X
    SourceMaximum=@n1.rotationMinMax.Y
   )
   Absolute = &RigVMFunction_MathDoubleAbs(Value=@Remap.Result)
   Multiply = &RigVMFunction_MathDoubleMul(
    A=&RigVMFunction_MathDoubleClamp(Value=RigVMDispatchNode(True=@Absolute.Result, Condition=@n1.absoluteValue, False=@Remap.Result) {template="DISPATCH_RigVMDispatch_If(in Condition,in True,in False,out Result)"}.Result).Result
    B=@n1.premult
   )
   > RigVMFunctionReturnNode(outputQuat=@n2.Result, outputWeight=@Multiply.Result)
  ]}
 New Function_ContainedGraph {nodes [
   RigVMFunctionEntryNode()
   > &RigUnit_ParentConstraint(
    AdvancedSettings=(InterpolationType=Average,RotationOrderForFilter=XZY)
    Child=(Type=Bone,Name="MSTR_sleeve_l")
    Filter=(TranslationFilter=(bX=true,bY=true,bZ=true),RotationFilter=(bX=True,bY=True,bZ=True),ScaleFilter=(bX=false,bY=false,bZ=false))
    Parents=((Item=(Type=Bone,Name="clavicle_l"),Weight=0.5),(Item=(Type=Bone,Name="upperarm_l"),Weight=0.5))
    Weight=1
   )
   > &RigUnit_ParentConstraint(
    AdvancedSettings=(InterpolationType=Average,RotationOrderForFilter=XZY)
    Child=(Type=Bone,Name="dyn_sleeve_03_l")
    Filter=(TranslationFilter=(bX=true,bY=true,bZ=true),RotationFilter=(bX=True,bY=True,bZ=True),ScaleFilter=(bX=false,bY=false,bZ=false))
    Parents=((Item=(Type=Bone,Name="lowerarm_l"),Weight=0.7),(Item=(Type=Bone,Name="dyn_sleeve_02_l"),Weight=0.3),(Item=(Type=Bone,Name="None"),Weight=1))
    Weight=1
   )
   > &RigUnit_ParentConstraint(
    AdvancedSettings=(InterpolationType=Average,RotationOrderForFilter=XZY)
    Child=(Type=Bone,Name="dyn_sleeve_04_l")
    Filter=(TranslationFilter=(bX=true,bY=true,bZ=true),RotationFilter=(bX=True,bY=True,bZ=True),ScaleFilter=(bX=false,bY=false,bZ=false))
    Parents=((Item=(Type=Bone,Name="lowerarm_l"),Weight=0.7),(Item=(Type=Bone,Name="dyn_sleeve_03_l"),Weight=0.3),(Item=(Type=Bone,Name="None"),Weight=1))
    Weight=1
   )
   > &RigUnit_ParentConstraint(
    AdvancedSettings=(InterpolationType=Average,RotationOrderForFilter=XZY)
    Child=(Type=Bone,Name="dyn_sleeve_05_l")
    Filter=(TranslationFilter=(bX=true,bY=true,bZ=true),RotationFilter=(bX=True,bY=True,bZ=True),ScaleFilter=(bX=false,bY=false,bZ=false))
    Parents=((Item=(Type=Bone,Name="lowerarm_l"),Weight=0.7),(Item=(Type=Bone,Name="dyn_sleeve_04_l"),Weight=0.3),(Item=(Type=Bone,Name="None"),Weight=1))
    Weight=1
   )
   > &RigUnit_ParentConstraint(
    AdvancedSettings=(InterpolationType=Average,RotationOrderForFilter=XZY)
    Child=(Type=Bone,Name="MSTR_sleeve_r")
    Filter=(TranslationFilter=(bX=true,bY=true,bZ=true),RotationFilter=(bX=True,bY=True,bZ=True),ScaleFilter=(bX=false,bY=false,bZ=false))
    Parents=((Item=(Type=Bone,Name="clavicle_r"),Weight=0.5),(Item=(Type=Bone,Name="upperarm_r"),Weight=0.5))
    Weight=1
   )
   > &RigUnit_ParentConstraint(
    AdvancedSettings=(InterpolationType=Average,RotationOrderForFilter=XZY)
    Child=(Type=Bone,Name="dyn_sleeve_03_r")
    Filter=(TranslationFilter=(bX=true,bY=true,bZ=true),RotationFilter=(bX=True,bY=True,bZ=True),ScaleFilter=(bX=false,bY=false,bZ=false))
    Parents=((Item=(Type=Bone,Name="lowerarm_r"),Weight=0.7),(Item=(Type=Bone,Name="dyn_sleeve_02_r"),Weight=0.3),(Item=(Type=Bone,Name="None"),Weight=1))
    Weight=1
   )
   > &RigUnit_ParentConstraint(
    AdvancedSettings=(InterpolationType=Average,RotationOrderForFilter=XZY)
    Child=(Type=Bone,Name="dyn_sleeve_04_r")
    Filter=(TranslationFilter=(bX=true,bY=true,bZ=true),RotationFilter=(bX=True,bY=True,bZ=True),ScaleFilter=(bX=false,bY=false,bZ=false))
    Parents=((Item=(Type=Bone,Name="lowerarm_r"),Weight=0.7),(Item=(Type=Bone,Name="dyn_sleeve_03_r"),Weight=0.3),(Item=(Type=Bone,Name="None"),Weight=1))
    Weight=1
   )
   > &RigUnit_ParentConstraint(
    AdvancedSettings=(InterpolationType=Average,RotationOrderForFilter=XZY)
    Child=(Type=Bone,Name="dyn_sleeve_05_r")
    Filter=(TranslationFilter=(bX=true,bY=true,bZ=true),RotationFilter=(bX=True,bY=True,bZ=True),ScaleFilter=(bX=false,bY=false,bZ=false))
    Parents=((Item=(Type=Bone,Name="lowerarm_r"),Weight=0.7),(Item=(Type=Bone,Name="dyn_sleeve_04_r"),Weight=0.3),(Item=(Type=Bone,Name="None"),Weight=1))
    Weight=1
   )
   > DeltaRotation v4_1 = RigVMFunctionReferenceNode(
    Rotation Order=ZYX
    child=(Type=Bone,Name="lowerarm_l")
    onlyTwist=true
    outputMinMax=(X=0,Y=1)
    parent=(Type=Null,Name="lowerarm_l_null")
    premult=1
    rotationMinMax=(X=0,Y=90)
    triggerAxis=X
    twistAxis=(X=1,Y=0,Z=0)
   )
   > &RigUnit_ModifyTransforms(
    ItemToModify="((Item=(Type=Bone,Name=\"dyn_sleeve_03_l\"),Transform=(Rotation=(X=-0.025,Y=-0.1572,Z=-0.2122,W=0.9642),Translation=(X=3.5,Y=-3,Z=1.6696),Scale3D=(X=1,Y=1,Z=1))),(Item=(Type=Bone,Name=\"dyn_sleeve_05_l\"),Transform=(Rotation=(X=0,Y=0,Z=0,W=1),Translation=(X=0, ...(truncated)"
    WeightMinimum=0
    Weight=@"DeltaRotation v4_1".outputWeight
   )
   > DeltaRotation v4_5 = RigVMFunctionReferenceNode(
    Rotation Order=ZYX
    child=(Type=Bone,Name="lowerarm_l")
    onlyTwist=true
    outputMinMax=(X=0,Y=1)
    parent=(Type=Null,Name="lowerarm_l_null")
    premult=1
    rotationMinMax=(X=0,Y=-90)
    triggerAxis=X
    twistAxis=(X=1,Y=0,Z=0)
   )
   > &RigUnit_ModifyTransforms(
    ItemToModify=((Item=(Type=Bone,Name="dyn_sleeve_03_l"),Transform=(Rotation=(X=0.4872,Y=0.1133,Z=0.2334,W=0.8339),Translation=(X=-3.2345,Y=4.3374,Z=-0.829),Scale3D=(X=1,Y=1,Z=1))))
    WeightMinimum=0
    Weight=@"DeltaRotation v4_5".outputWeight
   )
   > DeltaRotation v4_6 = RigVMFunctionReferenceNode(
    Rotation Order=ZYX
    absoluteValue=false
    child=(Type=Bone,Name="lowerarm_r")
    debugRotation=false
    onlyTwist=true
    outputMinMax=(X=0,Y=1)
    parent=(Type=Null,Name="lowerarm_r_null")
    premult=1
    rotationMinMax=(X=0,Y=90)
    triggerAxis=X
    twistAxis=(X=1,Y=0,Z=0)
   )
   > &RigUnit_ModifyTransforms(
    ItemToModify="((Item=(Type=Bone,Name=\"dyn_sleeve_03_r\"),Transform=(Rotation=(X=-0.025,Y=-0.1572,Z=-0.2122,W=0.9642),Translation=(X=-3.5,Y=3,Z=-1.6696),Scale3D=(X=1,Y=1,Z=1))),(Item=(Type=Bone,Name=\"dyn_sleeve_05_r\"),Transform=(Rotation=(X=0,Y=0,Z=0,W=1),Translation=(X=0 ...(truncated)"
    WeightMinimum=0
    Weight=@"DeltaRotation v4_6".outputWeight
   )
   > DeltaRotation v4_7 = RigVMFunctionReferenceNode(
    Rotation Order=ZYX
    child=(Type=Bone,Name="lowerarm_r")
    onlyTwist=true
    outputMinMax=(X=0,Y=1)
    parent=(Type=Null,Name="lowerarm_r_null")
    premult=1
    rotationMinMax=(X=0,Y=-90)
    triggerAxis=X
    twistAxis=(X=1,Y=0,Z=0)
   )
   > &RigUnit_ModifyTransforms(
    ItemToModify=((Item=(Type=Bone,Name="dyn_sleeve_03_r"),Transform=(Rotation=(X=0.4872,Y=0.1133,Z=0.2334,W=0.8339),Translation=(X=3.2345,Y=-4.3374,Z=0.829),Scale3D=(X=1,Y=1,Z=1))))
    WeightMinimum=0
    Weight=@"DeltaRotation v4_7".outputWeight
   )
   > RigVMFunctionReturnNode()
  ]}
 New Function_ContainedGraph {nodes [
   RigVMFunctionEntryNode()
   > &RigUnit_ParentConstraint(
    AdvancedSettings=(InterpolationType=Average,RotationOrderForFilter=XZY)
    Child=(Type=Bone,Name="MSTR_sword")
    Parents=((Item=(Type=Bone,Name="spine_01"),Weight=0.1),(Item=(Type=Bone,Name="spine_02"),Weight=0.1),(Item=(Type=Bone,Name="pelvis"),Weight=0.55))
    Weight=1
   )
   > DeltaRotation v4_8 = RigVMFunctionReferenceNode(
    Rotation Order=ZYX
    child=(Type=Bone,Name="thigh_l")
    debugRotation=false
    outputMinMax=(X=0,Y=1)
    parent=(Type=Null,Name="thigh_l_null")
    premult=1
    rotationMinMax=(X=-40,Y=-120)
    triggerAxis=Z
    twistAxis=(X=1,Y=0,Z=0)
   )
   > &RigUnit_ModifyTransforms(
    ItemToModify=((Item=(Type=Bone,Name="dyn_sword"),Transform=(Rotation=(X=0.2424,Y=-0.0254,Z=0.2363,W=0.9406),Translation=(X=2.3112,Y=-2.8818,Z=-1.1166),Scale3D=(X=1,Y=1,Z=1))))
    WeightMinimum=0
    Weight=@"DeltaRotation v4_8".outputWeight
   )
   > DeltaRotation v4_1 = RigVMFunctionReferenceNode(
    Rotation Order=ZYX
    child=(Type=Bone,Name="thigh_l")
    outputMinMax=(X=0,Y=1)
    parent=(Type=Null,Name="thigh_l_null")
    premult=1
    rotationMinMax=(X=0,Y=30)
    triggerAxis=Z
    twistAxis=(X=1,Y=0,Z=0)
   )
   > &RigUnit_ModifyTransforms(
    ItemToModify=((Item=(Type=Bone,Name="dyn_sword"),Transform=(Rotation=(X=0,Y=-0.0193,Z=0,W=0.9998),Translation=(X=0,Y=0,Z=0),Scale3D=(X=1,Y=1,Z=1))))
    WeightMinimum=0
    Weight=@"DeltaRotation v4_1".outputWeight
   )
   > DeltaRotation v4_2 = RigVMFunctionReferenceNode(
    Rotation Order=ZYX
    child=(Type=Bone,Name="thigh_l")
    outputMinMax=(X=0,Y=1)
    parent=(Type=Null,Name="thigh_l_null")
    premult=1
    rotationMinMax=(X=0,Y=-90)
    triggerAxis=Y
    twistAxis=(X=1,Y=0,Z=0)
   )
   > &RigUnit_ModifyTransforms(
    ItemToModify=((Item=(Type=Bone,Name="dyn_sword"),Transform=(Rotation=(X=0.1068,Y=-0.1601,Z=0.4087,W=0.8922),Translation=(X=2.7869,Y=-1.8221,Z=-1.9102),Scale3D=(X=1,Y=1,Z=1))))
    WeightMinimum=0
    Weight=@"DeltaRotation v4_2".outputWeight
   )
   Multiply = &RigVMFunction_MathDoubleMul(B=@"DeltaRotation v4_2".outputWeight, A=@"DeltaRotation v4_8".outputWeight)
   > &RigUnit_ModifyTransforms(
    ItemToModify=((Item=(Type=Bone,Name="dyn_sword"),Transform=(Rotation=(X=0,Y=0,Z=0,W=1),Translation=(X=10,Y=5,Z=8),Scale3D=(X=1,Y=1,Z=1))))
    WeightMinimum=0
    Weight=@Multiply.Result
   )
   > &RigUnit_ModifyTransforms(
    ItemToModify=((Item=(Type=Bone,Name="dyn_sword"),Transform=(Rotation=(X=-0.0429,Y=-0.3418,Z=0.117,W=0.9315),Translation=(X=0,Y=0,Z=0),Scale3D=(X=1,Y=1,Z=1))))
    WeightMinimum=0
    Weight=&RigVMFunction_MathDoubleMul(A=@"DeltaRotation v4_1".outputWeight, B=@"DeltaRotation v4_2".outputWeight).Result
   )
   > DeltaRotation v4_6 = RigVMFunctionReferenceNode(
    Rotation Order=ZYX
    child=(Type=Bone,Name="spine_04")
    debugRotation=false
    outputMinMax=(X=0,Y=1)
    parent=(Type=Null,Name="spine_04_null")
    premult=1
    rotationMinMax=(X=45,Y=60)
    triggerAxis=Z
    twistAxis=(X=1,Y=0,Z=0)
   )
   > &RigUnit_ModifyTransforms(
    ItemToModify=((Item=(Type=Bone,Name="dyn_sword"),Transform=(Rotation=(X=0,Y=0,Z=0.0872,W=0.9962),Translation=(X=0,Y=0,Z=0),Scale3D=(X=1,Y=1,Z=1))))
    WeightMinimum=0
    Weight=@"DeltaRotation v4_6".outputWeight
   )
   > DeltaRotation v4_7 = RigVMFunctionReferenceNode(
    Rotation Order=ZYX
    child=(Type=Bone,Name="spine_03")
    debugRotation=false
    outputMinMax=(X=0,Y=1)
    parent=(Type=Null,Name="spine_03_null")
    premult=1
    rotationMinMax=(X=10,Y=70)
    triggerAxis=Z
    twistAxis=(X=1,Y=0,Z=0)
   )
   > &RigUnit_ModifyTransforms(
    ItemToModify=((Item=(Type=Bone,Name="dyn_sword"),Transform=(Rotation=(X=0,Y=0,Z=0.1736,W=0.9848),Translation=(X=0,Y=0,Z=0),Scale3D=(X=1,Y=1,Z=1))))
    WeightMinimum=0
    Weight=@"DeltaRotation v4_7".outputWeight
   )
   > DeltaRotation v4_10 = RigVMFunctionReferenceNode(
    Rotation Order=ZYX
    child=(Type=Bone,Name="spine_02")
    debugRotation=false
    outputMinMax=(X=0,Y=1)
    parent=(Type=Null,Name="spine_02_null")
    premult=1
    rotationMinMax=(X=0,Y=60)
    triggerAxis=Z
    twistAxis=(X=1,Y=0,Z=0)
   )
   > &RigUnit_ModifyTransforms(
    ItemToModify=((Item=(Type=Bone,Name="dyn_sword"),Transform=(Rotation=(X=0,Y=0,Z=0.0755,W=0.9971),Translation=(X=-0.2345,Y=1.5387,Z=0),Scale3D=(X=1,Y=1,Z=1))))
    WeightMinimum=0
    Weight=@"DeltaRotation v4_10".outputWeight
   )
   > DeltaRotation v4_9 = RigVMFunctionReferenceNode(
    Rotation Order=ZYX
    child=(Type=Bone,Name="spine_01")
    debugRotation=false
    outputMinMax=(X=0,Y=1)
    parent=(Type=Null,Name="spine_01_null")
    premult=1
    rotationMinMax=(X=0,Y=60)
    triggerAxis=Z
    twistAxis=(X=1,Y=0,Z=0)
   )
   > &RigUnit_ModifyTransforms(
    ItemToModify=((Item=(Type=Bone,Name="dyn_sword"),Transform=(Rotation=(X=0,Y=0,Z=0.0755,W=0.9971),Translation=(X=-0.2345,Y=1.5387,Z=0),Scale3D=(X=1,Y=1,Z=1))))
    WeightMinimum=0
    Weight=@"DeltaRotation v4_9".outputWeight
   )
   n1 = RigVMVariableNode() {variable=crouching}
   RigVMFunction_MathBoolToFloat = &RigVMFunction_MathBoolToFloat(Value=@n1.Value)
   > &RigUnit_ModifyTransforms(
    ItemToModify=((Item=(Type=Bone,Name="dyn_sword"),Transform=(Rotation=(X=-0.4501,Y=0.1034,Z=0.0429,W=0.8859),Translation=(X=-10.3361,Y=-1.9777,Z=-0.2523),Scale3D=(X=0.9,Y=0.9,Z=0.9))))
    WeightMinimum=0
    Weight=@RigVMFunction_MathBoolToFloat.Result
   )
   > n2 = &RigUnit_SphericalPoseReader(
    ActiveRegionScaleFactors=(PositiveWidth=1,NegativeWidth=1,PositiveHeight=1,NegativeHeight=1)
    ActiveRegionSize=0.15
    Debug=(bDrawDebug=True,bDraw2D=False,bDrawLocalAxes=True,DebugScale=25,DebugSegments=20,DebugThickness=0.25)
    DriverAxis=(X=1,Y=0,Z=0)
    DriverItem=(Type=Bone,Name="upperarm_l")
    FalloffRegionScaleFactors=(PositiveWidth=1,NegativeWidth=1,PositiveHeight=1,NegativeHeight=1)
    FalloffSize=0.3
    FlipHeightScaling=False
    FlipWidthScaling=False
    OptionalParentItem=(Type=Bone,Name="clavicle_l")
    RotationOffset=(X=140,Y=0,Z=-10)
   )
   Cast = &RigVMFunction_MathBoolToFloat(Value=&RigVMFunction_MathBoolNot(Value=RigVMVariableNode() {variable=melee}.Value).Result)
   RigVMFunction_MathBoolNot = &RigVMFunction_MathBoolNot(Value=@n1.Value)
   Multiply_2 = RigVMAggregateNode(
    C=/Script/RigVM.RigVMFunction_AnimEasing(
     SourceMaximum=1
     SourceMinimum=0
     TargetMaximum=1
     TargetMinimum=0
     Type=CubicEaseInOut
     Value=&RigVMFunction_KalmanFloat(BufferSize=30, Value=@Cast.Result) {template="KalmanFilter::Execute(in Value,in BufferSize,out Result)"}.Result
    ) {is=struct RigVMUnitNode}.Result
    B=&RigVMFunction_AnimEasing(
     SourceMaximum=1
     SourceMinimum=0
     TargetMaximum=1
     TargetMinimum=0
     Type=CubicEaseInOut
     Value=&RigVMFunction_KalmanFloat(BufferSize=30, Value=@n2.OutputParam) {template="KalmanFilter::Execute(in Value,in BufferSize,out Result)"}.Result
    ).Result
    A=&RigVMFunction_MathBoolToFloat(Value=@RigVMFunction_MathBoolNot.Result).Result
   )
   > &RigUnit_ModifyTransforms(
    ItemToModify=((Item=(Type=Bone,Name="dyn_sword"),Transform=(Rotation=(X=0,Y=-0.1736,Z=0,W=0.9848),Translation=(X=0,Y=2,Z=0),Scale3D=(X=0.9,Y=0.9,Z=0.9))))
    WeightMinimum=0
    Weight=@Multiply_2.Result
   )
   > n3 = &RigUnit_SphericalPoseReader(
    ActiveRegionScaleFactors=(PositiveWidth=1,NegativeWidth=1,PositiveHeight=1,NegativeHeight=1)
    ActiveRegionSize=0.05
    Debug=(bDrawDebug=True,bDraw2D=false,bDrawLocalAxes=True,DebugScale=25,DebugSegments=20,DebugThickness=0.25)
    DriverAxis=(X=1,Y=0,Z=0)
    DriverItem=(Type=Bone,Name="thigh_l")
    FalloffRegionScaleFactors=(PositiveWidth=1,NegativeWidth=1,PositiveHeight=1,NegativeHeight=1)
    FalloffSize=0.1
    FlipHeightScaling=False
    FlipWidthScaling=False
    OptionalParentItem=(Type=Bone,Name="None")
    RotationOffset=(X=-15,Y=20,Z=0)
   )
   > &RigUnit_ModifyTransforms(
    ItemToModify=((Item=(Type=Bone,Name="dyn_sword"),Transform=(Rotation=(X=0.0102,Y=0.19,Z=-0.0525,W=0.9803),Translation=(X=-0.7507,Y=0.0807,Z=-1.8749),Scale3D=(X=1,Y=1,Z=1))))
    WeightMinimum=0
    Weight=@n3.OutputParam
   )
   > RigVMFunctionReturnNode()
  ]}
 New Function_ContainedGraph {nodes [
   RigVMFunctionEntryNode()
   > &RigUnit_ParentConstraint(
    AdvancedSettings=(RotationOrderForFilter=XZY)
    Child=(Type=Bone,Name="MSTR_strap_01_l")
    Parents="((Item=(Type=Bone,Name=\"deform_pec_l\"),Weight=0.0022),(Item=(Type=Bone,Name=\"upperarm_twist_02_l\"),Weight=0.0036),(Item=(Type=Bone,Name=\"deform_upperarm_bk_l\"),Weight=0.0037),(Item=(Type=Bone,Name=\"spine_04\"),Weight=0.0063),(Item=(Type=Bone,Name=\"deform_up ...(truncated)"
    Weight=1
   )
   > &RigUnit_ParentConstraint(
    AdvancedSettings=(RotationOrderForFilter=XZY)
    Child=(Type=Bone,Name="MSTR_strap_01_r")
    Parents="((Item=(Type=Bone,Name=\"deform_pec_r\"),Weight=0.0022),(Item=(Type=Bone,Name=\"upperarm_twist_02_r\"),Weight=0.0036),(Item=(Type=Bone,Name=\"deform_upperarm_bk_r\"),Weight=0.0037),(Item=(Type=Bone,Name=\"spine_04\"),Weight=0.0063),(Item=(Type=Bone,Name=\"deform_up ...(truncated)"
    Weight=1
   )
   > &RigUnit_ParentConstraint(
    AdvancedSettings=(RotationOrderForFilter=XZY)
    Child=(Type=Bone,Name="MSTR_strap_02_l")
    Parents="((Item=(Type=Bone,Name=\"spine_02\"),Weight=0.0013),(Item=(Type=Bone,Name=\"neck_01\"),Weight=0.0104),(Item=(Type=Bone,Name=\"spine_03\"),Weight=0.0149),(Item=(Type=Bone,Name=\"neck_02\"),Weight=0.0207),(Item=(Type=Bone,Name=\"upperarm_twist_02_l\"),Weight=0.029),(I ...(truncated)"
    Weight=1
   )
   > &RigUnit_ParentConstraint(
    AdvancedSettings=(RotationOrderForFilter=XZY)
    Child=(Type=Bone,Name="MSTR_strap_02_r")
    Parents="((Item=(Type=Bone,Name=\"spine_02\"),Weight=0.0013),(Item=(Type=Bone,Name=\"neck_01\"),Weight=0.0104),(Item=(Type=Bone,Name=\"spine_03\"),Weight=0.0149),(Item=(Type=Bone,Name=\"neck_02\"),Weight=0.0207),(Item=(Type=Bone,Name=\"upperarm_twist_02_r\"),Weight=0.029),(I ...(truncated)"
    Weight=1
   )
   > &RigUnit_ParentConstraint(
    AdvancedSettings=(RotationOrderForFilter=XZY)
    Child=(Type=Bone,Name="MSTR_strap_03_l")
    Parents="((Item=(Type=Bone,Name=\"upperarm_l\"),Weight=0.0024),(Item=(Type=Bone,Name=\"spine_02\"),Weight=0.0071),(Item=(Type=Bone,Name=\"deform_upperarm_up_l\"),Weight=0.0212),(Item=(Type=Bone,Name=\"spine_03\"),Weight=0.0422),(Item=(Type=Bone,Name=\"spine_05\"),Weight=0.04 ...(truncated)"
    Weight=1
   )
   > &RigUnit_ParentConstraint(
    AdvancedSettings=(RotationOrderForFilter=XZY)
    Child=(Type=Bone,Name="MSTR_strap_03_r")
    Parents="((Item=(Type=Bone,Name=\"upperarm_r\"),Weight=0.0024),(Item=(Type=Bone,Name=\"spine_02\"),Weight=0.0071),(Item=(Type=Bone,Name=\"deform_upperarm_up_r\"),Weight=0.0212),(Item=(Type=Bone,Name=\"spine_03\"),Weight=0.0422),(Item=(Type=Bone,Name=\"spine_05\"),Weight=0.04 ...(truncated)"
    Weight=1
   )
   > &RigUnit_ParentConstraint(
    AdvancedSettings=(RotationOrderForFilter=XZY)
    Child=(Type=Bone,Name="MSTR_strap_04_l")
    Parents="((Item=(Type=Bone,Name=\"lowerarm_l\"),Weight=0.0012),(Item=(Type=Bone,Name=\"deform_upperarm_bk_l\"),Weight=0.0021),(Item=(Type=Bone,Name=\"dyn_sleeve_01_l\"),Weight=0.0057),(Item=(Type=Bone,Name=\"pelvis\"),Weight=0.0068),(Item=(Type=Bone,Name=\"spine_01\"),Weight ...(truncated)"
    Weight=1
   )
   > &RigUnit_ParentConstraint(
    AdvancedSettings=(RotationOrderForFilter=XZY)
    Child=(Type=Bone,Name="MSTR_strap_04_r")
    Parents="((Item=(Type=Bone,Name=\"lowerarm_r\"),Weight=0.0012),(Item=(Type=Bone,Name=\"deform_upperarm_bk_r\"),Weight=0.0021),(Item=(Type=Bone,Name=\"dyn_sleeve_01_r\"),Weight=0.0057),(Item=(Type=Bone,Name=\"pelvis\"),Weight=0.0068),(Item=(Type=Bone,Name=\"spine_01\"),Weight ...(truncated)"
    Weight=1
   )
   > &RigUnit_ParentConstraint(
    AdvancedSettings=(RotationOrderForFilter=XZY)
    Child=(Type=Bone,Name="MSTR_strap_05_l")
    Parents="((Item=(Type=Bone,Name=\"dyn_sleeve_01_l\"),Weight=0.0023),(Item=(Type=Bone,Name=\"upperarm_l\"),Weight=0.0041),(Item=(Type=Bone,Name=\"deform_pec_l\"),Weight=0.006),(Item=(Type=Bone,Name=\"upperarm_twist_02_l\"),Weight=0.01),(Item=(Type=Bone,Name=\"deform_upperarm ...(truncated)"
    Weight=1
   )
   > &RigUnit_ParentConstraint(
    AdvancedSettings=(RotationOrderForFilter=XZY)
    Child=(Type=Bone,Name="MSTR_strap_05_r")
    Parents="((Item=(Type=Bone,Name=\"dyn_sleeve_01_r\"),Weight=0.0023),(Item=(Type=Bone,Name=\"upperarm_r\"),Weight=0.0041),(Item=(Type=Bone,Name=\"deform_pec_r\"),Weight=0.006),(Item=(Type=Bone,Name=\"upperarm_twist_02_r\"),Weight=0.01),(Item=(Type=Bone,Name=\"deform_upperarm ...(truncated)"
    Weight=1
   )
   > &RigUnit_ParentConstraint(
    AdvancedSettings=(RotationOrderForFilter=XZY)
    Child=(Type=Bone,Name="MSTR_strap_06_l")
    Parents="((Item=(Type=Bone,Name=\"dyn_backShirt\"),Weight=0.0014),(Item=(Type=Bone,Name=\"upperarm_twist_01_l\"),Weight=0.0039),(Item=(Type=Bone,Name=\"deform_upperarm_bk_l\"),Weight=0.006),(Item=(Type=Bone,Name=\"dyn_sideShirt_l\"),Weight=0.0081),(Item=(Type=Bone,Name=\"cl ...(truncated)"
    Weight=1
   )
   > &RigUnit_ParentConstraint(
    AdvancedSettings=(RotationOrderForFilter=XZY)
    Child=(Type=Bone,Name="MSTR_strap_06_r")
    Parents="((Item=(Type=Bone,Name=\"dyn_backShirt\"),Weight=0.0014),(Item=(Type=Bone,Name=\"upperarm_twist_01_r\"),Weight=0.0039),(Item=(Type=Bone,Name=\"deform_upperarm_bk_r\"),Weight=0.006),(Item=(Type=Bone,Name=\"dyn_sideShirt_r\"),Weight=0.0081),(Item=(Type=Bone,Name=\"cl ...(truncated)"
    Weight=1
   )
   > DeltaRotation v4_6 = RigVMFunctionReferenceNode(
    Rotation Order=ZYX
    child=(Type=Bone,Name="upperarm_l")
    debugRotation=false
    outputMinMax=(X=0,Y=1)
    parent=(Type=Null,Name="upperarm_l_null")
    premult=1
    rotationMinMax=(X=0,Y=-90)
    triggerAxis=Y
    twistAxis=(X=1,Y=0,Z=0)
   )
   > &RigUnit_ModifyTransforms(
    ItemToModify="((Item=(Type=Bone,Name=\"dyn_strap_04_l\"),Transform=(Rotation=(X=-0.1696,Y=-0.0386,Z=-0.2187,W=-0.9602),Translation=(X=-3.1062,Y=3.8152,Z=0),Scale3D=(X=1,Y=1,Z=1))),(Item=(Type=Bone,Name=\"dyn_strap_03_l\"),Transform=(Rotation=(X=0.1724,Y=0,Z=0,W=0.985),Trans ...(truncated)"
    WeightMinimum=0
    Weight=@"DeltaRotation v4_6".outputWeight
   )
   > DeltaRotation v4_8 = RigVMFunctionReferenceNode(
    Rotation Order=ZYX
    child=(Type=Bone,Name="clavicle_l")
    debugRotation=false
    outputMinMax=(X=0,Y=1)
    parent=(Type=Null,Name="clavicle_l_null")
    premult=1
    rotationMinMax=(X=0,Y=-60)
    triggerAxis=Y
    twistAxis=(X=1,Y=0,Z=0)
   )
   > &RigUnit_ModifyTransforms(
    ItemToModify="((Item=(Type=Bone,Name=\"dyn_strap_01_l\"),Transform=(Rotation=(X=0.1354,Y=0.0326,Z=0.6863,W=0.7138),Translation=(X=2.3161,Y=0.802,Z=0.0185),Scale3D=(X=1,Y=1,Z=1))),(Item=(Type=Bone,Name=\"dyn_strap_02_l\"),Transform=(Rotation=(X=-0.0679,Y=0.0007,Z=0.47,W=0.88 ...(truncated)"
    WeightMinimum=0
    Weight=@"DeltaRotation v4_8".outputWeight
   )
   > &RigUnit_ModifyTransforms(
    ItemToModify=((Item=(Type=Bone,Name="dyn_strap_01_l"),Transform=(Rotation=(X=0,Y=0,Z=-0.5115,W=-0.8593),Translation=(X=3,Y=0,Z=0),Scale3D=(X=1,Y=1,Z=1))))
    WeightMinimum=0
    Weight=&RigVMFunction_MathDoubleMul(A=@"DeltaRotation v4_6".outputWeight, B=@"DeltaRotation v4_8".outputWeight).Result
   )
   > DeltaRotation v4_11 = RigVMFunctionReferenceNode(
    Rotation Order=ZYX
    child=(Type=Bone,Name="upperarm_l")
    debugRotation=false
    outputMinMax=(X=0,Y=1)
    parent=(Type=Null,Name="upperarm_l_null")
    premult=1
    rotationMinMax=(X=0,Y=60)
    triggerAxis=Y
    twistAxis=(X=1,Y=0,Z=0)
   )
   > &RigUnit_ModifyTransforms(
    ItemToModify="((Item=(Type=Bone,Name=\"dyn_strap_04_l\"),Transform=(Rotation=(X=-0.0243,Y=0.2086,Z=0.2042,W=-0.9561),Translation=(X=3.5136,Y=-0.4684,Z=1.1841),Scale3D=(X=1,Y=1,Z=1))),(Item=(Type=Bone,Name=\"dyn_strap_03_l\"),Transform=(Rotation=(X=0,Y=0,Z=0,W=1),Translation ...(truncated)"
    WeightMinimum=0
    Weight=@"DeltaRotation v4_11".outputWeight
   )
   > DeltaRotation v4_2 = RigVMFunctionReferenceNode(
    Rotation Order=ZYX
    child=(Type=Bone,Name="clavicle_l")
    debugRotation=false
    outputMinMax=(X=0,Y=1)
    parent=(Type=Null,Name="clavicle_l_null")
    premult=1
    rotationMinMax=(X=0,Y=35)
    triggerAxis=Y
    twistAxis=(X=1,Y=0,Z=0)
   )
   > &RigUnit_ModifyTransforms(
    ItemToModify="((Item=(Type=Bone,Name=\"dyn_strap_05_l\"),Transform=(Rotation=(X=0.0046,Y=0.0454,Z=-0.1009,W=0.9939),Translation=(X=2.2791,Y=1.878,Z=-0.1387),Scale3D=(X=1,Y=1,Z=1))),(Item=(Type=Bone,Name=\"dyn_strap_04_l\"),Transform=(Rotation=(X=0.006,Y=0.0541,Z=0.1095,W=-0 ...(truncated)"
    WeightMinimum=0
    Weight=@"DeltaRotation v4_2".outputWeight
   )
   > DeltaRotation v4_9 = RigVMFunctionReferenceNode(
    Rotation Order=YZX
    absoluteValue=false
    child=(Type=Bone,Name="upperarm_l")
    debugRotation=false
    outputMinMax=(X=0,Y=1)
    parent=(Type=Null,Name="upperarm_l_null")
    premult=1.1
    rotationMinMax=(X=0,Y=80)
    triggerAxis=Z
    twistAxis=(X=1,Y=0,Z=0)
   )
   > &RigUnit_ModifyTransforms(
    ItemToModify="((Item=(Type=Bone,Name=\"dyn_strap_04_l\"),Transform=(Rotation=(X=0.0048,Y=0.0673,Z=-0.3149,W=-0.9467),Translation=(X=2.932,Y=6.4167,Z=-3.5166),Scale3D=(X=1,Y=1,Z=1))),(Item=(Type=Bone,Name=\"dyn_strap_03_l\"),Transform=(Rotation=(X=0.1091,Y=0.0048,Z=0.0002,W= ...(truncated)"
    WeightMinimum=0
    Weight=@"DeltaRotation v4_9".outputWeight
   )
   > DeltaRotation v4_1 = RigVMFunctionReferenceNode(
    Rotation Order=ZYX
    child=(Type=Bone,Name="clavicle_l")
    debugRotation=false
    outputMinMax=(X=0,Y=1)
    parent=(Type=Null,Name="clavicle_l_null")
    premult=1
    rotationMinMax=(X=20,Y=45)
    triggerAxis=Z
    twistAxis=(X=1,Y=0,Z=0)
   )
   > &RigUnit_ModifyTransforms(
    ItemToModify="((Item=(Type=Bone,Name=\"dyn_strap_03_l\"),Transform=(Rotation=(X=-0.0033,Y=0.0106,Z=0.2997,W=0.954),Translation=(X=-1.5116,Y=1.3398,Z=-0.0269),Scale3D=(X=1,Y=1,Z=1))),(Item=(Type=Bone,Name=\"dyn_strap_02_l\"),Transform=(Rotation=(X=0,Y=0,Z=0,W=1),Translation= ...(truncated)"
    WeightMinimum=0
    Weight=@"DeltaRotation v4_1".outputWeight
   )
   > &RigUnit_ModifyTransforms(
    ItemToModify="((Item=(Type=Bone,Name=\"dyn_strap_04_l\"),Transform=(Rotation=(X=0,Y=0,Z=0,W=1),Translation=(X=-1.252,Y=0,Z=0),Scale3D=(X=1,Y=1,Z=1))),(Item=(Type=Bone,Name=\"dyn_strap_03_l\"),Transform=(Rotation=(X=0,Y=0,Z=0,W=1),Translation=(X=-1.0896,Y=0,Z=0),Scale3D=(X=1 ...(truncated)"
    WeightMinimum=0
    Weight=&RigVMFunction_MathDoubleMul(B=@"DeltaRotation v4_11".outputWeight, A=@"DeltaRotation v4_9".outputWeight).Result
   )
   n1 = &RigVMFunction_MathDoubleMul(A=@"DeltaRotation v4_6".outputWeight, B=@"DeltaRotation v4_9".outputWeight)
   > &RigUnit_ModifyTransforms(
    ItemToModify="((Item=(Type=Bone,Name=\"dyn_strap_04_l\"),Transform=(Rotation=(X=0.0451,Y=0.0892,Z=-0.4486,W=0.8881),Translation=(X=-6.4311,Y=-3.652,Z=4.3924),Scale3D=(X=1,Y=1,Z=1))),(Item=(Type=Bone,Name=\"dyn_strap_03_l\"),Transform=(Rotation=(X=0,Y=0,Z=-0.2429,W=0.97),Tra ...(truncated)"
    WeightMinimum=0
    Weight=@n1.Result
   )
   > DeltaRotation v4_28 = RigVMFunctionReferenceNode(
    Rotation Order=ZYX
    absoluteValue=true
    child=(Type=Bone,Name="upperarm_l")
    debugRotation=false
    outputMinMax=(X=0,Y=1)
    parent=(Type=Null,Name="upperarm_l_null")
    premult=1
    rotationMinMax=(X=80,Y=120)
    triggerAxis=Z
    twistAxis=(X=1,Y=0,Z=0)
   )
   > &RigUnit_ModifyTransforms(
    ItemToModify="((Item=(Type=Bone,Name=\"dyn_strap_04_l\"),Transform=(Rotation=(X=0,Y=0,Z=0,W=1),Translation=(X=3,Y=0,Z=0),Scale3D=(X=1,Y=1,Z=1))),(Item=(Type=Bone,Name=\"dyn_strap_03_l\"),Transform=(Rotation=(X=0,Y=0,Z=0,W=-1),Translation=(X=2,Y=0,Z=0),Scale3D=(X=1,Y=1,Z=1)) ...(truncated)"
    WeightMinimum=0
    Weight=@"DeltaRotation v4_28".outputWeight
   )
   > DeltaRotation v4_14 = RigVMFunctionReferenceNode(
    Rotation Order=ZYX
    child=(Type=Bone,Name="upperarm_l")
    debugRotation=false
    onlyTwist=true
    outputMinMax=(X=0,Y=1)
    parent=(Type=Null,Name="upperarm_l_null")
    premult=1
    rotationMinMax=(X=0,Y=-120)
    triggerAxis=X
    twistAxis=(X=1,Y=0,Z=0)
   )
   > &RigUnit_ModifyTransforms(
    ItemToModify="((Item=(Type=Bone,Name=\"dyn_strap_04_l\"),Transform=(Rotation=(X=0,Y=0,Z=0.0273,W=-0.9996),Translation=(X=1.0286,Y=-0.0562,Z=0),Scale3D=(X=1,Y=1,Z=1))),(Item=(Type=Bone,Name=\"dyn_strap_03_l\"),Transform=(Rotation=(X=0,Y=0,Z=0,W=1),Translation=(X=0,Y=0,Z=0),S ...(truncated)"
    WeightMinimum=0
    Weight=@"DeltaRotation v4_14".outputWeight
   )
   > DeltaRotation v4_3 = RigVMFunctionReferenceNode(
    Rotation Order=ZYX
    child=(Type=Bone,Name="upperarm_l")
    debugRotation=false
    onlyTwist=true
    outputMinMax=(X=0,Y=1)
    parent=(Type=Null,Name="upperarm_l_null")
    premult=1
    rotationMinMax=(X=0,Y=120)
    triggerAxis=X
    twistAxis=(X=1,Y=0,Z=0)
   )
   > &RigUnit_ModifyTransforms(
    ItemToModify="((Item=(Type=Bone,Name=\"dyn_strap_04_l\"),Transform=(Rotation=(X=0,Y=0,Z=-0.2408,W=-0.9706),Translation=(X=-0.4961,Y=1.6754,Z=0),Scale3D=(X=1,Y=1,Z=1))),(Item=(Type=Bone,Name=\"dyn_strap_03_l\"),Transform=(Rotation=(X=0,Y=0,Z=0,W=1),Translation=(X=0,Y=0.9779, ...(truncated)"
    WeightMinimum=0
    Weight=@"DeltaRotation v4_3".outputWeight
   )
   Multiply = &RigVMFunction_MathDoubleMul(B=@"DeltaRotation v4_14".outputWeight, A=@"DeltaRotation v4_9".outputWeight)
   > &RigUnit_ModifyTransforms(
    ItemToModify="((Item=(Type=Bone,Name=\"dyn_strap_04_l\"),Transform=(Rotation=(X=0,Y=0,Z=0.1563,W=-0.9877),Translation=(X=0,Y=0,Z=0),Scale3D=(X=1,Y=1,Z=1))),(Item=(Type=Bone,Name=\"dyn_strap_03_l\"),Transform=(Rotation=(X=0,Y=0,Z=-0.1047,W=0.9945),Translation=(X=0.8195,Y=0.0 ...(truncated)"
    WeightMinimum=0
    Weight=@Multiply.Result
   )
   n2 = &RigVMFunction_MathDoubleMul(B=@"DeltaRotation v4_11".outputWeight, A=@"DeltaRotation v4_14".outputWeight)
   > &RigUnit_ModifyTransforms(
    ItemToModify="((Item=(Type=Bone,Name=\"dyn_strap_04_l\"),Transform=(Rotation=(X=0,Y=0,Z=0.188,W=-0.9822),Translation=(X=0,Y=0,Z=0),Scale3D=(X=1,Y=1,Z=1))),(Item=(Type=Bone,Name=\"dyn_strap_03_l\"),Transform=(Rotation=(X=0,Y=0,Z=-0.2063,W=0.9785),Translation=(X=0.4528,Y=-0.1 ...(truncated)"
    WeightMinimum=0
    Weight=@n2.Result
   )
   > &RigUnit_ModifyTransforms(
    ItemToModify="((Item=(Type=Bone,Name=\"dyn_strap_04_l\"),Transform=(Rotation=(X=0,Y=0,Z=0,W=-1),Translation=(X=4.3871,Y=0,Z=-0.2449),Scale3D=(X=1,Y=1,Z=1))),(Item=(Type=Bone,Name=\"dyn_strap_03_l\"),Transform=(Rotation=(X=0,Y=0,Z=0,W=1),Translation=(X=0,Y=0,Z=0),Scale3D=(X= ...(truncated)"
    WeightMinimum=0
    Weight=&RigVMFunction_MathDoubleMul(A=@Multiply.Result, B=@n2.Result).Result
   )
   > DeltaRotation v4_10 = RigVMFunctionReferenceNode(
    Rotation Order=ZYX
    child=(Type=Bone,Name="clavicle_l")
    debugRotation=false
    outputMinMax=(X=0,Y=1)
    parent=(Type=Null,Name="clavicle_l_null")
    premult=1
    rotationMinMax=(X=0,Y=-35)
    triggerAxis=Z
    twistAxis=(X=1,Y=0,Z=0)
   )
   > &RigUnit_ModifyTransforms(
    ItemToModify="((Item=(Type=Bone,Name=\"dyn_strap_04_l\"),Transform=(Rotation=(X=0,Y=0,Z=-0.2872,W=0.9579),Translation=(X=4.7549,Y=-0.923,Z=0),Scale3D=(X=1,Y=1,Z=1))),(Item=(Type=Bone,Name=\"dyn_strap_03_l\"),Transform=(Rotation=(X=0,Y=0,Z=0,W=1),Translation=(X=2.2289,Y=0,Z= ...(truncated)"
    WeightMinimum=0
    Weight=&RigVMFunction_MathDoubleMul(A=@"DeltaRotation v4_10".outputWeight, B=@"DeltaRotation v4_9".outputWeight).Result
   )
   > &RigUnit_ModifyTransforms(
    ItemToModify="((Item=(Type=Bone,Name=\"dyn_strap_04_l\"),Transform=(Rotation=(X=0,Y=0,Z=-0.2872,W=0.9579),Translation=(X=4.7549,Y=-0.923,Z=0),Scale3D=(X=1,Y=1,Z=1))),(Item=(Type=Bone,Name=\"dyn_strap_03_l\"),Transform=(Rotation=(X=0,Y=0,Z=0,W=1),Translation=(X=2.2289,Y=0,Z= ...(truncated)"
    WeightMinimum=0
    Weight=&RigVMFunction_MathDoubleMul(B=@"DeltaRotation v4_3".outputWeight, A=@n1.Result).Result
   )
   Multiply_27 = RigVMAggregateNode(C=1, A=@"DeltaRotation v4_8".outputWeight, B=@"DeltaRotation v4_9".outputWeight)
   > &RigUnit_ModifyTransforms(
    ItemToModify="((Item=(Type=Bone,Name=\"dyn_strap_04_l\"),Transform=(Rotation=(X=0,Y=0,Z=0,W=-1),Translation=(X=-1.3555,Y=-0.8601,Z=0),Scale3D=(X=1,Y=1,Z=1))),(Item=(Type=Bone,Name=\"dyn_strap_03_l\"),Transform=(Rotation=(X=0,Y=0,Z=0,W=1),Translation=(X=-0.7996,Y=0.6845,Z=0) ...(truncated)"
    WeightMaximum=0.5
    WeightMinimum=0
    Weight=@Multiply_27.Result
   )
   > DeltaRotation v4_17 = RigVMFunctionReferenceNode(
    Rotation Order=ZYX
    child=(Type=Bone,Name="upperarm_r")
    debugRotation=false
    outputMinMax=(X=0,Y=1)
    parent=(Type=Null,Name="upperarm_r_null")
    premult=1
    rotationMinMax=(X=0,Y=-90)
    triggerAxis=Y
    twistAxis=(X=1,Y=0,Z=0)
   )
   > &RigUnit_ModifyTransforms(
    ItemToModify="((Item=(Type=Bone,Name=\"dyn_strap_04_r\"),Transform=(Rotation=(X=-0.1696,Y=-0.0386,Z=-0.2187,W=-0.9602),Translation=(X=3.1062,Y=-3.8152,Z=0),Scale3D=(X=1,Y=1,Z=1))),(Item=(Type=Bone,Name=\"dyn_strap_03_r\"),Transform=(Rotation=(X=0.1724,Y=0,Z=0,W=0.985),Trans ...(truncated)"
    WeightMinimum=0
    Weight=@"DeltaRotation v4_17".outputWeight
   )
   > DeltaRotation v4_13 = RigVMFunctionReferenceNode(
    Rotation Order=ZYX
    child=(Type=Bone,Name="clavicle_r")
    debugRotation=false
    outputMinMax=(X=0,Y=1)
    parent=(Type=Null,Name="clavicle_r_null")
    premult=1
    rotationMinMax=(X=0,Y=-60)
    triggerAxis=Y
    twistAxis=(X=1,Y=0,Z=0)
   )
   > &RigUnit_ModifyTransforms(
    ItemToModify="((Item=(Type=Bone,Name=\"dyn_strap_01_r\"),Transform=(Rotation=(X=0.1354,Y=0.0326,Z=0.6863,W=0.7138),Translation=(X=-2.3161,Y=-0.802,Z=-0.0185),Scale3D=(X=1,Y=1,Z=1))),(Item=(Type=Bone,Name=\"dyn_strap_02_r\"),Transform=(Rotation=(X=-0.0679,Y=0.0007,Z=0.47,W=0 ...(truncated)"
    WeightMinimum=0
    Weight=@"DeltaRotation v4_13".outputWeight
   )
   > &RigUnit_ModifyTransforms(
    ItemToModify=((Item=(Type=Bone,Name="dyn_strap_01_r"),Transform=(Rotation=(X=0,Y=0,Z=-0.5115,W=-0.8593),Translation=(X=-3,Y=0,Z=0),Scale3D=(X=1,Y=1,Z=1))))
    WeightMinimum=0
    Weight=&RigVMFunction_MathDoubleMul(B=@"DeltaRotation v4_13".outputWeight, A=@"DeltaRotation v4_17".outputWeight).Result
   )
   > DeltaRotation v4_19 = RigVMFunctionReferenceNode(
    Rotation Order=ZYX
    child=(Type=Bone,Name="upperarm_r")
    debugRotation=false
    outputMinMax=(X=0,Y=1)
    parent=(Type=Null,Name="upperarm_r_null")
    premult=1
    rotationMinMax=(X=0,Y=60)
    triggerAxis=Y
    twistAxis=(X=1,Y=0,Z=0)
   )
   > &RigUnit_ModifyTransforms(
    ItemToModify="((Item=(Type=Bone,Name=\"dyn_strap_04_r\"),Transform=(Rotation=(X=0.0558,Y=0.2025,Z=-0.1698,W=-0.9628),Translation=(X=-2.6463,Y=-0.6396,Z=-0.7481),Scale3D=(X=1,Y=1,Z=1))),(Item=(Type=Bone,Name=\"dyn_strap_03_r\"),Transform=(Rotation=(X=0,Y=0.1335,Z=0,W=-0.991) ...(truncated)"
    WeightMinimum=0
    Weight=@"DeltaRotation v4_19".outputWeight
   )
   > DeltaRotation v4_16 = RigVMFunctionReferenceNode(
    Rotation Order=ZYX
    child=(Type=Bone,Name="clavicle_r")
    debugRotation=false
    outputMinMax=(X=0,Y=1)
    parent=(Type=Null,Name="clavicle_r_null")
    premult=1
    rotationMinMax=(X=0,Y=35)
    triggerAxis=Y
    twistAxis=(X=1,Y=0,Z=0)
   )
   > &RigUnit_ModifyTransforms(
    ItemToModify="((Item=(Type=Bone,Name=\"dyn_strap_05_r\"),Transform=(Rotation=(X=0.0046,Y=0.0454,Z=-0.1009,W=0.9939),Translation=(X=-2.2791,Y=-1.878,Z=0.1387),Scale3D=(X=1,Y=1,Z=1))),(Item=(Type=Bone,Name=\"dyn_strap_04_r\"),Transform=(Rotation=(X=0.006,Y=0.0541,Z=0.1095,W=- ...(truncated)"
    WeightMinimum=0
    Weight=@"DeltaRotation v4_16".outputWeight
   )
   > DeltaRotation v4_18 = RigVMFunctionReferenceNode(
    Rotation Order=ZYX
    child=(Type=Bone,Name="upperarm_r")
    debugRotation=false
    outputMinMax=(X=0,Y=1)
    parent=(Type=Null,Name="upperarm_r_null")
    premult=1
    rotationMinMax=(X=0,Y=80)
    triggerAxis=Z
    twistAxis=(X=1,Y=0,Z=0)
   )
   > &RigUnit_ModifyTransforms(
    ItemToModify="((Item=(Type=Bone,Name=\"dyn_strap_04_r\"),Transform=(Rotation=(X=0.0048,Y=0.0673,Z=-0.3149,W=-0.9467),Translation=(X=-2.932,Y=-6.4167,Z=3.5166),Scale3D=(X=1,Y=1,Z=1))),(Item=(Type=Bone,Name=\"dyn_strap_03_r\"),Transform=(Rotation=(X=0.1091,Y=0.0048,Z=0.0002,W ...(truncated)"
    WeightMinimum=0
    Weight=@"DeltaRotation v4_18".outputWeight
   )
   > DeltaRotation v4_15 = RigVMFunctionReferenceNode(
    Rotation Order=ZYX
    child=(Type=Bone,Name="clavicle_r")
    debugRotation=false
    outputMinMax=(X=0,Y=1)
    parent=(Type=Null,Name="clavicle_r_null")
    premult=1
    rotationMinMax=(X=20,Y=45)
    triggerAxis=Z
    twistAxis=(X=1,Y=0,Z=0)
   )
   > &RigUnit_ModifyTransforms(
    ItemToModify="((Item=(Type=Bone,Name=\"dyn_strap_03_r\"),Transform=(Rotation=(X=-0.0033,Y=0.0106,Z=0.2997,W=0.954),Translation=(X=1.5116,Y=-1.3398,Z=0.0269),Scale3D=(X=1,Y=1,Z=1))),(Item=(Type=Bone,Name=\"dyn_strap_02_r\"),Transform=(Rotation=(X=0,Y=0,Z=0,W=1),Translation=( ...(truncated)"
    WeightMinimum=0
    Weight=@"DeltaRotation v4_15".outputWeight
   )
   > &RigUnit_ModifyTransforms(
    ItemToModify="((Item=(Type=Bone,Name=\"dyn_strap_04_r\"),Transform=(Rotation=(X=0,Y=0,Z=0,W=1),Translation=(X=-0.5,Y=1,Z=0),Scale3D=(X=1,Y=1,Z=1))),(Item=(Type=Bone,Name=\"dyn_strap_03_r\"),Transform=(Rotation=(X=0,Y=-0.1736,Z=0,W=0.9848),Translation=(X=-3,Y=0,Z=0),Scale3D= ...(truncated)"
    WeightMaximum=0
    WeightMinimum=0
    Weight=&RigVMFunction_MathDoubleMul(B=@"DeltaRotation v4_19".outputWeight, A=@"DeltaRotation v4_18".outputWeight).Result
   )
   n3 = &RigVMFunction_MathDoubleMul(A=@"DeltaRotation v4_17".outputWeight, B=@"DeltaRotation v4_18".outputWeight)
   > &RigUnit_ModifyTransforms(
    ItemToModify="((Item=(Type=Bone,Name=\"dyn_strap_04_r\"),Transform=(Rotation=(X=0.0451,Y=0.0892,Z=-0.4486,W=0.8881),Translation=(X=6.4311,Y=3.652,Z=-4.3924),Scale3D=(X=1,Y=1,Z=1))),(Item=(Type=Bone,Name=\"dyn_strap_03_r\"),Transform=(Rotation=(X=0,Y=0,Z=-0.2429,W=0.97),Tran ...(truncated)"
    WeightMinimum=0
    Weight=@n3.Result
   )
   > DeltaRotation v4_27 = RigVMFunctionReferenceNode(
    Rotation Order=ZYX
    child=(Type=Bone,Name="upperarm_r")
    debugRotation=false
    outputMinMax=(X=0,Y=1)
    parent=(Type=Null,Name="upperarm_r_null")
    premult=1
    rotationMinMax=(X=80,Y=120)
    triggerAxis=Z
    twistAxis=(X=1,Y=0,Z=0)
   )
   > &RigUnit_ModifyTransforms(
    ItemToModify="((Item=(Type=Bone,Name=\"dyn_strap_04_r\"),Transform=(Rotation=(X=0,Y=0,Z=0,W=1),Translation=(X=-2,Y=0,Z=0),Scale3D=(X=1,Y=1,Z=1))),(Item=(Type=Bone,Name=\"dyn_strap_03_r\"),Transform=(Rotation=(X=0,Y=0,Z=0,W=-1),Translation=(X=-2,Y=0,Z=0),Scale3D=(X=1,Y=1,Z=1 ...(truncated)"
    WeightMinimum=0
    Weight=@"DeltaRotation v4_27".outputWeight
   )
   > DeltaRotation v4_20 = RigVMFunctionReferenceNode(
    Rotation Order=ZYX
    child=(Type=Bone,Name="upperarm_r")
    debugRotation=false
    onlyTwist=true
    outputMinMax=(X=0,Y=1)
    parent=(Type=Null,Name="upperarm_r_null")
    premult=1
    rotationMinMax=(X=0,Y=-120)
    triggerAxis=X
    twistAxis=(X=1,Y=0,Z=0)
   )
   > &RigUnit_ModifyTransforms(
    ItemToModify="((Item=(Type=Bone,Name=\"dyn_strap_04_r\"),Transform=(Rotation=(X=0,Y=0,Z=0.0273,W=-0.9996),Translation=(X=-1.0286,Y=0.0562,Z=0),Scale3D=(X=1,Y=1,Z=1))),(Item=(Type=Bone,Name=\"dyn_strap_03_r\"),Transform=(Rotation=(X=0,Y=0,Z=0,W=1),Translation=(X=0,Y=0,Z=0),S ...(truncated)"
    WeightMinimum=0
    Weight=@"DeltaRotation v4_20".outputWeight
   )
   > DeltaRotation v4_21 = RigVMFunctionReferenceNode(
    Rotation Order=ZYX
    child=(Type=Bone,Name="upperarm_r")
    debugRotation=false
    onlyTwist=true
    outputMinMax=(X=0,Y=1)
    parent=(Type=Null,Name="upperarm_r_null")
    premult=1
    rotationMinMax=(X=0,Y=120)
    triggerAxis=X
    twistAxis=(X=1,Y=0,Z=0)
   )
   > &RigUnit_ModifyTransforms(
    ItemToModify="((Item=(Type=Bone,Name=\"dyn_strap_04_r\"),Transform=(Rotation=(X=0,Y=0,Z=-0.2408,W=-0.9706),Translation=(X=0.4961,Y=-1.6754,Z=0),Scale3D=(X=1,Y=1,Z=1))),(Item=(Type=Bone,Name=\"dyn_strap_03_r\"),Transform=(Rotation=(X=0,Y=0,Z=0,W=1),Translation=(X=0,Y=-0.9779 ...(truncated)"
    WeightMinimum=0
    Weight=@"DeltaRotation v4_21".outputWeight
   )
   n4 = &RigVMFunction_MathDoubleMul(A=@"DeltaRotation v4_18".outputWeight, B=@"DeltaRotation v4_20".outputWeight)
   > &RigUnit_ModifyTransforms(
    ItemToModify="((Item=(Type=Bone,Name=\"dyn_strap_04_r\"),Transform=(Rotation=(X=0,Y=0,Z=0.1563,W=-0.9877),Translation=(X=0,Y=0,Z=0),Scale3D=(X=1,Y=1,Z=1))),(Item=(Type=Bone,Name=\"dyn_strap_03_r\"),Transform=(Rotation=(X=0,Y=0,Z=-0.1047,W=0.9945),Translation=(X=-0.8195,Y=-0 ...(truncated)"
    WeightMinimum=0
    Weight=@n4.Result
   )
   n5 = &RigVMFunction_MathDoubleMul(B=@"DeltaRotation v4_19".outputWeight, A=@"DeltaRotation v4_20".outputWeight)
   > &RigUnit_ModifyTransforms(
    ItemToModify="((Item=(Type=Bone,Name=\"dyn_strap_04_r\"),Transform=(Rotation=(X=0,Y=0,Z=0.188,W=-0.9822),Translation=(X=0,Y=0,Z=0),Scale3D=(X=1,Y=1,Z=1))),(Item=(Type=Bone,Name=\"dyn_strap_03_r\"),Transform=(Rotation=(X=0,Y=0,Z=-0.2063,W=0.9785),Translation=(X=-0.4528,Y=0.1 ...(truncated)"
    WeightMinimum=0
    Weight=@n5.Result
   )
   > &RigUnit_ModifyTransforms(
    ItemToModify="((Item=(Type=Bone,Name=\"dyn_strap_04_r\"),Transform=(Rotation=(X=0,Y=0,Z=0,W=-1),Translation=(X=-4.3871,Y=0,Z=0.2449),Scale3D=(X=1,Y=1,Z=1))),(Item=(Type=Bone,Name=\"dyn_strap_03_r\"),Transform=(Rotation=(X=0,Y=0,Z=0,W=1),Translation=(X=0,Y=0,Z=0),Scale3D=(X= ...(truncated)"
    WeightMinimum=0
    Weight=&RigVMFunction_MathDoubleMul(B=@n5.Result, A=@n4.Result).Result
   )
   Multiply_33 = RigVMAggregateNode(A=@"DeltaRotation v4_13".outputWeight, B=@"DeltaRotation v4_19".outputWeight, C=@"DeltaRotation v4_20".outputWeight)
   > &RigUnit_ModifyTransforms(
    ItemToModify="((Item=(Type=Bone,Name=\"dyn_strap_04_r\"),Transform=(Rotation=(X=0,Y=0,Z=0,W=-1),Translation=(X=-5,Y=0,Z=0),Scale3D=(X=1,Y=1,Z=1))),(Item=(Type=Bone,Name=\"dyn_strap_03_r\"),Transform=(Rotation=(X=0,Y=0,Z=0,W=1),Translation=(X=-1,Y=0,Z=0),Scale3D=(X=1,Y=1,Z=1 ...(truncated)"
    WeightMaximum=0.05
    WeightMinimum=0
    Weight=@Multiply_33.Result
   )
   > DeltaRotation v4_22 = RigVMFunctionReferenceNode(
    Rotation Order=ZYX
    child=(Type=Bone,Name="clavicle_r")
    debugRotation=false
    outputMinMax=(X=0,Y=1)
    parent=(Type=Null,Name="clavicle_r_null")
    premult=1
    rotationMinMax=(X=0,Y=-35)
    triggerAxis=Z
    twistAxis=(X=1,Y=0,Z=0)
   )
   > &RigUnit_ModifyTransforms(
    ItemToModify="((Item=(Type=Bone,Name=\"dyn_strap_04_r\"),Transform=(Rotation=(X=0,Y=0,Z=-0.2872,W=0.9579),Translation=(X=-4.7549,Y=0.923,Z=0),Scale3D=(X=1,Y=1,Z=1))),(Item=(Type=Bone,Name=\"dyn_strap_03_r\"),Transform=(Rotation=(X=0,Y=0,Z=0,W=1),Translation=(X=-2.2289,Y=0,Z ...(truncated)"
    WeightMinimum=0
    Weight=&RigVMFunction_MathDoubleMul(B=@"DeltaRotation v4_18".outputWeight, A=@"DeltaRotation v4_22".outputWeight).Result
   )
   > &RigUnit_ModifyTransforms(
    ItemToModify="((Item=(Type=Bone,Name=\"dyn_strap_04_r\"),Transform=(Rotation=(X=0,Y=0,Z=0.3244,W=-0.9459),Translation=(X=-0.9033,Y=5.7225,Z=0),Scale3D=(X=1,Y=1,Z=1))),(Item=(Type=Bone,Name=\"dyn_strap_03_r\"),Transform=(Rotation=(X=0,Y=0,Z=0.1702,W=-0.9854),Translation=(X=- ...(truncated)"
    WeightMinimum=0
    Weight=&RigVMFunction_MathDoubleMul(A=@"DeltaRotation v4_21".outputWeight, B=@n3.Result).Result
   )
   Multiply_26 = RigVMAggregateNode(C=1, B=@"DeltaRotation v4_13".outputWeight, A=@"DeltaRotation v4_18".outputWeight)
   > &RigUnit_ModifyTransforms(
    ItemToModify="((Item=(Type=Bone,Name=\"dyn_strap_04_r\"),Transform=(Rotation=(X=0,Y=0,Z=0,W=1),Translation=(X=1.3555,Y=0.8601,Z=0),Scale3D=(X=1,Y=1,Z=1))),(Item=(Type=Bone,Name=\"dyn_strap_03_r\"),Transform=(Rotation=(X=0,Y=0,Z=0,W=1),Translation=(X=0.7996,Y=-0.6845,Z=0),Sc ...(truncated)"
    WeightMaximum=0.5
    WeightMinimum=0
    Weight=@Multiply_26.Result
   )
   > &RigUnit_ModifyTransforms(
    ItemToModify="((Item=(Type=Bone,Name=\"dyn_strap_04_r\"),Transform=(Rotation=(X=0,Y=0,Z=0,W=1),Translation=(X=-2,Y=0,Z=0),Scale3D=(X=1,Y=1,Z=1))),(Item=(Type=Bone,Name=\"dyn_strap_03_r\"),Transform=(Rotation=(X=0,Y=0,Z=0,W=-1),Translation=(X=-2,Y=0,Z=0),Scale3D=(X=1,Y=1,Z=1 ...(truncated)"
    WeightMaximum=0
    WeightMinimum=0
    Weight=&RigVMFunction_MathDoubleMul(A=@"DeltaRotation v4_18".outputWeight, B=@"DeltaRotation v4_21".outputWeight).Result
   )
   > DeltaRotation v4_4 = RigVMFunctionReferenceNode(
    Rotation Order=ZYX
    child=(Type=Bone,Name="spine_05")
    debugRotation=false
    onlyTwist=true
    outputMinMax=(X=0,Y=1)
    parent=(Type=Null,Name="spine_05_null")
    premult=1
    rotationMinMax=(X=0,Y=70)
    triggerAxis=X
    twistAxis=(X=1,Y=0,Z=0)
   )
   > &RigUnit_ModifyTransforms(
    ItemToModify="((Item=(Type=Bone,Name=\"dyn_strap_04_l\"),Transform=(Rotation=(X=-0.0007,Y=0.286,Z=0.0049,W=-0.9582),Translation=(X=0,Y=2.0908,Z=0),Scale3D=(X=1,Y=1,Z=1))),(Item=(Type=Bone,Name=\"dyn_strap_03_l\"),Transform=(Rotation=(X=0,Y=0,Z=0,W=1),Translation=(X=0,Y=4.04 ...(truncated)"
    WeightMinimum=0
    Weight=@"DeltaRotation v4_4".outputWeight
   )
   > DeltaRotation v4_24 = RigVMFunctionReferenceNode(
    Rotation Order=ZYX
    child=(Type=Bone,Name="spine_05")
    debugRotation=false
    onlyTwist=true
    outputMinMax=(X=0,Y=1)
    parent=(Type=Null,Name="spine_05_null")
    premult=1
    rotationMinMax=(X=0,Y=-70)
    triggerAxis=X
    twistAxis=(X=1,Y=0,Z=0)
   )
   > &RigUnit_ModifyTransforms(
    ItemToModify="((Item=(Type=Bone,Name=\"dyn_strap_04_r\"),Transform=(Rotation=(X=-0.0007,Y=0.286,Z=0.0049,W=-0.9582),Translation=(X=0,Y=-2.0908,Z=0),Scale3D=(X=1,Y=1,Z=1))),(Item=(Type=Bone,Name=\"dyn_strap_03_r\"),Transform=(Rotation=(X=0,Y=0,Z=0,W=1),Translation=(X=0,Y=-4. ...(truncated)"
    WeightMinimum=0
    Weight=@"DeltaRotation v4_24".outputWeight
   )
   > DeltaRotation v4_5 = RigVMFunctionReferenceNode(
    Rotation Order=ZYX
    child=(Type=Bone,Name="spine_04")
    debugRotation=false
    onlyTwist=true
    outputMinMax=(X=0,Y=1)
    parent=(Type=Null,Name="spine_04_null")
    premult=1
    rotationMinMax=(X=0,Y=45)
    triggerAxis=X
    twistAxis=(X=1,Y=0,Z=0)
   )
   > &RigUnit_ModifyTransforms(
    ItemToModify="((Item=(Type=Bone,Name=\"dyn_strap_04_l\"),Transform=(Rotation=(X=0,Y=0.131,Z=0,W=0.9914),Translation=(X=-0.9,Y=0,Z=0),Scale3D=(X=1,Y=1,Z=1))),(Item=(Type=Bone,Name=\"dyn_strap_03_l\"),Transform=(Rotation=(X=0.0123,Y=0.0586,Z=0.2044,W=0.9771),Translation=(X=-0 ...(truncated)"
    WeightMinimum=0
    Weight=@"DeltaRotation v4_5".outputWeight
   )
   > DeltaRotation v4_25 = RigVMFunctionReferenceNode(
    Rotation Order=ZYX
    child=(Type=Bone,Name="spine_04")
    debugRotation=false
    onlyTwist=true
    outputMinMax=(X=0,Y=1)
    parent=(Type=Null,Name="spine_04_null")
    premult=1
    rotationMinMax=(X=0,Y=-45)
    triggerAxis=X
    twistAxis=(X=1,Y=0,Z=0)
   )
   > &RigUnit_ModifyTransforms(
    ItemToModify="((Item=(Type=Bone,Name=\"dyn_strap_04_r\"),Transform=(Rotation=(X=0,Y=0.131,Z=0,W=0.9914),Translation=(X=0.9,Y=0,Z=0),Scale3D=(X=1,Y=1,Z=1))),(Item=(Type=Bone,Name=\"dyn_strap_03_r\"),Transform=(Rotation=(X=0.0123,Y=0.0586,Z=0.2044,W=0.9771),Translation=(X=0.2 ...(truncated)"
    WeightMinimum=0
    Weight=@"DeltaRotation v4_25".outputWeight
   )
   > DeltaRotation v4_7 = RigVMFunctionReferenceNode(
    Rotation Order=ZYX
    child=(Type=Bone,Name="spine_03")
    debugRotation=false
    onlyTwist=true
    outputMinMax=(X=0,Y=1)
    parent=(Type=Null,Name="spine_03_null")
    premult=1
    rotationMinMax=(X=0,Y=45)
    triggerAxis=X
    twistAxis=(X=1,Y=0,Z=0)
   )
   > &RigUnit_ModifyTransforms(
    ItemToModify="((Item=(Type=Bone,Name=\"dyn_strap_04_l\"),Transform=(Rotation=(X=0,Y=-0.1411,Z=0,W=-0.99),Translation=(X=1.3583,Y=0,Z=-0.1015),Scale3D=(X=1,Y=1,Z=1))),(Item=(Type=Bone,Name=\"dyn_strap_03_l\"),Transform=(Rotation=(X=0,Y=0.0001,Z=0,W=1),Translation=(X=0.636,Y= ...(truncated)"
    WeightMinimum=0
    Weight=@"DeltaRotation v4_7".outputWeight
   )
   > DeltaRotation v4_23 = RigVMFunctionReferenceNode(
    Rotation Order=ZYX
    child=(Type=Bone,Name="spine_03")
    debugRotation=false
    onlyTwist=true
    outputMinMax=(X=0,Y=1)
    parent=(Type=Null,Name="spine_03_null")
    premult=1
    rotationMinMax=(X=0,Y=-45)
    triggerAxis=X
    twistAxis=(X=1,Y=0,Z=0)
   )
   > &RigUnit_ModifyTransforms(
    ItemToModify="((Item=(Type=Bone,Name=\"dyn_strap_04_r\"),Transform=(Rotation=(X=0,Y=-0.1411,Z=0,W=-0.99),Translation=(X=-1.3583,Y=0,Z=0.1015),Scale3D=(X=1,Y=1,Z=1))),(Item=(Type=Bone,Name=\"dyn_strap_03_r\"),Transform=(Rotation=(X=0,Y=0.0001,Z=0,W=1),Translation=(X=-0.636,Y ...(truncated)"
    WeightMinimum=0
    Weight=@"DeltaRotation v4_23".outputWeight
   )
   > &RigUnit_ModifyTransforms(
    ItemToModify="((Item=(Type=Bone,Name=\"dyn_strap_04_l\"),Transform=(Rotation=(X=-0.5006,Y=0.0016,Z=0.1035,W=0.8595),Translation=(X=4.9405,Y=0.0136,Z=-3.8583),Scale3D=(X=1,Y=1,Z=1))),(Item=(Type=Bone,Name=\"dyn_strap_03_l\"),Transform=(Rotation=(X=0,Y=0,Z=0,W=1),Translation= ...(truncated)"
    WeightMinimum=0
    Weight=&RigVMFunction_MathDoubleMul(B=@"DeltaRotation v4_4".outputWeight, A=@"DeltaRotation v4_1".outputWeight).Result
   )
   > DeltaRotation v4_12 = RigVMFunctionReferenceNode(
    Rotation Order=ZYX
    child=(Type=Bone,Name="spine_05")
    debugRotation=false
    onlyTwist=true
    outputMinMax=(X=0,Y=1)
    parent=(Type=Null,Name="spine_05_null")
    premult=1
    rotationMinMax=(X=0,Y=45)
    triggerAxis=X
    twistAxis=(X=1,Y=0,Z=0)
   )
   Multiply_7 = RigVMAggregateNode(C=@"DeltaRotation v4_12".outputWeight, A=@"DeltaRotation v4_3".outputWeight, B=@"DeltaRotation v4_9".outputWeight)
   > &RigUnit_ModifyTransforms(
    ItemToModify="((Item=(Type=Bone,Name=\"dyn_strap_04_l\"),Transform=(Rotation=(X=0,Y=0,Z=0,W=-1),Translation=(X=0,Y=-7.5,Z=0),Scale3D=(X=1,Y=1,Z=1))),(Item=(Type=Bone,Name=\"dyn_strap_03_l\"),Transform=(Rotation=(X=0,Y=0,Z=0,W=1),Translation=(X=0,Y=0,Z=0),Scale3D=(X=1,Y=1,Z= ...(truncated)"
    WeightMinimum=0
    Weight=@Multiply_7.Result
   )
   > &RigUnit_ModifyTransforms(
    ItemToModify="((Item=(Type=Bone,Name=\"dyn_strap_04_r\"),Transform=(Rotation=(X=-0.5006,Y=0.0016,Z=0.1035,W=0.8595),Translation=(X=-4.9405,Y=-0.0136,Z=3.8583),Scale3D=(X=1,Y=1,Z=1))),(Item=(Type=Bone,Name=\"dyn_strap_03_r\"),Transform=(Rotation=(X=0,Y=0,Z=0,W=1),Translation ...(truncated)"
    WeightMinimum=0
    Weight=&RigVMFunction_MathDoubleMul(B=@"DeltaRotation v4_15".outputWeight, A=@"DeltaRotation v4_24".outputWeight).Result
   )
   > DeltaRotation v4_26 = RigVMFunctionReferenceNode(
    Rotation Order=ZYX
    child=(Type=Bone,Name="spine_05")
    debugRotation=false
    onlyTwist=true
    outputMinMax=(X=0,Y=1)
    parent=(Type=Null,Name="spine_05_null")
    premult=1
    rotationMinMax=(X=0,Y=-45)
    triggerAxis=X
    twistAxis=(X=1,Y=0,Z=0)
   )
   Multiply_15 = RigVMAggregateNode(A=@"DeltaRotation v4_18".outputWeight, B=@"DeltaRotation v4_21".outputWeight, C=@"DeltaRotation v4_26".outputWeight)
   > &RigUnit_ModifyTransforms(
    ItemToModify="((Item=(Type=Bone,Name=\"dyn_strap_04_r\"),Transform=(Rotation=(X=0,Y=0,Z=0,W=-1),Translation=(X=0,Y=7.5,Z=0),Scale3D=(X=1,Y=1,Z=1))),(Item=(Type=Bone,Name=\"dyn_strap_03_r\"),Transform=(Rotation=(X=0,Y=0,Z=0,W=1),Translation=(X=0,Y=0,Z=0),Scale3D=(X=1,Y=1,Z=1 ...(truncated)"
    WeightMinimum=0
    Weight=@Multiply_15.Result
   )
   Multiply_16 = RigVMAggregateNode(B=@"DeltaRotation v4_1".outputWeight, C=@"DeltaRotation v4_26".outputWeight, A=@"DeltaRotation v4_8".outputWeight)
   > &RigUnit_ModifyTransforms(
    ItemToModify=((Item=(Type=Bone,Name="dyn_strap_04_l"),Transform=(Rotation=(X=0,Y=0,Z=0,W=-1),Translation=(X=2.5897,Y=2.2791,Z=0),Scale3D=(X=1,Y=1,Z=1))))
    WeightMinimum=0
    Weight=@Multiply_16.Result
   )
   Multiply_17 = RigVMAggregateNode(C=@"DeltaRotation v4_12".outputWeight, A=@"DeltaRotation v4_13".outputWeight, B=@"DeltaRotation v4_15".outputWeight)
   > &RigUnit_ModifyTransforms(
    ItemToModify=((Item=(Type=Bone,Name="dyn_strap_04_r"),Transform=(Rotation=(X=0,Y=0,Z=0,W=-1),Translation=(X=-2.5897,Y=-2.2791,Z=0),Scale3D=(X=1,Y=1,Z=1))))
    WeightMinimum=0
    Weight=@Multiply_17.Result
   )
   > &RigUnit_ModifyTransforms(
    ItemToModify=((Item=(Type=Bone,Name="dyn_strap_04_r"),Transform=(Rotation=(X=0.0977,Y=0,Z=0,W=0.9952),Translation=(X=-1.758,Y=-1.634,Z=-2.1867),Scale3D=(X=1,Y=1,Z=1))))
    WeightMinimum=0
    Weight=&RigVMFunction_MathDoubleMul(B=@"DeltaRotation v4_12".outputWeight, A=@"DeltaRotation v4_17".outputWeight).Result
   )
   > &RigUnit_ModifyTransforms(
    ItemToModify=((Item=(Type=Bone,Name="dyn_strap_04_l"),Transform=(Rotation=(X=0.0977,Y=0,Z=0,W=0.9952),Translation=(X=1.758,Y=1.634,Z=2.1867),Scale3D=(X=1,Y=1,Z=1))))
    WeightMinimum=0
    Weight=&RigVMFunction_MathDoubleMul(A=@"DeltaRotation v4_26".outputWeight, B=@"DeltaRotation v4_6".outputWeight).Result
   )
   > &RigUnit_ModifyTransforms(
    ItemToModify="((Item=(Type=Bone,Name=\"dyn_strap_04_r\"),Transform=(Rotation=(X=0.0891,Y=-0.1425,Z=0.1662,W=0.9717),Translation=(X=-2.8103,Y=-1.8243,Z=-1.2275),Scale3D=(X=1,Y=1,Z=1))),(Item=(Type=Bone,Name=\"dyn_strap_05_r\"),Transform=(Rotation=(X=0,Y=-0.054,Z=0,W=-0.9985) ...(truncated)"
    WeightMinimum=0
    Weight=&RigVMFunction_MathDoubleMul(A=@"DeltaRotation v4_18".outputWeight, B=@"DeltaRotation v4_26".outputWeight).Result
   )
   > &RigUnit_ModifyTransforms(
    ItemToModify="((Item=(Type=Bone,Name=\"dyn_strap_04_l\"),Transform=(Rotation=(X=0.0891,Y=-0.1425,Z=0.1662,W=0.9717),Translation=(X=2.8103,Y=1.8243,Z=1.2275),Scale3D=(X=1,Y=1,Z=1))),(Item=(Type=Bone,Name=\"dyn_strap_05_l\"),Transform=(Rotation=(X=0,Y=-0.054,Z=0,W=-0.9985),Tr ...(truncated)"
    WeightMinimum=0
    Weight=&RigVMFunction_MathDoubleMul(B=@"DeltaRotation v4_12".outputWeight, A=@"DeltaRotation v4_9".outputWeight).Result
   )
   > &RigUnit_ModifyTransforms(
    ItemToModify=((Item=(Type=Bone,Name="dyn_strap_04_r"),Transform=(Rotation=(X=-0.0375,Y=0.0916,Z=0.3765,W=0.9211),Translation=(X=0.4092,Y=-2.5983,Z=0),Scale3D=(X=1,Y=1,Z=1))))
    WeightMinimum=0
    Weight=&RigVMFunction_MathDoubleMul(A=@"DeltaRotation v4_16".outputWeight, B=@"DeltaRotation v4_26".outputWeight).Result
   )
   > &RigUnit_ModifyTransforms(
    ItemToModify=((Item=(Type=Bone,Name="dyn_strap_04_l"),Transform=(Rotation=(X=-0.0375,Y=0.0916,Z=0.3765,W=0.9211),Translation=(X=-0.4092,Y=2.5983,Z=0),Scale3D=(X=1,Y=1,Z=1))))
    WeightMinimum=0
    Weight=&RigVMFunction_MathDoubleMul(B=@"DeltaRotation v4_12".outputWeight, A=@"DeltaRotation v4_2".outputWeight).Result
   )
   > &RigUnit_ModifyTransforms(
    ItemToModify="((Item=(Type=Bone,Name=\"dyn_strap_04_l\"),Transform=(Rotation=(X=-0.2332,Y=-0.4256,Z=0.2194,W=-0.8464),Translation=(X=-2.6795,Y=0,Z=1.5145),Scale3D=(X=1,Y=1,Z=1))),(Item=(Type=Bone,Name=\"dyn_strap_03_l\"),Transform=(Rotation=(X=-0.0096,Y=0.0894,Z=-0.0264,W=0 ...(truncated)"
    WeightMinimum=0
    Weight=&RigVMFunction_MathDoubleMul(B=@"DeltaRotation v4_1".outputWeight, A=@"DeltaRotation v4_12".outputWeight).Result
   )
   > &RigUnit_ModifyTransforms(
    ItemToModify="((Item=(Type=Bone,Name=\"dyn_strap_04_r\"),Transform=(Rotation=(X=-0.2332,Y=-0.4256,Z=0.2194,W=-0.8464),Translation=(X=2.6795,Y=0,Z=-1.5145),Scale3D=(X=1,Y=1,Z=1))),(Item=(Type=Bone,Name=\"dyn_strap_03_r\"),Transform=(Rotation=(X=-0.0096,Y=0.0894,Z=-0.0264,W=0 ...(truncated)"
    WeightMinimum=0
    Weight=&RigVMFunction_MathDoubleMul(B=@"DeltaRotation v4_15".outputWeight, A=@"DeltaRotation v4_26".outputWeight).Result
   )
   > &RigUnit_ParentConstraint(
    AdvancedSettings=(RotationOrderForFilter=XZY)
    Child=(Type=Bone,Name="MSTR_strap_06_l")
    Parents=((Item=(Type=Bone,Name="attach_backpack"),Weight=1))
    Weight=0.2
   )
   > &RigUnit_ParentConstraint(
    AdvancedSettings=(RotationOrderForFilter=XZY)
    Child=(Type=Bone,Name="dyn_strap_06_r")
    Parents=((Item=(Type=Bone,Name="attach_backpack"),Weight=1))
    Weight=0.2
   )
   > RigVMFunctionReturnNode()
  ]}
 RigVMFunctionLibrary {nodes [
   DeltaRotation v4_1 = RigVMCollapseNode(
    ExecuteContext=()
    Rotation Order=ZYX
    child=(Type=Bone,Name="None")
    outputMinMax=(X=0,Y=1)
    parent=(Type=Bone,Name="None")
    premult=1
    rotationMinMax=(X=0,Y=0)
    twistAxis=(X=1,Y=0,Z=0)
   )
   skirtSetup = RigVMCollapseNode(ExecuteContext=())
   sleeves = RigVMCollapseNode(ExecuteContext=())
   straps = RigVMCollapseNode(ExecuteContext=())
   sword = RigVMCollapseNode(ExecuteContext=())
  ]}
 RigVMFunction_Sequence_2_ContainedGraph {nodes [
   n1 = RigVMFunctionEntryNode()
   n2 = &RigVMFunction_Sequence(ExecuteContext=@n1.ExecuteContext)
   n3 = &RigVMFunction_Sequence(ExecuteContext=@n2.B)
   n4 = &RigVMFunction_Sequence(ExecuteContext=@n3.B)
   n5 = &RigVMFunction_Sequence(ExecuteContext=@n4.B)
   RigVMFunctionReturnNode(A=@n2.A, B=@n3.A, C=@n4.A, D=@n5.A, E=@n5.B)
  ]}
 RigVMModel {nodes [
   RigUnit_BeginExecution = &RigUnit_BeginExecution()
   /Script/FortniteGame.RigUnit_FortDeformRig(
    Bool=(bUpperArm=true,bElbow=true,bWrist=true,bGroin=true,bGlute=False,bKnee=true,bPec=true,bBicep=true,bLat=true)
    Curve="(bLUpperArmUpAlpha=False,bRUpperArmUpAlpha=False,bLUpperArmFrAlpha=False,bRUpperArmFrAlpha=False,bLUpperArmBkAlpha=False,bRUpperArmBkAlpha=False,bLWristUpAlpha=False,bRWristUpAlpha=False,bLWristDnAlpha=False,bRWristDnAlpha=False,bLKneeAlpha=False,bRKneeAlp ...(truncated)"
    Debug=(bUpperArmDebug=False,bWristDebug=False,bLatDebug=False,bGroinGluteDebug=False)
    Value="(UpperArmUp=(Rotation=(X=0.0,Y=-0.0,Z=0.0,W=1.0),Translation=(X=0.0,Y=0.0,Z=8.0),Scale3D=(X=0.70,Y=1.0,Z=1.0)),UpperArmFwdFr=(Rotation=(X=0.0,Y=-0.0,Z=0.0,W=1.0),Translation=(X=0.0,Y=10.0,Z=0.0),Scale3D=(X=0.50,Y=1.0,Z=1.0)),UpperArmFwdBk=(Rotation=(X=0.0, ...(truncated)"
    ExecutePin=@RigUnit_BeginExecution.ExecutePin
   ) {is=struct RigVMUnitNode}
   > &RigUnit_ParentConstraint(
    AdvancedSettings=(InterpolationType=Shortest,RotationOrderForFilter=ZYX)
    Child=(Type=Null,Name="lowerarm_r_null")
    Parents=((Item=(Type=Bone,Name="clavicle_r"),Weight=0.85),(Item=(Type=Bone,Name="upperarm_r"),Weight=0.15))
    Weight=1
   )
   > &RigUnit_ParentConstraint(
    AdvancedSettings=(InterpolationType=Shortest,RotationOrderForFilter=ZYX)
    Child=(Type=Null,Name="lowerarm_l_null")
    Parents=((Item=(Type=Bone,Name="clavicle_l"),Weight=0.85),(Item=(Type=Bone,Name="upperarm_l"),Weight=0.15))
    Weight=1
   )
   > skirtSetup = RigVMFunctionReferenceNode()
   > sleeves = RigVMFunctionReferenceNode()
   > sword = RigVMFunctionReferenceNode()
   > straps = RigVMFunctionReferenceNode()
  ]}
 RigVMModel Construction Event Graph {nodes [
   n1 = &RigUnit_PrepareForExecution()
   ItemArray = RigVMDispatchNode(Value="((Type=Bone,Name=\"thigh_l\"),(Type=Bone,Name=\"thigh_r\"),(Type=Bone,Name=\"clavicle_l\"),(Type=Bone,Name=\"clavicle_r\"),(Type=Bone,Name=\"upperarm_l\"),(Type=Bone,Name=\"upperarm_r\"),(Type=Bone,Name=\"lowerarm_l\"),(Type=Bone,Name=\"lowerarm_r\"),(Type=Bone,Name=\"spin ...(truncated)") {template="DISPATCH_RigVMDispatch_Constant(io Value)"}
   For_Each_1 = RigVMDispatchNode(
    Array=/Script/ControlRig.RigUnit_GetTransformItemArray(Space=GlobalSpace, bInitial=False, Items=@ItemArray.Value) {is=struct RigVMUnitNode}.Transforms
    ExecuteContext=@n1.ExecutePin
   ) {template="DISPATCH_RigVMDispatch_ArrayIterator(in Array,out Element,out Index,out Count,out Ratio)"}
   n2 = RigVMDispatchNode(Index=@For_Each_1.Index, Array=@ItemArray.Value) {template="DISPATCH_RigVMDispatch_ArrayGetAtIndex(in Array,in Index,out Element)"}
   Concat = /Script/RigVM.RigVMFunction_NameConcat(B=_null, A=@n2.Element.Name) {is=struct RigVMUnitNode, template="Concat::Execute(in A,in B,out Result)"}
   > &RigUnit_HierarchyAddNull(
    Parent=(Type=Bone,Name="spine_02")
    Space=GlobalSpace
    Name=@Concat.Result
    Transform=@For_Each_1.Element
    Parent.Name=/Script/ControlRig.RigUnit_HierarchyGetParent(Child=(Type=Bone,Name="None"), bDefaultParent=True, Child.Name=@n2.Element.Name) {is=struct RigVMUnitNode}.Parent.Name
   )
   /Script/ControlRig.RigUnit_SwitchParent(
    Child=(Type=Null,Name="upperarm_l_null")
    Mode=ParentItem
    Parent=(Type=Bone,Name="spine_05")
    bMaintainGlobal=True
    ExecutePin=@For_Each_1.Completed
   ) {is=struct RigVMUnitNode}
   > &RigUnit_SwitchParent(Child=(Type=Null,Name="upperarm_r_null"), Mode=ParentItem, Parent=(Type=Bone,Name="spine_05"), bMaintainGlobal=True)
   > &RigUnit_SwitchParent(Child=(Type=Null,Name="spine_04_null"), Mode=ParentItem, Parent=(Type=Bone,Name="pelvis"), bMaintainGlobal=True)
   > &RigUnit_SwitchParent(Child=(Type=Null,Name="spine_03_null"), Mode=ParentItem, Parent=(Type=Bone,Name="pelvis"), bMaintainGlobal=True)
   > &RigUnit_SwitchParent(Child=(Type=Null,Name="spine_05_null"), Mode=ParentItem, Parent=(Type=Bone,Name="pelvis"), bMaintainGlobal=True)
   > &RigUnit_SwitchParent(Child=(Type=Null,Name="spine_02_null"), Mode=ParentItem, Parent=(Type=Bone,Name="pelvis"), bMaintainGlobal=True)
  ]}
]
hierarchy [
 root {loc=<0 0 0>, rot=<0 0 0 1>}
 pelvis {parent=root, loc=<0 -0.7071 93.1191>, rot=<0.0127 0.707 -0.0127 -0.707>}
 spine_01 {parent=pelvis, loc=<6.7494 -0.6388 0>, rot=<0 0 0.034 -0.9994>}
 spine_02 {parent=spine_01, loc=<7.2518 -0.035 0>, rot=<0 0 0.0238 -0.9997>}
 spine_03 {parent=spine_02, loc=<6.0104 0 0>, rot=<0 0 0.0149 -0.9999>}
 spine_04 {parent=spine_03, loc=<8.1024 0 0>, rot=<0 0 -0.0132 -0.9999>}
 spine_05 {parent=spine_04, loc=<10.0146 0 0>, rot=<0 0 -0.0015 -1>}
 neck_01 {parent=spine_05, loc=<11.3138 0.1529 0>, rot=<0 0 -0.2003 -0.9797>}
 neck_02 {parent=neck_01, loc=<4.5627 0 0>, rot=<0 0 0.0504 -0.9987>}
 head {parent=neck_02, loc=<5.3914 0 0>, rot=<0 0 0.115 -0.9934>}
 attach_cape {parent=spine_05, loc=<9.017 -11.714 0>, rot=<0.0616 -0.7044 -0.0616 -0.7044>}
 cape {parent=attach_cape, type=Null, loc=<0 0 0>, rot=<0 0 0 1>}
 attach_backpack {parent=spine_05, loc=<-0.123 -13.473 0>, rot=<-0.0093 -0.707 0.0093 -0.707>}
 Backpack_BR {parent=attach_backpack, type=Null, loc=<0 0.089 -1.2297>, rot=<-0.0748 0 0 0.9972>}
 clavicle_l {parent=spine_05, loc=<4.3575 -0.1466 -2.3223>, rot=<0.0465 -0.7289 -0.009 -0.6829>}
 upperarm_l {parent=clavicle_l, loc=<15.5111 0 0>, rot=<0.0226 -0.3234 -0.0088 -0.9459>}
 lowerarm_l {parent=upperarm_l, loc=<26.6514 0 0>, rot=<0.0391 -0.0099 -0.2453 -0.9686>}
 hand_l {parent=lowerarm_l, loc=<22.3491 0 0>, rot=<-0.7081 -0.0507 0.0975 -0.6975>}
 thumb_01_l {parent=hand_l, loc=<1.3118 -0.3191 -2.459>, rot=<0.6546 -0.4042 0.0664 -0.6354>}
 thumb_02_l {parent=thumb_01_l, loc=<5.9336 -0.0448 0.0401>, rot=<-0.0037 0.0213 0.298 -0.9543>}
 thumb_03_l {parent=thumb_02_l, loc=<2.2223 0.3682 -0.0907>, rot=<0.0118 -0.013 0.0066 -0.9998>}
 index_metacarpal_l {parent=hand_l, loc=<1.6535 0.0183 -2.96>, rot=<0.0158 -0.0124 -0.0173 -0.9996>}
 index_01_l {parent=index_metacarpal_l, loc=<8.3008 0.0357 -0.0524>, rot=<-0.0044 0.0485 0.0924 -0.9945>}
 index_02_l {parent=index_01_l, loc=<3.3587 -0.3004 0.0452>, rot=<-0.002 0.0027 0.3225 -0.9466>}
 index_03_l {parent=index_02_l, loc=<2.5578 -0.123 -0.0252>, rot=<-0.002 0.0067 0.2558 -0.9667>}
 middle_metacarpal_l {parent=hand_l, loc=<2.1856 -0.1725 -0.5984>, rot=<0.006 0.0168 -0.0329 -0.9993>}
 middle_01_l {parent=middle_metacarpal_l, loc=<7.3275 -0.2447 0.0129>, rot=<-0.0208 0.0042 0.0858 -0.9961>}
 middle_02_l {parent=middle_01_l, loc=<4.0123 0.1401 0.1175>, rot=<-0.0218 0.0258 0.3512 -0.9357>}
 middle_03_l {parent=middle_02_l, loc=<2.6077 0.0534 0.1373>, rot=<0.0082 0.0188 0.1708 -0.9851>}
 ring_metacarpal_l {parent=hand_l, loc=<2.2074 -0.3625 1.3638>, rot=<-0.0284 0.0722 -0.0458 -0.9959>}
 ring_01_l {parent=ring_metacarpal_l, loc=<7.0295 -0.6197 -0.1446>, rot=<-0.0489 -0.0252 0.1892 -0.9804>}
 ring_02_l {parent=ring_01_l, loc=<3.6174 0.4647 -0.1027>, rot=<0 0 0.2806 -0.9598>}
 ring_03_l {parent=ring_02_l, loc=<2.4152 0.114 0.0199>, rot=<0 0 0.1189 -0.9929>}
 pinky_metacarpal_l {parent=hand_l, loc=<1.7616 -0.9753 3.3814>, rot=<-0.0397 0.0814 -0.069 -0.9935>}
 pinky_01_l {parent=pinky_metacarpal_l, loc=<7.1709 -0.4831 -0.1094>, rot=<-0.0654 -0.0035 0.2242 -0.9723>}
 pinky_02_l {parent=pinky_01_l, loc=<2.8274 0.0893 -0.1043>, rot=<0.0037 0.0011 0.2854 -0.9584>}
 pinky_03_l {parent=pinky_02_l, loc=<2.2046 0.0307 -0.0078>, rot=<-0.0002 0.0098 0.1445 -0.9895>}
 deform_wrist_dn_base_l {parent=hand_l, loc=<0 -2.2342 0>, rot=<0 0 0 -1>}
 deform_wrist_dn_l {parent=deform_wrist_dn_base_l, loc=<0 0 0>, rot=<0 0 0 1>}
 deform_wrist_up_base_l {parent=hand_l, loc=<0 2.2682 0>, rot=<0 0 0 -1>}
 deform_wrist_up_l {parent=deform_wrist_up_base_l, loc=<0 0 0>, rot=<0 0 0 1>}
 lowerarm_twist_02_l {parent=lowerarm_l, loc=<9.6199 0 0>, rot=<0 0 0 -1>}
 lowerarm_twist_01_l {parent=lowerarm_l, loc=<17.8549 0 0>, rot=<0 0 0 -1>}
 deform_elbow_in_base_l {parent=lowerarm_l, loc=<0.2112 5.259 0.8274>, rot=<0 0 0 -1>}
 deform_elbow_in_l {parent=deform_elbow_in_base_l, loc=<0 0 0>, rot=<0 0 0 1>}
 deform_elbow_out_base_l {parent=lowerarm_l, loc=<0 -4 0>, rot=<0 0 0 -1>}
 deform_elbow_out_l {parent=deform_elbow_out_base_l, loc=<0 0 0>, rot=<0 0 0 1>}
 upperarm_twist_01_l {parent=upperarm_l, loc=<0.0041 0 0>, rot=<0 0 0 -1>}
 deform_upperarm_bk_base_l {parent=upperarm_twist_01_l, loc=<1.0813 -7.2193 -1.9453>, rot=<0 0 0 1>}
 deform_upperarm_bk_l {parent=deform_upperarm_bk_base_l, loc=<0 0 0>, rot=<0 0 0 1>}
 deform_upperarm_fr_base_l {parent=upperarm_twist_01_l, loc=<-0.3691 7.0506 -2.765>, rot=<0 0 0 1>}
 deform_upperarm_fr_l {parent=deform_upperarm_fr_base_l, loc=<0 0 0>, rot=<0 0 0 1>}
 deform_upperarm_up_base_l {parent=upperarm_twist_01_l, loc=<-3.2216 -1.1391 3.3311>, rot=<0 0 0 1>}
 deform_upperarm_up_l {parent=deform_upperarm_up_base_l, loc=<0 0 0>, rot=<0 0 0 1>}
 upperarm_twist_02_l {parent=upperarm_l, loc=<13.7941 0 0>, rot=<0 0 0 -1>}
 MSTR_sleeve_l {parent=upperarm_l, loc=<10.3336 1.9068 -7.9378>, rot=<-0.0073 -0.0503 -0.0174 -0.9986>}
 dyn_sleeve_01_l {parent=MSTR_sleeve_l, loc=<0 0 0>, rot=<0 0 0 1>}
 dyn_sleeve_02_l {parent=dyn_sleeve_01_l, loc=<6.241 0 0>, rot=<-0.0011 -0.0457 -0.0584 -0.9972>}
 dyn_sleeve_03_l {parent=dyn_sleeve_02_l, loc=<8.1586 0 0>, rot=<-0.0021 -0.0328 0.0197 -0.9993>}
 dyn_sleeve_04_l {parent=dyn_sleeve_03_l, loc=<7.7993 0 0>, rot=<-0.0006 -0.0193 0.0113 -0.9998>}
 dyn_sleeve_05_l {parent=dyn_sleeve_04_l, loc=<6.8133 0 0>, rot=<0.0025 0.0513 -0.0287 -0.9983>}
 dyn_sleeve_06_l {parent=dyn_sleeve_05_l, loc=<3.5662 0 0>, rot=<0 0 0 1>}
 MSTR_strap_01_l {parent=clavicle_l, loc=<7.4608 -2.8843 8.9481>, rot=<-0.603 0.4484 -0.2844 -0.5953>}
 dyn_strap_01_l {parent=MSTR_strap_01_l, loc=<0 0 0>, rot=<0 0 0 1>}
 MSTR_strap_02_l {parent=clavicle_l, loc=<10.8992 5.0019 5.2868>, rot=<-0.2204 0.3612 -0.4894 -0.7626>}
 dyn_strap_02_l {parent=MSTR_strap_02_l, loc=<0 0 0>, rot=<0 0 0 1>}
 clavicle_r {parent=spine_05, loc=<4.3575 -0.1466 2.3223>, rot=<0.7289 0.0465 -0.6829 0.009>}
 upperarm_r {parent=clavicle_r, loc=<-15.5111 0 0>, rot=<0.0226 -0.3234 -0.0088 -0.9459>}
 lowerarm_r {parent=upperarm_r, loc=<-26.6514 0 0>, rot=<0.0391 -0.0099 -0.2453 -0.9686>}
 hand_r {parent=lowerarm_r, loc=<-22.3491 0 0>, rot=<-0.7081 -0.0507 0.0975 -0.6975>}
 thumb_01_r {parent=hand_r, loc=<-1.3118 0.3191 2.459>, rot=<0.6546 -0.4042 0.0664 -0.6354>}
 thumb_02_r {parent=thumb_01_r, loc=<-5.9336 0.0448 -0.0401>, rot=<-0.0037 0.0213 0.298 -0.9543>}
 thumb_03_r {parent=thumb_02_r, loc=<-2.2223 -0.3682 0.0907>, rot=<0.0118 -0.013 0.0066 -0.9998>}
 index_metacarpal_r {parent=hand_r, loc=<-1.6535 -0.0183 2.96>, rot=<0.0158 -0.0124 -0.0173 -0.9996>}
 index_01_r {parent=index_metacarpal_r, loc=<-8.3008 -0.0357 0.0524>, rot=<-0.0044 0.0485 0.0924 -0.9945>}
 index_02_r {parent=index_01_r, loc=<-3.3587 0.3004 -0.0452>, rot=<-0.002 0.0027 0.3225 -0.9466>}
 index_03_r {parent=index_02_r, loc=<-2.5578 0.123 0.0252>, rot=<-0.002 0.0067 0.2558 -0.9667>}
 middle_metacarpal_r {parent=hand_r, loc=<-2.1856 0.1725 0.5984>, rot=<0.006 0.0168 -0.0329 -0.9993>}
 middle_01_r {parent=middle_metacarpal_r, loc=<-7.3275 0.2447 -0.0129>, rot=<-0.0208 0.0042 0.0858 -0.9961>}
 middle_02_r {parent=middle_01_r, loc=<-4.0123 -0.1401 -0.1175>, rot=<-0.0218 0.0258 0.3512 -0.9357>}
 middle_03_r {parent=middle_02_r, loc=<-2.6077 -0.0534 -0.1373>, rot=<0.0082 0.0188 0.1708 -0.9851>}
 ring_metacarpal_r {parent=hand_r, loc=<-2.2074 0.3625 -1.3638>, rot=<-0.0284 0.0722 -0.0458 -0.9959>}
 ring_01_r {parent=ring_metacarpal_r, loc=<-7.0295 0.6197 0.1446>, rot=<-0.0489 -0.0252 0.1892 -0.9804>}
 ring_02_r {parent=ring_01_r, loc=<-3.6174 -0.4647 0.1027>, rot=<0 0 0.2806 -0.9598>}
 ring_03_r {parent=ring_02_r, loc=<-2.4152 -0.114 -0.0199>, rot=<0 0 0.1189 -0.9929>}
 pinky_metacarpal_r {parent=hand_r, loc=<-1.7616 0.9753 -3.3814>, rot=<-0.0397 0.0814 -0.069 -0.9935>}
 pinky_01_r {parent=pinky_metacarpal_r, loc=<-7.1709 0.4831 0.1094>, rot=<-0.0654 -0.0035 0.2242 -0.9723>}
 pinky_02_r {parent=pinky_01_r, loc=<-2.8274 -0.0893 0.1043>, rot=<0.0037 0.0011 0.2854 -0.9584>}
 pinky_03_r {parent=pinky_02_r, loc=<-2.2046 -0.0307 0.0078>, rot=<-0.0002 0.0098 0.1445 -0.9895>}
 deform_wrist_dn_base_r {parent=hand_r, loc=<-0.0004 2.2345 -0.0001>, rot=<0 0 0 -1>}
 deform_wrist_dn_r {parent=deform_wrist_dn_base_r, loc=<0 0 0>, rot=<0 0 0 1>}
 deform_wrist_up_base_r {parent=hand_r, loc=<0.0003 -2.2684 0.0001>, rot=<0 0 0 -1>}
 deform_wrist_up_r {parent=deform_wrist_up_base_r, loc=<0 0 0>, rot=<0 0 0 1>}
 lowerarm_twist_02_r {parent=lowerarm_r, loc=<-9.6199 0 0>, rot=<0 0 0 -1>}
 lowerarm_twist_01_r {parent=lowerarm_r, loc=<-17.8549 0 0>, rot=<0 0 0 -1>}
 deform_elbow_in_base_r {parent=lowerarm_r, loc=<-0.2111 -5.2591 -0.8275>, rot=<0 0 0 -1>}
 deform_elbow_in_r {parent=deform_elbow_in_base_r, loc=<-0.0001 0 0.0002>, rot=<0 0 0 1>}
 deform_elbow_out_base_r {parent=lowerarm_r, loc=<0 4 -0.0001>, rot=<0 0 0 -1>}
 deform_elbow_out_r {parent=deform_elbow_out_base_r, loc=<0 0 0>, rot=<0 0 0 1>}
 upperarm_twist_01_r {parent=upperarm_r, loc=<-0.0041 0 0>, rot=<0 0 0 -1>}
 deform_upperarm_bk_base_r {parent=upperarm_twist_01_r, loc=<-1.0814 7.2192 1.9455>, rot=<0 0 0 1>}
 deform_upperarm_bk_r {parent=deform_upperarm_bk_base_r, loc=<0 0 0>, rot=<0 0 0 1>}
 deform_upperarm_fr_base_r {parent=upperarm_twist_01_r, loc=<0.3689 -7.0506 2.7652>, rot=<0 0 0 1>}
 deform_upperarm_fr_r {parent=deform_upperarm_fr_base_r, loc=<0 0 0>, rot=<0 0 0 1>}
 deform_upperarm_up_base_r {parent=upperarm_twist_01_r, loc=<3.2214 1.1391 -3.3308>, rot=<0 0 0 1>}
 deform_upperarm_up_r {parent=deform_upperarm_up_base_r, loc=<0 0 0>, rot=<0 0 0 -1>}
 upperarm_twist_02_r {parent=upperarm_r, loc=<-13.7941 0 0>, rot=<0 0 0 -1>}
 MSTR_sleeve_r {parent=upperarm_r, loc=<-10.3338 -1.9068 7.938>, rot=<-0.0073 -0.0503 -0.0174 -0.9986>}
 dyn_sleeve_01_r {parent=MSTR_sleeve_r, loc=<0 0 0>, rot=<0 0 0 1>}
 dyn_sleeve_02_r {parent=dyn_sleeve_01_r, loc=<-6.2408 0 -0.0002>, rot=<-0.0011 -0.0457 -0.0584 -0.9972>}
 dyn_sleeve_03_r {parent=dyn_sleeve_02_r, loc=<-8.1583 0 -0.0003>, rot=<-0.0021 -0.0328 0.0197 -0.9993>}
 dyn_sleeve_04_r {parent=dyn_sleeve_03_r, loc=<-7.7999 0.0001 0.0005>, rot=<-0.0006 -0.0193 0.0113 -0.9998>}
 dyn_sleeve_05_r {parent=dyn_sleeve_04_r, loc=<-6.8131 0 -0.0001>, rot=<0.0025 0.0513 -0.0287 -0.9983>}
 dyn_sleeve_06_r {parent=dyn_sleeve_05_r, loc=<-3.5662 0 0>, rot=<0 0 0 1>}
 MSTR_strap_02_r {parent=clavicle_r, loc=<-10.8993 -5.0019 -5.2864>, rot=<-0.2204 0.3612 -0.4894 -0.7626>}
 dyn_strap_02_r {parent=MSTR_strap_02_r, loc=<0 0 0>, rot=<0 0 0 1>}
 MSTR_strap_01_r {parent=clavicle_r, loc=<-7.4608 2.8843 -8.948>, rot=<-0.603 0.4484 -0.2844 -0.5953>}
 dyn_strap_01_r {parent=MSTR_strap_01_r, loc=<0 0 0>, rot=<0 0 0 1>}
 MSTR_strap_03_l {parent=spine_05, loc=<0.3814 8.3308 -16.598>, rot=<-0.3642 -0.5183 -0.3222 -0.7035>}
 dyn_strap_03_l {parent=MSTR_strap_03_l, loc=<0 0 0>, rot=<0 0 0 1>}
 MSTR_strap_03_r {parent=spine_05, loc=<0.3816 8.3308 16.598>, rot=<-0.5183 0.3642 0.7035 -0.3222>}
 dyn_strap_03_r {parent=MSTR_strap_03_r, loc=<0 0 0>, rot=<0 0 0 1>}
 deform_pec_aim_base_l {parent=spine_04, loc=<7.8732 12.6388 -2.4123>, rot=<0.0102 -0.4602 0.3979 -0.7936>}
 deform_pec_aim_l {parent=deform_pec_aim_base_l, loc=<0 0 0>, rot=<0 0 0 1>}
 deform_pec_base_l {parent=deform_pec_aim_l, loc=<10.766 0 0>, rot=<0 0 0 1>}
 deform_pec_l {parent=deform_pec_base_l, loc=<0.055 0 0>, rot=<0 0 0 1>}
 deform_pec_aim_base_r {parent=spine_04, loc=<7.8727 12.6388 2.4123>, rot=<-0.4602 -0.0102 0.7936 0.3979>}
 deform_pec_aim_r {parent=deform_pec_aim_base_r, loc=<0 0 0>, rot=<0 0 0 1>}
 deform_pec_base_r {parent=deform_pec_aim_r, loc=<-10.7662 -0.0006 -0.0007>, rot=<0 0 0 1>}
 deform_pec_r {parent=deform_pec_base_r, loc=<0 0 0>, rot=<0 0 0 1>}
 MSTR_strap_04_l {parent=spine_04, loc=<1.0876 6.9188 -17.6572>, rot=<-0.057 -0.6592 -0.2977 -0.6881>}
 dyn_strap_04_l {parent=MSTR_strap_04_l, loc=<0 0 0>, rot=<0 0 0 1>}
 MSTR_strap_04_r {parent=spine_04, loc=<1.0877 6.9188 17.6572>, rot=<-0.6592 0.057 0.6881 -0.2977>}
 dyn_strap_04_r {parent=MSTR_strap_04_r, loc=<0 0 0>, rot=<0 0 0 1>}
 MSTR_strap_05_l {parent=spine_03, loc=<1.3897 -1.339 -17.4626>, rot=<0.4345 -0.6106 -0.2822 -0.5989>}
 dyn_strap_05_l {parent=MSTR_strap_05_l, loc=<0 0 0>, rot=<0 0 0 1>}
 MSTR_strap_05_r {parent=spine_03, loc=<1.3895 -1.3391 17.4626>, rot=<0.6106 0.4345 -0.5989 0.2822>}
 dyn_strap_05_r {parent=MSTR_strap_05_r, loc=<0 0 0>, rot=<0 0 0 1>}
 MSTR_frontShirt {parent=spine_02, loc=<0 14.1683 0>, rot=<0 0 -0.7071 -0.7071>}
 dyn_frontShirt {parent=MSTR_frontShirt, loc=<0 0 0>, rot=<0 0 0 1>}
 MSTR_sideShirt_l {parent=spine_02, loc=<2.0002 0 -15.3967>, rot=<0.5 -0.5 -0.5 -0.5>}
 dyn_sideShirt_l {parent=MSTR_sideShirt_l, loc=<0 0 0>, rot=<0 0 0 1>}
 MSTR_backShirt {parent=spine_02, loc=<1.6596 -10.0582 0>, rot=<0.7071 -0.7071 0 0>}
 dyn_backShirt {parent=MSTR_backShirt, loc=<0 0 0>, rot=<0 0 0 1>}
 MSTR_sideShirt_r {parent=spine_02, loc=<2.0005 0 15.3967>, rot=<-0.5 0.5 -0.5 -0.5>}
 dyn_sideShirt_r {parent=MSTR_sideShirt_r, loc=<0 0 0>, rot=<0 0 0 1>}
 MSTR_strap_06_l {parent=spine_02, loc=<2.8962 -10.3727 -14.5773>, rot=<0.5705 -0.6136 -0.1857 -0.5133>}
 dyn_strap_06_l {parent=MSTR_strap_06_l, loc=<0 0 0>, rot=<0 0 0 1>}
 MSTR_strap_06_r {parent=spine_02, loc=<2.8959 -10.3727 14.5773>, rot=<0.6136 0.5705 -0.5133 0.1857>}
 dyn_strap_06_r {parent=MSTR_strap_06_r, loc=<0 0 0>, rot=<0 0 0 1>}
 MSTR_sword {parent=spine_01, loc=<3.1091 7.3318 -16.8539>, rot=<0.0296 0.1249 -0.2594 -0.9572>}
 dyn_sword {parent=MSTR_sword, loc=<0 0 0>, rot=<0 0 0 1>}
 thigh_l {parent=pelvis, loc=<-2.5693 -0.0109 -7.7501>, rot=<0.0012 0.0572 0.0155 -0.9982>}
 calf_l {parent=thigh_l, loc=<-39.0484 0 0>, rot=<0.0006 -0.0161 -0.0391 -0.9991>}
 foot_l {parent=calf_l, loc=<-42.561 0.1077 0.4292>, rot=<0.0418 -0.0455 0.0343 -0.9975>}
 ball_l {parent=foot_l, loc=<-6.1599 13.8264 0.6806>, rot=<0.0001 -0.0001 -0.7063 -0.7079>}
 calf_twist_02_l {parent=calf_l, loc=<-18.8064 -0.0967 0.0397>, rot=<0.0237 0.0095 -0.0035 -0.9997>}
 MSTR_calfStrapA_l {parent=calf_twist_02_l, loc=<-4.5711 -4.4415 -9.3809>, rot=<0.4013 -0.6445 0.3254 -0.5637>}
 dyn_calfStrapA_01_l {parent=MSTR_calfStrapA_l, loc=<0 0 0>, rot=<0 0 0 1>}
 MSTR_calfStrapB_l {parent=calf_twist_02_l, loc=<-6.172 -3.8767 -8.8723>, rot=<0.23 -0.7683 0.4756 -0.3614>}
 dyn_calfStrapB_01_l {parent=MSTR_calfStrapB_l, loc=<0 0 0>, rot=<0 0 0 1>}
 calf_twist_01_l {parent=calf_l, loc=<-35.813 -0.0967 0.0397>, rot=<0.0237 0.0095 -0.0035 -0.9997>}
 MSTR_calfStrapC_l {parent=calf_twist_01_l, loc=<3.1083 -3.6122 -7.9004>, rot=<0.4172 -0.6231 0.3018 -0.5888>}
 dyn_calfStrapC_01_l {parent=MSTR_calfStrapC_l, loc=<0 0 0>, rot=<0 0 0 1>}
 MSTR_calfStrapD_l {parent=calf_twist_01_l, loc=<1.5404 -3.2698 -7.5004>, rot=<0.1875 -0.7572 0.5258 -0.3391>}
 dyn_calfStrapD_01_l {parent=MSTR_calfStrapD_l, loc=<0 0 0>, rot=<0 0 0 1>}
 deform_knee_in_base_l {parent=calf_l, loc=<1.298 -5.2957 -0.0496>, rot=<0 0 0 -1>}
 deform_knee_in_l {parent=deform_knee_in_base_l, loc=<0 0 0>, rot=<0 0 0 1>}
 deform_knee_out_base_l {parent=calf_l, loc=<0 4 0>, rot=<0 0 0 -1>}
 deform_knee_out_l {parent=deform_knee_out_base_l, loc=<0 0 0>, rot=<0 0 0 1>}
 thigh_twist_01_l {parent=thigh_l, loc=<-21.6303 0 0>, rot=<0.0013 0 0 -1>}
 deform_groin_base_l {parent=thigh_l, loc=<-2.1908 8.1142 -1.7442>, rot=<0 0 0 -1>}
 deform_groin_l {parent=deform_groin_base_l, loc=<0 0 0>, rot=<0 0 0 1>}
 thigh_r {parent=pelvis, loc=<-2.5693 -0.0109 7.7501>, rot=<0.0572 -0.0012 0.9982 0.0155>}
 calf_r {parent=thigh_r, loc=<39.0484 0 0>, rot=<0.0006 -0.0161 -0.0391 -0.9991>}
 foot_r {parent=calf_r, loc=<42.561 -0.1077 -0.4292>, rot=<0.0418 -0.0455 0.0343 -0.9975>}
 ball_r {parent=foot_r, loc=<6.1599 -13.8264 -0.6806>, rot=<0.0001 -0.0001 -0.7063 -0.7079>}
 calf_twist_02_r {parent=calf_r, loc=<18.8064 0.0967 -0.0397>, rot=<0.0237 0.0095 -0.0035 -0.9997>}
 MSTR_calfStrapA_r {parent=calf_twist_02_r, loc=<4.5711 4.4415 9.3809>, rot=<0.4013 -0.6445 0.3254 -0.5637>}
 dyn_calfStrapA_01_r {parent=MSTR_calfStrapA_r, loc=<0 0 0>, rot=<0 0 0 1>}
 MSTR_calfStrapB_r {parent=calf_twist_02_r, loc=<6.1721 3.8767 8.8723>, rot=<0.23 -0.7683 0.4756 -0.3614>}
 dyn_calfStrapB_01_r {parent=MSTR_calfStrapB_r, loc=<0 0 0>, rot=<0 0 0 1>}
 calf_twist_01_r {parent=calf_r, loc=<35.813 0.0967 -0.0397>, rot=<0.0237 0.0095 -0.0035 -0.9997>}
 MSTR_calfStrapC_r {parent=calf_twist_01_r, loc=<-3.1083 3.6122 7.9004>, rot=<0.4172 -0.6231 0.3018 -0.5888>}
 dyn_calfStrapC_01_r {parent=MSTR_calfStrapC_r, loc=<0 0 0>, rot=<0 0 0 1>}
 MSTR_calfStrapD_r {parent=calf_twist_01_r, loc=<-1.5404 3.2698 7.5005>, rot=<0.1875 -0.7572 0.5258 -0.3391>}
 dyn_calfStrapD_01_r {parent=MSTR_calfStrapD_r, loc=<0 0 0>, rot=<0 0 0 1>}
 deform_knee_in_base_r {parent=calf_r, loc=<-1.298 5.2957 0.0496>, rot=<0 0 0 -1>}
 deform_knee_in_r {parent=deform_knee_in_base_r, loc=<0 0 0>, rot=<0 0 0 1>}
 deform_knee_out_base_r {parent=calf_r, loc=<0 -4 0>, rot=<0 0 0 -1>}
 deform_knee_out_r {parent=deform_knee_out_base_r, loc=<0 0 0>, rot=<0 0 0 1>}
 thigh_twist_01_r {parent=thigh_r, loc=<21.6303 0 0>, rot=<0.0013 0 0 -1>}
 deform_groin_base_r {parent=thigh_r, loc=<2.1908 -8.1142 1.7442>, rot=<0 0 0 -1>}
 deform_groin_r {parent=deform_groin_base_r, loc=<0 0 0>, rot=<0 0 0 1>}
 deform_glute_base_l {parent=pelvis, loc=<-8.8064 -8.8222 -8.4287>, rot=<0.0275 0.0501 -0.4472 -0.8926>}
 deform_glute_l {parent=deform_glute_base_l, loc=<0 0 0>, rot=<0 0 0 1>}
 deform_glute_base_r {parent=pelvis, loc=<-8.8062 -8.8222 8.4287>, rot=<0.0501 -0.0275 0.8926 -0.4472>}
 deform_glute_r {parent=deform_glute_base_r, loc=<0 0 0>, rot=<0 0 0 1>}
 MSTR_frontSkirt {parent=pelvis, loc=<-0.7438 12.142 0.0311>, rot=<-0.0653 -0.7123 -0.6972 -0.0476>}
 dyn_frontSkirt_01 {parent=MSTR_frontSkirt, loc=<0 0 0>, rot=<0 0 0 1>}
 dyn_frontSkirt_02 {parent=dyn_frontSkirt_01, loc=<9.7239 0 0>, rot=<-0.0006 -0.0226 -0.0056 -0.9997>}
 MSTR_backSkirt {parent=pelvis, loc=<-0.1335 -14.1119 -0.0399>, rot=<0.0482 -0.6966 0.7129 -0.0648>}
 dyn_backSkirt_01 {parent=MSTR_backSkirt, loc=<0 0 0>, rot=<0 0 0 1>}
 dyn_backSkirt_02 {parent=dyn_backSkirt_01, loc=<9.7239 0 0>, rot=<-0.0006 -0.0226 -0.0056 -0.9997>}
 MSTR_backSkirt_l {parent=pelvis, loc=<0.1795 -10.7525 -11.176>, rot=<-0.01 -0.8308 0.546 -0.1075>}
 dyn_backSkirt_01_l {parent=MSTR_backSkirt_l, loc=<0 0 0>, rot=<0 0 0 1>}
 dyn_backSkirt_02_l {parent=dyn_backSkirt_01_l, loc=<9.1921 0 0>, rot=<0.0011 -0.024 0.0108 -0.9997>}
 MSTR_frontSkirt_l {parent=pelvis, loc=<-0.0422 9.2116 -11.571>, rot=<0.006 -0.8582 -0.5055 -0.0891>}
 dyn_frontSkirt_01_l {parent=MSTR_frontSkirt_l, loc=<0 0 0>, rot=<-0.1106 0 0 -0.9939>}
 dyn_frontSkirt_02_l {parent=dyn_frontSkirt_01_l, loc=<9.1324 0 0>, rot=<0.0004 0.0056 0.0071 -1>}
 MSTR_sideSkirt_l {parent=pelvis, loc=<0.3345 -1.1585 -16.33>, rot=<-0.0051 -0.9888 0.0602 -0.1368>}
 dyn_sideSkirt_01_l {parent=MSTR_sideSkirt_l, loc=<0 0 0>, rot=<0 0 0 1>}
 dyn_sideSkirt_02_l {parent=dyn_sideSkirt_01_l, loc=<9.3419 0 0>, rot=<-0.1108 -0.0213 -0.0011 -0.9936>}
 MSTR_frontSkirt_r {parent=pelvis, loc=<-0.1632 9.2732 11.509>, rot=<0.0974 0.5422 0.8346 0.0005>}
 dyn_frontSkirt_01_r {parent=MSTR_frontSkirt_r, loc=<0 0 0>, rot=<0 0 0 1>}
 dyn_frontSkirt_02_r {parent=dyn_frontSkirt_01_r, loc=<9.1921 0 0>, rot=<0.0011 -0.024 0.0108 -0.9997>}
 MSTR_sideSkirt_r {parent=pelvis, loc=<0.3345 -1.1585 16.33>, rot=<0.1368 -0.0602 0.9888 0.0051>}
 dyn_sideSkirt_01_r {parent=MSTR_sideSkirt_r, loc=<0 0 0>, rot=<0 0 0 1>}
 dyn_sideSkirt_02_r {parent=dyn_sideSkirt_01_r, loc=<9.3419 0 0>, rot=<0.0002 -0.021 0.0035 -0.9998>}
 MSTR_backSkirt_r {parent=pelvis, loc=<-0.044 -10.7299 11.2138>, rot=<0.1013 -0.5094 0.8543 0.0187>}
 dyn_backSkirt_01_r {parent=MSTR_backSkirt_r, loc=<0 0 0>, rot=<0 0 0 1>}
 dyn_backSkirt_02_r {parent=dyn_backSkirt_01_r, loc=<9.1324 0 0>, rot=<0.0004 0.0056 0.0071 -1>}
]
variables [
 crouching {type=bool, default=False}
 melee {type=bool, default=False}
 passenger {type=bool, default=False}
 straps_enabled {type=bool, default=False}
]
