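# Physics asset for the NeonFriction Await medium male player body cosmetic, set up for secondary motion on shirt, sleeves and calf straps.
# Capsules on pelvis, spine and arms form the core that MSTR anchor bodies constrain to, with calf straps tied to pelvis as no leg bodies exist.
# Anchors carry limited dyn bodies, five-link chains for sleeves and single links elsewhere.
# Anchor swing and twist are free, except the right side shirt's.
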PhysicsAsset /BRCosmetics/Characters/Player/Male/Medium/Bodies/M_MED_NeonFriction_Await/Meshes/M_MED_NeonFriction_Await_Physics
bodies {
 MSTR_backShirt [{shapeType=Capsule, length=0.505, radius=0.505, shapeName=MSTR_backShirt_capsule, rotation=<rot 90 0 0>}]
 MSTR_calfStrapA_l [{shapeType=Capsule, length=0.505, radius=0.505, shapeName=MSTR_calfStrapA_l_capsule, rotation=<rot 0 -90 0>}]
 MSTR_calfStrapA_r [{shapeType=Capsule, length=0.505, radius=0.505, shapeName=MSTR_calfStrapA_r_capsule, rotation=<rot 0 -90 0>}]
 MSTR_calfStrapB_l [{shapeType=Capsule, length=0.505, radius=0.505, shapeName=MSTR_calfStrapB_l_capsule, rotation=<rot 0 -90 0>}]
 MSTR_calfStrapB_r [{shapeType=Capsule, length=0.505, radius=0.505, shapeName=MSTR_calfStrapB_r_capsule, rotation=<rot 0 -90 0>}]
 MSTR_calfStrapC_l [{shapeType=Capsule, length=0.505, radius=0.505, shapeName=MSTR_calfStrapC_l_capsule, rotation=<rot 0 -90 0>}]
 MSTR_calfStrapC_r [{shapeType=Capsule, length=0.505, radius=0.505, shapeName=MSTR_calfStrapC_r_capsule, rotation=<rot 0 -90 0>}]
 MSTR_calfStrapD_l [{shapeType=Capsule, length=0.505, radius=0.505, shapeName=MSTR_calfStrapD_l_capsule, rotation=<rot 0 -90 0>}]
 MSTR_calfStrapD_r [{shapeType=Capsule, length=0.505, radius=0.505, shapeName=MSTR_calfStrapD_r_capsule, rotation=<rot 0 -90 0>}]
 MSTR_frontShirt [{shapeType=Capsule, length=0.505, radius=0.505, shapeName=MSTR_frontShirt_capsule, rotation=<rot 90 0 0>}]
 MSTR_sideShirt_l [{shapeType=Capsule, length=0.505, radius=0.505, shapeName=MSTR_sideShirt_l_capsule, rotation=<rot 90 0 0>}]
 MSTR_sideShirt_r [{shapeType=Capsule, length=0.505, radius=0.505, shapeName=MSTR_sideShirt_r_capsule, rotation=<rot 90 0 0>}]
 MSTR_sleeve_l [{shapeType=Capsule, length=5, radius=1, shapeName=MSTR_sleeve_l_capsule, rotation=<rot 0 0 90>}]
 MSTR_sleeve_r [{shapeType=Capsule, length=5, radius=1, shapeName=MSTR_sleeve_r_capsule, rotation=<rot 0 0 90>}]
 dyn_backShirt [{shapeType=Capsule, length=13.5, radius=3, shapeName=dyn_backShirt_capsule, center=<5.0314 -0.7781 -0.1299>, rotation=<rot 0 90 0>}]
 dyn_calfStrapA_01_l [{shapeType=Capsule, length=2, radius=1, shapeName=dyn_calfStrapA_01_l_capsule, center=<3.4203 0 0>, rotation=<rot 90 0 0>}]
 dyn_calfStrapA_01_r [{shapeType=Capsule, length=2, radius=1, shapeName=dyn_calfStrapA_01_r_capsule, center=<-3.4203 0 0>, rotation=<rot 90 0 0>}]
 dyn_calfStrapB_01_l [{shapeType=Capsule, length=2, radius=1, shapeName=dyn_calfStrapB_01_l_capsule, center=<3.4203 0 0>, rotation=<rot 90 0 0>}]
 dyn_calfStrapB_01_r [{shapeType=Capsule, length=2, radius=1, shapeName=dyn_calfStrapB_01_r_capsule, center=<-3.4203 0 0>, rotation=<rot 90 0 0>}]
 dyn_calfStrapC_01_l [{shapeType=Capsule, length=2, radius=1, shapeName=dyn_calfStrapC_01_l_capsule, center=<3.4203 0 0>, rotation=<rot 90 0 0>}]
 dyn_calfStrapC_01_r [{shapeType=Capsule, length=2, radius=1, shapeName=dyn_calfStrapC_01_r_capsule, center=<-3.4203 0 0>, rotation=<rot 90 0 0>}]
 dyn_calfStrapD_01_l [{shapeType=Capsule, length=2, radius=1, shapeName=dyn_calfStrapD_01_l_capsule, center=<3.4203 0 0>, rotation=<rot 90 0 0>}]
 dyn_calfStrapD_01_r [{shapeType=Capsule, length=2, radius=1, shapeName=dyn_calfStrapD_01_r_capsule, center=<-3.4203 0 0>, rotation=<rot 90 0 0>}]
 dyn_frontShirt [{shapeType=Capsule, length=13.5, radius=3, shapeName=dyn_frontShirt_capsule, center=<1.0377 0 0>, rotation=<rot 0 90 0>}]
 dyn_sideShirt_l [{shapeType=Capsule, length=13.5, radius=3, shapeName=dyn_sideShirt_l_capsule, center=<4.5173 0 0>, rotation=<rot 0 90 0>}]
 dyn_sideShirt_r [{shapeType=Capsule, length=13.5, radius=3, shapeName=dyn_sideShirt_r_capsule, center=<4.5173 0 0>, rotation=<rot 0 90 0>}]
 dyn_sleeve_01_l [{shapeType=Capsule, length=10, radius=3, shapeName=dyn_sleeve_01_l_capsule, center=<0 0 -2.269>, rotation=<rot 0 0 90>}]
 dyn_sleeve_01_r [{shapeType=Capsule, length=10, radius=3, shapeName=dyn_sleeve_01_r_capsule, center=<0 0 2.269>, rotation=<rot 0 0 90>}]
 dyn_sleeve_02_l [{shapeType=Capsule, length=10, radius=3, shapeName=dyn_sleeve_02_l_capsule, center=<1.0523 0 -1.4869>, rotation=<rot 0 0 90>}]
 dyn_sleeve_02_r [{shapeType=Capsule, length=10, radius=3, shapeName=dyn_sleeve_02_r_capsule, center=<-1.0523 0 1.4869>, rotation=<rot 0 0 90>}]
 dyn_sleeve_03_l [{shapeType=Capsule, length=10, radius=3, shapeName=dyn_sleeve_03_l_capsule, center=<0 0 -1.1309>, rotation=<rot 0 0 90>}]
 dyn_sleeve_03_r [{shapeType=Capsule, length=10, radius=3, shapeName=dyn_sleeve_03_r_capsule, center=<0 0 1.1309>, rotation=<rot 0 0 90>}]
 dyn_sleeve_04_l [{shapeType=Capsule, length=7, radius=4, shapeName=dyn_sleeve_04_l_capsule, center=<-0.6641 0 0>, rotation=<rot 0 0 90>}]
 dyn_sleeve_04_r [{shapeType=Capsule, length=7, radius=4, shapeName=dyn_sleeve_04_r_capsule, center=<0.6641 0 0>, rotation=<rot 0 0 90>}]
 dyn_sleeve_05_l [{shapeType=Capsule, length=5, radius=5, shapeName=dyn_sleeve_05_l_capsule, center=<-0.0523 0 0>, rotation=<rot 0 0 90>}]
 dyn_sleeve_05_r [{shapeType=Capsule, length=5, radius=5, shapeName=dyn_sleeve_05_r_capsule, rotation=<rot 0 0 90>}]
 lowerarm_l [{shapeType=Capsule, length=15.9883, radius=4.9192, center=<12.6771 0.1352 -1.191>, rotation=<rot 90 0 0>}]
 lowerarm_r [{shapeType=Capsule, length=16.0047, radius=4.9203, center=<-12.6658 -0.145 1.1896>, rotation=<rot 90 0 0>}]
 pelvis [{shapeType=Capsule, length=12.3042, radius=15.1958, center=<-7.9872 -0.2232 0.1569>}]
 spine_02 [{shapeType=Capsule, length=12.5099, radius=10.4133, center=<0 0.7331 0>}]
 spine_03 [{shapeType=Capsule, length=10.2509, radius=11.965}]
 upperarm_l [{shapeType=Capsule, length=20.7001, radius=5.8699, center=<11.855 0.1686 -1.1056>, rotation=<rot 90 0 0>}]
 upperarm_r [{shapeType=Capsule, length=20.6882, radius=5.8436, center=<-11.8274 -0.1116 1.1281>, rotation=<rot 90 0 0>}]
}
constraints [
 {bone1Name=spine_02, bone2Name=pelvis, swing1LimitDegrees=45, swing2LimitDegrees=45, twistLimitDegrees=45}
 {bone1Name=spine_03, bone2Name=spine_02, swing1LimitDegrees=45, swing2LimitDegrees=45, twistLimitDegrees=45}
 {bone1Name=upperarm_l, bone2Name=spine_03, swing1LimitDegrees=45, swing2LimitDegrees=45, twistLimitDegrees=45}
 {bone1Name=lowerarm_l, bone2Name=upperarm_l, swing1LimitDegrees=45, swing2LimitDegrees=45, twistLimitDegrees=45}
 {bone1Name=upperarm_r, bone2Name=spine_03, swing1LimitDegrees=45, swing2LimitDegrees=45, twistLimitDegrees=45}
 {bone1Name=lowerarm_r, bone2Name=upperarm_r, swing1LimitDegrees=45, swing2LimitDegrees=45, twistLimitDegrees=45}
 {
  bone1Name=MSTR_backShirt
  bone2Name=spine_02
  swing1LimitDegrees=45
  swing1Motion=Free
  swing2LimitDegrees=45
  swing2Motion=Free
  twistLimitDegrees=45
  twistMotion=Free
 }
 {bone1Name=dyn_backShirt, bone2Name=MSTR_backShirt, swing1LimitDegrees=45, swing2LimitDegrees=45, twistLimitDegrees=5}
 {
  bone1Name=MSTR_frontShirt
  bone2Name=spine_02
  swing1LimitDegrees=45
  swing1Motion=Free
  swing2LimitDegrees=45
  swing2Motion=Free
  twistLimitDegrees=45
  twistMotion=Free
 }
 {bone1Name=dyn_frontShirt, bone2Name=MSTR_frontShirt, swing1LimitDegrees=45, swing2LimitDegrees=45, twistLimitDegrees=5}
 {
  bone1Name=MSTR_sideShirt_l
  bone2Name=spine_02
  swing1LimitDegrees=45
  swing1Motion=Free
  swing2LimitDegrees=45
  swing2Motion=Free
  twistLimitDegrees=45
  twistMotion=Free
 }
 {bone1Name=dyn_sideShirt_l, bone2Name=MSTR_sideShirt_l, swing1LimitDegrees=45, swing2LimitDegrees=45, twistLimitDegrees=5}
 {bone1Name=MSTR_sideShirt_r, bone2Name=spine_02, swing1LimitDegrees=10, swing2LimitDegrees=10, twistLimitDegrees=5}
 {bone1Name=dyn_sideShirt_r, bone2Name=MSTR_sideShirt_r, swing1LimitDegrees=45, swing2LimitDegrees=45, twistLimitDegrees=5}
 {
  bone1Name=MSTR_sleeve_l
  bone2Name=upperarm_l
  swing1LimitDegrees=45
  swing1Motion=Free
  swing2LimitDegrees=45
  swing2Motion=Free
  twistLimitDegrees=45
  twistMotion=Free
 }
 {bone1Name=dyn_sleeve_01_l, bone2Name=MSTR_sleeve_l, swing1LimitDegrees=15, swing2LimitDegrees=5, twistLimitDegrees=5}
 {bone1Name=dyn_sleeve_02_l, bone2Name=dyn_sleeve_01_l, swing1LimitDegrees=15, swing2LimitDegrees=5, twistLimitDegrees=5}
 {bone1Name=dyn_sleeve_03_l, bone2Name=dyn_sleeve_02_l, swing1LimitDegrees=15, swing2LimitDegrees=5, twistLimitDegrees=5}
 {bone1Name=dyn_sleeve_04_l, bone2Name=dyn_sleeve_03_l, swing1LimitDegrees=20, swing2LimitDegrees=20, twistLimitDegrees=5}
 {bone1Name=dyn_sleeve_05_l, bone2Name=dyn_sleeve_04_l, swing1LimitDegrees=45, swing2LimitDegrees=70, twistLimitDegrees=20}
 {
  bone1Name=MSTR_sleeve_r
  bone2Name=upperarm_r
  swing1LimitDegrees=45
  swing1Motion=Free
  swing2LimitDegrees=45
  swing2Motion=Free
  twistLimitDegrees=45
  twistMotion=Free
 }
 {bone1Name=dyn_sleeve_01_r, bone2Name=MSTR_sleeve_r, swing1LimitDegrees=15, swing2LimitDegrees=5, twistLimitDegrees=5}
 {bone1Name=dyn_sleeve_02_r, bone2Name=dyn_sleeve_01_r, swing1LimitDegrees=15, swing2LimitDegrees=5, twistLimitDegrees=5}
 {bone1Name=dyn_sleeve_03_r, bone2Name=dyn_sleeve_02_r, swing1LimitDegrees=15, swing2LimitDegrees=5, twistLimitDegrees=5}
 {bone1Name=dyn_sleeve_04_r, bone2Name=dyn_sleeve_03_r, swing1LimitDegrees=20, swing2LimitDegrees=20, twistLimitDegrees=5}
 {bone1Name=dyn_sleeve_05_r, bone2Name=dyn_sleeve_04_r, swing1LimitDegrees=45, swing2LimitDegrees=70, twistLimitDegrees=20}
 {
  bone1Name=MSTR_calfStrapC_l
  bone2Name=pelvis
  swing1LimitDegrees=45
  swing1Motion=Free
  swing2LimitDegrees=45
  swing2Motion=Free
  twistLimitDegrees=45
  twistMotion=Free
 }
 {bone1Name=dyn_calfStrapC_01_l, bone2Name=MSTR_calfStrapC_l, swing1LimitDegrees=45, swing2LimitDegrees=45, twistLimitDegrees=0}
 {
  bone1Name=MSTR_calfStrapD_l
  bone2Name=pelvis
  swing1LimitDegrees=45
  swing1Motion=Free
  swing2LimitDegrees=45
  swing2Motion=Free
  twistLimitDegrees=45
  twistMotion=Free
 }
 {bone1Name=dyn_calfStrapD_01_l, bone2Name=MSTR_calfStrapD_l, swing1LimitDegrees=45, swing2LimitDegrees=45, twistLimitDegrees=0}
 {
  bone1Name=MSTR_calfStrapA_l
  bone2Name=pelvis
  swing1LimitDegrees=45
  swing1Motion=Free
  swing2LimitDegrees=45
  swing2Motion=Free
  twistLimitDegrees=45
  twistMotion=Free
 }
 {bone1Name=dyn_calfStrapA_01_l, bone2Name=MSTR_calfStrapA_l, swing1LimitDegrees=45, swing2LimitDegrees=45, twistLimitDegrees=0}
 {
  bone1Name=MSTR_calfStrapB_l
  bone2Name=pelvis
  swing1LimitDegrees=45
  swing1Motion=Free
  swing2LimitDegrees=45
  swing2Motion=Free
  twistLimitDegrees=45
  twistMotion=Free
 }
 {bone1Name=dyn_calfStrapB_01_l, bone2Name=MSTR_calfStrapB_l, swing1LimitDegrees=45, swing2LimitDegrees=45, twistLimitDegrees=0}
 {
  bone1Name=MSTR_calfStrapC_r
  bone2Name=pelvis
  swing1LimitDegrees=45
  swing1Motion=Free
  swing2LimitDegrees=45
  swing2Motion=Free
  twistLimitDegrees=45
  twistMotion=Free
 }
 {bone1Name=dyn_calfStrapC_01_r, bone2Name=MSTR_calfStrapC_r, swing1LimitDegrees=45, swing2LimitDegrees=45, twistLimitDegrees=0}
 {
  bone1Name=MSTR_calfStrapD_r
  bone2Name=pelvis
  swing1LimitDegrees=45
  swing1Motion=Free
  swing2LimitDegrees=45
  swing2Motion=Free
  twistLimitDegrees=45
  twistMotion=Free
 }
 {bone1Name=dyn_calfStrapD_01_r, bone2Name=MSTR_calfStrapD_r, swing1LimitDegrees=45, swing2LimitDegrees=45, twistLimitDegrees=0}
 {
  bone1Name=MSTR_calfStrapA_r
  bone2Name=pelvis
  swing1LimitDegrees=45
  swing1Motion=Free
  swing2LimitDegrees=45
  swing2Motion=Free
  twistLimitDegrees=45
  twistMotion=Free
 }
 {bone1Name=dyn_calfStrapA_01_r, bone2Name=MSTR_calfStrapA_r, swing1LimitDegrees=45, swing2LimitDegrees=45, twistLimitDegrees=0}
 {
  bone1Name=MSTR_calfStrapB_r
  bone2Name=pelvis
  swing1LimitDegrees=45
  swing1Motion=Free
  swing2LimitDegrees=45
  swing2Motion=Free
  twistLimitDegrees=45
  twistMotion=Free
 }
 {bone1Name=dyn_calfStrapB_01_r, bone2Name=MSTR_calfStrapB_r, swing1LimitDegrees=45, swing2LimitDegrees=45, twistLimitDegrees=0}
]
