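# Animation Blueprint for the NeonFriction Await male cosmetic body, layering secondary motion and material toggles onto the attached parent's pose.
# Copied pose feeds a Control Rig, skirt AnimDynamics, bone modifiers for hidden legs and skydiving backpack offset, then RigidBody.
# Each update refreshes state flags and computes gravity and rigid-body settings from a parameter asset.
# Events toggle "Hide Element" material parameters using NeonFriction backpack/pickaxe tags and skydiving.
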AnimBlueprint /BRCosmetics/Characters/Player/Male/Medium/Bodies/M_MED_NeonFriction_Await/Meshes/M_MED_NeonFriction_Await_AnimBP
target_skeleton=&/M_MED_NeonFriction_Await_Skeleton
graphs [
 AnimGraph {nodes [
   n1 = Control Rig!(
    Source=Copy Pose From Mesh() {bUseAttachedParent=true}
    Alpha
    crouching:bool=Get crouching() -> (crouching:bool)
    melee:bool=Get melee() -> (melee:bool)
    passenger:bool=Get bIsValetPassenger() -> (bIsValetPassenger:bool)
   ) -> (Pose) {controlRigAssetReference {blueprintRigClass=&/M_MED_NeonFriction_Await_CtrlRig.M_MED_NeonFriction_Await_CtrlRig_C}}
   n2 = Anim Dynamics - Bone: dyn_frontSkirt_02_l!(ComponentPose=Local To Component(LocalPose=@n1), GravityOverride:Vector="0,0,-980", ExternalForce="0,0,0") -> (Pose) {
    alpha=0.75
    angularDampingOverride=0.999
    angularSpringConstant=1000
    bAngularSpring=true
    bUseGravityOverride=true
    gravityScale=0.35
    simulationSpace=BoneRelative
    boundBone {boneName=dyn_frontSkirt_02_l}
    componentAppliedLinearAccClamp=<0 0 0>
    gravityOverride {z=-980}
    physicsBodyDefinitions [{
      boundBone {boneName=dyn_frontSkirt_02_l}
      boxExtents=<20 30 30>
      constraintSetup {angularLimitsMax=<0 0 5>, angularLimitsMin=<0 -5 -5>, angularTarget=<1 0 0>}
      localJointOffset=<2 0 0>
     }]
    relativeSpaceBone {boneName=pelvis}
    simSpaceSettings {maxAngularAcceleration=20, maxAngularVelocity=5, simSpaceAngularAlpha=1}
   }
   n3 = Anim Dynamics - Bone: dyn_sideSkirt_02_l!(ComponentPose=@n2, GravityOverride:Vector="0,0,-980", ExternalForce="0,0,0") -> (Pose) {
    alpha=0.75
    angularDampingOverride=0.999
    angularSpringConstant=1000
    bAngularSpring=true
    bUseGravityOverride=true
    gravityScale=0.35
    simulationSpace=BoneRelative
    boundBone {boneName=dyn_sideSkirt_02_l}
    componentAppliedLinearAccClamp=<0 0 0>
    gravityOverride {z=-980}
    physicsBodyDefinitions [{
      boundBone {boneName=dyn_sideSkirt_02_l}
      boxExtents=<20 30 30>
      constraintSetup {angularLimitsMax=<0 0 5>, angularLimitsMin=<0 -5 -5>, angularTarget=<1 0 0>}
      localJointOffset=<2 0 0>
     }]
    relativeSpaceBone {boneName=pelvis}
    simSpaceSettings {maxAngularAcceleration=20, maxAngularVelocity=5, simSpaceAngularAlpha=1}
   }
   n4 = Anim Dynamics - Bone: dyn_backSkirt_02_l!(ComponentPose=@n3, GravityOverride:Vector="0,0,-980", ExternalForce="0,0,0") -> (Pose) {
    alpha=0.75
    angularDampingOverride=0.999
    angularSpringConstant=1000
    bAngularSpring=true
    bUseGravityOverride=true
    gravityScale=0.35
    simulationSpace=BoneRelative
    boundBone {boneName=dyn_backSkirt_02_l}
    componentAppliedLinearAccClamp=<0 0 0>
    gravityOverride {z=-980}
    physicsBodyDefinitions [{
      boundBone {boneName=dyn_backSkirt_02_l}
      boxExtents=<20 30 30>
      constraintSetup {angularLimitsMax=<0 0 5>, angularLimitsMin=<0 -5 -5>, angularTarget=<1 0 0>}
      localJointOffset=<2 0 0>
     }]
    relativeSpaceBone {boneName=pelvis}
    simSpaceSettings {maxAngularAcceleration=20, maxAngularVelocity=5, simSpaceAngularAlpha=1}
   }
   n5 = Anim Dynamics - Bone: dyn_frontSkirt_02_r!(ComponentPose=@n4, GravityOverride:Vector="0,0,-980", ExternalForce="0,0,0") -> (Pose) {
    alpha=0.75
    angularDampingOverride=0.999
    angularSpringConstant=1000
    bAngularSpring=true
    bUseGravityOverride=true
    gravityScale=0.35
    simulationSpace=BoneRelative
    boundBone {boneName=dyn_frontSkirt_02_r}
    componentAppliedLinearAccClamp=<0 0 0>
    gravityOverride {z=-980}
    physicsBodyDefinitions [{
      boundBone {boneName=dyn_frontSkirt_02_r}
      boxExtents=<20 30 30>
      constraintSetup {angularLimitsMax=<0 0 5>, angularLimitsMin=<0 -5 -5>, angularTarget=<1 0 0>}
      localJointOffset=<2 0 0>
     }]
    relativeSpaceBone {boneName=pelvis}
    simSpaceSettings {maxAngularAcceleration=20, maxAngularVelocity=5, simSpaceAngularAlpha=1}
   }
   n6 = Anim Dynamics - Bone: dyn_sideSkirt_02_r!(ComponentPose=@n5, GravityOverride:Vector="0,0,-980", ExternalForce="0,0,0") -> (Pose) {
    alpha=0.75
    angularDampingOverride=0.999
    angularSpringConstant=1000
    bAngularSpring=true
    bUseGravityOverride=true
    gravityScale=0.35
    simulationSpace=BoneRelative
    boundBone {boneName=dyn_sideSkirt_02_r}
    componentAppliedLinearAccClamp=<0 0 0>
    gravityOverride {z=-980}
    physicsBodyDefinitions [{
      boundBone {boneName=dyn_sideSkirt_02_r}
      boxExtents=<20 30 30>
      constraintSetup {angularLimitsMax=<0 0 5>, angularLimitsMin=<0 -5 -5>, angularTarget=<1 0 0>}
      localJointOffset=<2 0 0>
     }]
    relativeSpaceBone {boneName=pelvis}
    simSpaceSettings {maxAngularAcceleration=20, maxAngularVelocity=5, simSpaceAngularAlpha=1}
   }
   n7 = Anim Dynamics - Bone: dyn_backSkirt_02_r!(ComponentPose=@n6, GravityOverride:Vector="0,0,-980", ExternalForce="0,0,0") -> (Pose) {
    alpha=0.75
    angularDampingOverride=0.999
    angularSpringConstant=1000
    bAngularSpring=true
    bUseGravityOverride=true
    gravityScale=0.35
    simulationSpace=BoneRelative
    boundBone {boneName=dyn_backSkirt_02_r}
    componentAppliedLinearAccClamp=<0 0 0>
    gravityOverride {z=-980}
    physicsBodyDefinitions [{
      boundBone {boneName=dyn_backSkirt_02_r}
      boxExtents=<20 30 30>
      constraintSetup {angularLimitsMax=<0 0 5>, angularLimitsMin=<0 -5 -5>, angularTarget=<1 0 0>}
      localJointOffset=<2 0 0>
     }]
    relativeSpaceBone {boneName=pelvis}
    simSpaceSettings {maxAngularAcceleration=20, maxAngularVelocity=5, simSpaceAngularAlpha=1}
   }
   n8 = Anim Dynamics - Bone: dyn_frontSkirt_02!(ComponentPose=@n7, GravityOverride:Vector="0,0,-980", ExternalForce="0,0,0") -> (Pose) {
    alpha=0.75
    angularDampingOverride=0.999
    angularSpringConstant=1000
    bAngularSpring=true
    bUseGravityOverride=true
    gravityScale=0.35
    simulationSpace=BoneRelative
    boundBone {boneName=dyn_frontSkirt_02}
    componentAppliedLinearAccClamp=<0 0 0>
    gravityOverride {z=-980}
    physicsBodyDefinitions [{
      boundBone {boneName=dyn_frontSkirt_02}
      boxExtents=<20 30 30>
      constraintSetup {angularLimitsMax=<0 0 5>, angularLimitsMin=<0 -5 -5>, angularTarget=<1 0 0>}
      localJointOffset=<2 0 0>
     }]
    relativeSpaceBone {boneName=pelvis}
    simSpaceSettings {maxAngularAcceleration=20, maxAngularVelocity=5, simSpaceAngularAlpha=1}
   }
   n9 = Anim Dynamics - Bone: dyn_backSkirt_02!(ComponentPose=@n8, GravityOverride:Vector="0,0,-980", ExternalForce="0,0,0") -> (Pose) {
    alpha=0.75
    angularDampingOverride=0.999
    angularSpringConstant=1000
    bAngularSpring=true
    bUseGravityOverride=true
    gravityScale=0.35
    simulationSpace=BoneRelative
    boundBone {boneName=dyn_backSkirt_02}
    componentAppliedLinearAccClamp=<0 0 0>
    gravityOverride {z=-980}
    physicsBodyDefinitions [{
      boundBone {boneName=dyn_backSkirt_02}
      boxExtents=<20 30 30>
      constraintSetup {angularLimitsMax=<0 0 5>, angularLimitsMin=<0 -5 -5>, angularTarget=<1 0 0>}
      localJointOffset=<2 0 0>
     }]
    relativeSpaceBone {boneName=pelvis}
    simSpaceSettings {maxAngularAcceleration=20, maxAngularVelocity=5, simSpaceAngularAlpha=1}
   }
   n10 = Get left_leg_hidden() -> (left_leg_hidden:double)
   n11 = "Transform (Modify) Bone - Bone: attach_backpack"(
    ComponentPose="Transform (Modify) Bone - Bone: calf_r"(
     ComponentPose="Transform (Modify) Bone - Bone: calf_l"(ComponentPose=@n9, Alpha=@n10, Translation="0,0,0", Rotation="0,0,0", Scale="1,0.5,0.5") {scaleMode=BMM_Replace, scaleSpace=BCS_BoneSpace, boneToModify {boneName=calf_l}, scale {y=0.5, z=0.5}}
     Alpha=Get right_leg_hidden() -> (right_leg_hidden:double)
     Translation="0,0,0"
     Rotation="0,0,0"
     Scale="1,0.5,0.5"
    ) {scaleMode=BMM_Replace, scaleSpace=BCS_BoneSpace, boneToModify {boneName=calf_r}, scale {y=0.5, z=0.5}}
    Alpha=Get skydive_floating() -> (skydive_floating:double)
    Translation="0,-6,4"
    Rotation="0,0,0"
    Scale="1,1,1"
   ) {
    scaleSpace=BCS_BoneSpace
    translationMode=BMM_Additive
    translationSpace=BCS_BoneSpace
    boneToModify {boneName=attach_backpack}
    translation {y=-6, z=4}
   }
   n12 = RigidBody!(
    ComponentPose=@n11
    Alpha
    OverrideWorldGravity:Vector="0,0,0"
    ExternalForce
    SimSpaceSettings:Sim Space Settings Structure=(WorldAlpha=1,VelocityScaleZ=0.4,DampingAlpha=1,MaxLinearVelocity=300,MaxAngularVelocity=10,MaxLinearAcceleration=1000,MaxAngularAcceleration=40,ExternalLinearDragV=(X=0.3,Y=0.3,Z=0.3),ExternalLinearVelocity=(X=0,Y=0,Z=0),ExternalAngularVelocity=(X=0,Y=0,Z=0))
   ) -> (Pose) {
    bClampLinearTranslationLimitToRefPose=true
    bOverrideWorldGravity=true
    cachedBoundsScale=1
    overridePhysicsAsset=&/M_MED_NeonFriction_Await_Physics
    simulationSpace=BaseBoneSpace
    alphaBoolBlend {blendOption=Linear}
    baseBoneRef {boneName=root}
    componentAppliedLinearAccClamp=<0 0 0>
    overrideWorldGravity {z=-980}
    simSpaceSettings {
     maxAngularAcceleration=40
     maxAngularVelocity=10
     maxLinearAcceleration=1000
     maxLinearVelocity=300
     velocityScaleZ=0.4
     worldAlpha=1
     externalLinearDragV=<0.3 0.3 0.3>
    }
   }
   Output Pose(Result=Component To Local(ComponentPose=@n12)) {layerGroup=None}
  ]}
 EventGraph {nodes [
   Event BlueprintUpdateAnimation()
   > n1 = "Gravity Override State Function (Parameter Asset) (Nativized)"(
    CharacterPartAnimInstance=Self-Reference()
    GameTimeInSeconds=GetGameTimeInSeconds()
    Parameters=&/M_MED_NeonFriction_Await_Parameters
   )
   > Set Gravity Override(Gravity Override=@n1)
   > Set RBAN Settings Out(RBAN Settings Out=@n1.OutRigidBodyAnimNodeParameters)
   > Set crouching(crouching:bool=Get bIsCrouching()) -> (Output_Get:bool)
   n2 = Make Array("[0]"=(TagName="Cosmetics.MetaTags.NeonFriction"))
   n3 = BackpackItemDefinitionHasAnyMetaTag!(Tags:Array of Gameplay Tag Structures=@n2, bRequiresEquipped:bool=false) -> (ReturnValue:bool)
   > Set backpack_neonfriction_equipped(backpack_neonfriction_equipped:bool=@n3) -> (Output_Get:bool)
   > Set melee(melee:bool=Get bIsPlayingMeleeAnim()) -> (Output_Get:bool)
   > Set skydive_floating(skydive_floating:double="To Float (Boolean)"(InBool=Get bIsSkyDiving())) -> (Output_Get:double)
   > Set left_leg_hidden(left_leg_hidden:double="To Float (Boolean)"(InBool=Get bIsLeftLegHidden() -> (bIsLeftLegHidden:bool))) -> (Output_Get:double)
   > Set right_leg_hidden(right_leg_hidden:double="To Float (Boolean)"(InBool=Get bIsRightLegHidden() -> (bIsRightLegHidden:bool))) -> (Output_Get:double)
   n4 = Event OnWeaponChanged()
   n5 = GetPartSkeletalMeshComponent()
   n6 = Get Dynamic Material Instance(execute=[@n4.then, @n13.then], SkeletalMeshComponent=@n5)
   > n7 = Get Dynamic Material Instance(SkeletalMeshComponent=@n5, ElementIndex=1)
   > n8 = GetOrCreateDynamicMaterialSlotOverlayMaterial!(execute, MeshComponent:Mesh Component Object Reference=@n5, Index:int=0, MaterialName:Name=None) -> (then, ReturnValue:Material Instance Dynamic Object Reference)
   > n9 = Branch(Condition=@n3)
   > SetScalarParameterValue(self=@n6, ParameterName=Hide Element 02, Value=0)
   > SetScalarParameterValue(self=@n7, ParameterName=Hide Element 02, Value=0)
   > SetScalarParameterValue(self=@n8, ParameterName=Hide Element 02, Value=0)
   > Set backpack_neonfriction_equipped(backpack_neonfriction_equipped=false)
   SetScalarParameterValue(execute=@n9.else, self=@n6, ParameterName=Hide Element 02, Value=1)
   > SetScalarParameterValue(self=@n7, ParameterName=Hide Element 02, Value=1)
   > SetScalarParameterValue(self=@n8, ParameterName=Hide Element 02, Value=1)
   n10 = Event OnBeginSkydiving()
   SetScalarParameterValue(execute=[@n10.then, @n12.else, @n15.then, @n14.else], self=@n6, ParameterName=Hide Element 01, Value=0)
   > SetScalarParameterValue(self=@n7, ParameterName=Hide Element 01, Value=0)
   Event OnEndSkydiving()
   n11 = PickaxeItemDefinitionHasAnyMetaTag(Tags=@n2, bRequiresEquipped=true)
   > n12 = Branch(Condition=@n11)
   SetScalarParameterValue(execute=[@n12.then, @n15.else], self=@n6, ParameterName=Hide Element 01, Value=1)
   > SetScalarParameterValue(self=@n7, ParameterName=Hide Element 01, Value=1)
   n13 = Event OnCharacterCustomizationLoaded()
   n14 = Branch(Condition=@n11)
   > n15 = Branch(Condition=Get bIsSkyDiving())
  ]}
]
variables [
 backpack_neonfriction_equipped {type=boolean}
 crouching {type=boolean}
 Gravity Override {type=Vector}
 left_leg_hidden {type=number}
 melee {type=boolean}
 passenger {type=boolean}
 RBAN Settings Out {
  type=FortRigidBodyAnimNodeParameters
  default {simSpaceSettings {dampingAlpha=0, maxAngularAcceleration=16, maxAngularVelocity=4, maxLinearAcceleration=800, maxLinearVelocity=200, velocityScaleZ=0.5}}
 }
 right_leg_hidden {type=number}
 skydive_floating {type=number}
]
