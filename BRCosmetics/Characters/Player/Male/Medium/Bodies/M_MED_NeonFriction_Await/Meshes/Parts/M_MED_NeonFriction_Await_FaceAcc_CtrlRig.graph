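# Control Rig for the NeonFriction_Await face accessory part that applies corrective transforms to hair cluster bones, including the bottom pair, as the head rotates.
# Setup creates a null at the head and reparents it to neck_01, keeping its global transform, as a rotation reference.
# The DeltaRotation v4 function swing-twist decomposes head rotation against that null, remaps the Z angle to a clamped weight, and drives Modify Transforms with it.
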ControlRigBlueprint /BRCosmetics/Characters/Player/Male/Medium/Bodies/M_MED_NeonFriction_Await/Meshes/Parts/M_MED_NeonFriction_Await_FaceAcc_CtrlRig
graphs [
 New Function_ContainedGraph {nodes [
   n1 = RigVMFunctionEntryNode()
   RigVMFunction_MathBoolToInteger = &RigVMFunction_MathBoolToInteger(Value=@n1.onlyTwist)
   RigVMFunction_MathQuaternionSwingTwist = &RigVMFunction_MathQuaternionSwingTwist(
    twistAxis=@n1.twistAxis
    Input=&RigUnit_GetRelativeTransformForItem(bChildInitial=False, bParentInitial=False, child=@n1.child, parent=@n1.parent).RelativeTransform.Rotation
   )
   n2 = RigVMDispatchNode(
    Values=((),())
    Index=@RigVMFunction_MathBoolToInteger.Result
    Values.0=@RigVMFunction_MathQuaternionSwingTwist.Swing
    Values.1=@RigVMFunction_MathQuaternionSwingTwist.Twist
   ) {template="DISPATCH_RigVMDispatch_SelectInt32(in Index,in Values,out Result)"}
   RigVMFunction_MathQuaternionToEuler = &RigVMFunction_MathQuaternionToEuler(Value=@n2.Result, RotationOrder=@n1.Rotation Order)
   > DISPATCH_RigVMDispatch_Print = RigVMDispatchNode(ScreenColor=(R=1,G=1,B=1,A=1), ScreenDuration=1.5, Enabled=@n1.debugRotation, Value=@RigVMFunction_MathQuaternionToEuler.Result) {template="DISPATCH_RigVMDispatch_Print(in Prefix,in Value,in Enabled,in ScreenDuration,in ScreenColor)"}
   Cast = RigVMDispatchNode(Value=@n1.triggerAxis) {template="DISPATCH_RigVMDispatch_CastEnumToInt(in Value,out Result)"}
   DISPATCH_RigVMDispatch_SelectInt32 = RigVMDispatchNode(
    Values=(0,0,0)
    Index=@Cast.Result
    Values.0=@RigVMFunction_MathQuaternionToEuler.Result.X
    Values.1=@RigVMFunction_MathQuaternionToEuler.Result.Y
    Values.2=@RigVMFunction_MathQuaternionToEuler.Result.Z
   ) {template="DISPATCH_RigVMDispatch_SelectInt32(in Index,in Values,out Result)"}
   Remap = &RigVMFunction_MathDoubleRemap(
    bClamp=False
    Value=@DISPATCH_RigVMDispatch_SelectInt32.Result
    TargetMinimum=@n1.outputMinMax.X
    TargetMaximum=@n1.outputMinMax.Y
    SourceMinimum=@n1.rotationMinMax.X
    SourceMaximum=@n1.rotationMinMax.Y
   )
   Absolute = &RigVMFunction_MathDoubleAbs(Value=@Remap.Result)
   Multiply = &RigVMFunction_MathDoubleMul(
    A=&RigVMFunction_MathDoubleClamp(Value=RigVMDispatchNode(True=@Absolute.Result, Condition=@n1.absoluteValue, False=@Remap.Result) {template="DISPATCH_RigVMDispatch_If(in Condition,in True,in False,out Result)"}.Result).Result
    B=@n1.premult
   )
   > RigVMFunctionReturnNode(outputQuat=@n2.Result, outputWeight=@Multiply.Result)
  ]}
 RigVMFunctionLibrary {nodes [DeltaRotation v4 = RigVMCollapseNode(
    ExecuteContext=()
    Rotation Order=ZYX
    child=(Type=Bone,Name="None")
    outputMinMax=(X=0,Y=1)
    parent=(Type=Bone,Name="None")
    premult=1
    rotationMinMax=(X=0,Y=0)
    twistAxis=(X=1,Y=0,Z=0)
   )]}
 RigVMModel {nodes [
   n1 = &RigUnit_PrepareForExecution()
   &RigUnit_HierarchyAddNull(
    Name=head_null
    Parent=(Type=Bone,Name="head")
    Space=LocalSpace
    Transform=(Rotation=(X=0,Y=0,Z=0,W=1),Translation=(X=0,Y=0,Z=0),Scale3D=(X=1,Y=1,Z=1))
    ExecutePin=@n1.ExecutePin
   )
   > /Script/ControlRig.RigUnit_SwitchParent(Child=(Type=Null,Name="head_null"), Mode=ParentItem, Parent=(Type=Bone,Name="neck_01"), bMaintainGlobal=True) {is=struct RigVMUnitNode}
   RigUnit_BeginExecution = &RigUnit_BeginExecution()
   DeltaRotation v4 = RigVMFunctionReferenceNode(
    Rotation Order=ZYX
    child=(Type=Bone,Name="head")
    outputMinMax=(X=0,Y=1)
    parent=(Type=Null,Name="head_null")
    premult=1
    rotationMinMax=(X=0,Y=-25)
    triggerAxis=Z
    twistAxis=(X=1,Y=0,Z=0)
    ExecuteContext=@RigUnit_BeginExecution.ExecutePin
   )
   > &RigUnit_ModifyTransforms(
    ItemToModify="((Item=(Type=Bone,Name=\"dyn_haircluster_bt_l_mstr\"),Transform=(Rotation=(X=0,Y=0,Z=-0.5,W=0.866),Translation=(X=-1.5,Y=0,Z=0),Scale3D=(X=1,Y=1,Z=1))),(Item=(Type=Bone,Name=\"dyn_haircluster_bt_r_mstr\"),Transform=(Rotation=(X=0,Y=0,Z=-0.5,W=0.866),Translatio ...(truncated)"
    WeightMinimum=0
    Weight=@"DeltaRotation v4".outputWeight
   )
  ]}
]
hierarchy [
 root {loc=<0 0 0>, rot=<0 0 0 1>}
 pelvis {parent=root, loc=<0 -0.7071 93.1191>, rot=<0.0127 0.707 -0.0127 -0.707>}
 spine_01 {parent=pelvis, loc=<6.7494 -0.6388 0>, rot=<0 0 0.034 -0.9994>}
 spine_02 {parent=spine_01, loc=<7.2518 -0.035 0>, rot=<0 0 0.0238 -0.9997>}
 spine_03 {parent=spine_02, loc=<6.0104 0 0>, rot=<0 0 0.0149 -0.9999>}
 spine_04 {parent=spine_03, loc=<8.1024 0 0>, rot=<0 0 -0.0132 -0.9999>}
 spine_05 {parent=spine_04, loc=<10.0146 0 0>, rot=<0 0 -0.0015 -1>}
 neck_01 {parent=spine_05, loc=<11.3138 0.1529 0>, rot=<0 0 -0.2003 -0.9797>}
 neck_02 {parent=neck_01, loc=<4.5627 0 0>, rot=<0 0 0.0504 -0.9987>}
 head {parent=neck_02, loc=<5.3914 0 0>, rot=<0 0 0.115 -0.9934>}
 MSTR_mask {parent=head, loc=<12.1817 -0.4457 -10.8597>, rot=<0.0373 -0.6698 -0.1267 -0.7307>}
 dyn_mask {parent=MSTR_mask, loc=<0 0 0>, rot=<0 0 0 1>}
 dyn_earring_r_mstr {parent=head, loc=<-0.0106 -0.1929 8.5774>, rot=<0.9561 -0.0022 -0.0188 0.2924>}
 dyn_earring_r_01 {parent=dyn_earring_r_mstr, loc=<0 0 0>, rot=<0 0 0 1>}
 dyn_earring_r_02 {parent=dyn_earring_r_01, loc=<-1.3101 0 0>, rot=<0 0 0 1>}
 dyn_earring_l_mstr {parent=head, loc=<-0.0108 -0.1929 -8.5774>, rot=<-0.0022 -0.9561 -0.2924 -0.0188>}
 dyn_earring_l_01 {parent=dyn_earring_l_mstr, loc=<0 0 0>, rot=<0 0 0 1>}
 dyn_earring_l_02 {parent=dyn_earring_l_01, loc=<1.3103 0 0>, rot=<0 0 0 1>}
 dyn_haircluster_l_mstr {parent=head, loc=<7.686 1.6176 -10.6075>, rot=<0.1237 -0.9781 -0.0671 -0.1535>}
 dyn_haircluster_l {parent=dyn_haircluster_l_mstr, loc=<0 0 0>, rot=<0 0 0 1>}
 dyn_haircluster_bt_l_mstr {parent=head, loc=<-0.8548 -6.4998 -4.6644>, rot=<0.0961 -0.9657 -0.241 -0.0072>}
 dyn_haircluster_bt_l {parent=dyn_haircluster_bt_l_mstr, loc=<0 0 0>, rot=<0 0 0 1>}
 dyn_haircluster_bk_mstr {parent=head, loc=<9.95 -11.1615 -1.0612>, rot=<0.2961 -0.6614 0.61 -0.3205>}
 dyn_haircluster_bk {parent=dyn_haircluster_bk_mstr, loc=<0 0 0>, rot=<0 0 0 1>}
 dyn_haircluster_tp_mstr {parent=head, loc=<20.9802 0.4922 -0.0451>, rot=<0.565 -0.4068 0.3854 -0.6056>}
 dyn_haircluster_tp {parent=dyn_haircluster_tp_mstr, loc=<0.7463 -0.6161 0.5018>, rot=<0 0 0 1>}
 dyn_haircluster_r_mstr {parent=head, loc=<7.6857 1.6176 10.6075>, rot=<0.9781 0.1237 -0.1535 0.0671>}
 dyn_haircluster_r {parent=dyn_haircluster_r_mstr, loc=<0 0 0>, rot=<0 0 0 1>}
 dyn_haircluster_bt_r_mstr {parent=head, loc=<-0.8545 -6.4998 4.6644>, rot=<0.9657 0.0961 -0.0072 0.241>}
 dyn_haircluster_bt_r {parent=dyn_haircluster_bt_r_mstr, loc=<0 0 0>, rot=<0 0 0 1>}
 dyn_haircluster_bk_l_mstr {parent=head, loc=<9.9447 -10.5134 -7.0385>, rot=<0.2961 -0.6614 0.61 -0.3205>}
 dyn_haircluster_bk_l {parent=dyn_haircluster_bk_l_mstr, loc=<0 0 0>, rot=<0 0 0 1>}
 dyn_haircluster_bk_r_mstr {parent=head, loc=<9.9455 -10.6166 5.8583>, rot=<0.2961 -0.6614 0.61 -0.3205>}
 dyn_haircluster_bk_r {parent=dyn_haircluster_bk_r_mstr, loc=<0 0 0>, rot=<0 0 0 1>}
 MSTR_frontHair_A_01_l {parent=head, loc=<8.993 3.3132 -10.3654>, rot=<0.0549 0.6875 0.7228 0.0436>}
 dyn_frontHair_A_01_l {parent=MSTR_frontHair_A_01_l, loc=<0 0 0>, rot=<0 0 0 1>}
 dyn_frontHair_A_02_l {parent=dyn_frontHair_A_01_l, loc=<4.4273 0 0>, rot=<-0.0886 0.3863 0.0409 -0.9172>}
 MSTR_frontHair_B_01_l {parent=head, loc=<5.5433 0.6705 -10.377>, rot=<-0.2103 -0.7075 -0.6709 -0.0713>}
 dyn_frontHair_B_01_l {parent=MSTR_frontHair_B_01_l, loc=<0 0 0>, rot=<0 0 0 1>}
 MSTR_frontHair_A_01_r {parent=head, loc=<8.9927 3.3132 10.3654>, rot=<-0.6875 0.0549 0.0436 -0.7228>}
 dyn_frontHair_A_01_r {parent=MSTR_frontHair_A_01_r, loc=<0 0 0>, rot=<0 0 0 1>}
 dyn_frontHair_A_02_r {parent=dyn_frontHair_A_01_r, loc=<-4.4268 0 -0.0001>, rot=<-0.0886 0.3863 0.0409 -0.9172>}
 MSTR_frontHair_B_01_r {parent=head, loc=<5.7083 1.5772 9.6115>, rot=<-0.7111 0.1438 0.0018 -0.6882>}
 dyn_frontHair_B_01_r {parent=MSTR_frontHair_B_01_r, loc=<0 0 0>, rot=<0 0 0 1>}
]
